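# Translucent, unlit, two-sided test material for spline meshes and Niagara mesh particles and ribbons in the Apprentice character's FX folder.
# Two rotated procedural bands blend background and foreground colors for emissive, while depth fade, particle alpha and an optional UV gradient build opacity.
# World position offset combines sine-pulsed SplineThicken width, depth-based camera offset, masked Perlin-noise displacement and a clamped, Niagara-driven positional interpolation.
# Shading path switches zero that offset and skip depth fade on mobile.
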Material /BRCosmetics/Characters/Player/Male/Medium/Bodies/M_MED_Apprentice/FX/M_SplineTestMaterial
graph {
 properties {
  bUsedWithNiagaraMeshParticles=true
  bUsedWithNiagaraRibbons=true
  bUsedWithSplineMeshes=true
  blendMode=BLEND_Translucent
  pixelDepthOffsetMode=PDOM_Legacy
  shadingModel=MSM_Unlit
  twoSided=true
  usedShadingModels=Unlit
 }
 nodes [
  Constant()
  n1 = ScalarParameter() {defaultValue=10, parameterName=Width}
  LinearInterpolate(
   A=ScalarParameter() {parameterName=Width Low}
   B=@n1
   Alpha=MaterialFunctionCall() {materialFunction=&DebugTimeSine}.Result
  ) {constA=10}
  n2 = Constant() {r=1}
  n3 = MaterialFunctionCall!(Angle=ScalarParameter() {defaultValue=0.25, parameterName=Rotate}, UV/2DVector) {materialFunction=/Game/Characters/Player/Male/Medium/Bodies/M_MED_Faux/Materials/MF_2DRotate}
  n4 = MaterialFunctionCall!(
   Width=ScalarParameter() {defaultValue=0.5, parameterName=Param_3}
   Sharpness=@n2
   Offset
   Direction Switch
   Compare
   Input Coordinates=@n3.Result
  ) {materialFunction=/Engine/Functions/Engine_MaterialFunctions03/Procedurals/GeneratedBand}
  n5 = VectorParameter() {parameterName=BackgroundColor, defaultValue=<rgba 0.0218 0.1573 0.5 0>}
  n6 = MaterialFunctionCall(Width=ScalarParameter() {defaultValue=0.3, parameterName=Bandwitdth}, Sharpness=@n2, Input Coordinates=@n3.Result) {materialFunction=&GeneratedBand}
  n7 = VectorParameter() {parameterName=ForegroundColor, defaultValue=<rgba 0.002 0.8012 2 0>}
  MaterialFunctionCall(Emissive=LinearInterpolate(A=Multiply(A=@n4.Result, B=@n5.RGB) {constB=0.05}, B=Multiply(A=@n6.Result, B=@n7.RGB), Alpha=@n6.Result)) {materialFunction=&MF_FX_EmissiveAdjustments}
  n8 = TextureCoordinate()
  n9 = Ceil(None=Sine(None=ComponentMask(None=@n8) {r=true}))
  n10 = Ceil(None=Cosine(None=ComponentMask(None=@n8) {g=true}) {period=2})
  n11 = OneMinus(None=@n10)
  n12 = Subtract(A=@n9, B=@n11)
  n13 = OneMinus(None=@n9)
  n14 = Multiply(A=Subtract(A=@n13, B=@n11)) {constB=0.75}
  n15 = Multiply(A=Subtract(A=@n9, B=@n10)) {constB=0.5}
  n16 = Multiply(A=Subtract(A=@n13, B=@n10)) {constB=0.25}
  Multiply(A=LinearInterpolate(A=Add(A=@n12, B=@n14), B=Add(A=@n15, B=@n16), Alpha=@n11), B=ParticleColor().RGB)
  n17 = DepthFade(FadeDistance=ScalarParameter() {defaultValue=20, group=Opacity, parameterName=DepthFade})
  n18 = Constant() {r=1}
  n19 = ComponentMask(None=TextureCoordinate()) {r=true}
  Multiply(
   A=Multiply(A=ShadingPathSwitch(Default=@n17, Mobile=@n18) {desc=Migrated from FeatureLevelSwitch}, B=ParticleColor().A)
   B=LinearInterpolate(A=Constant() {r=1}, B=OneMinus(None=@n19), Alpha=ScalarParameter() {parameterName=UseGradient})
  )
  n20 = TextureCoordinate()
  OneMinus(None=@n20)
  n21 = TextureCoordinate()
  n22 = DynamicParameter() {parameterIndex=1, defaultValue=<rgba 0 10 1 1>, paramNames [Random Offset, Noise Intensty, Dissolve, Dissolve Edge Intensity]}
  n23 = Add(
   A=Add(A=ComponentMask(None=@n21) {g=true}, B=@n22.Random Offset)
   B=Multiply(A=Time(), B=ScalarParameter() {defaultValue=0.5, parameterName=WidthNoiseSpeed}) {constB=0.5}
  )
  n24 = ScalarParameter() {defaultValue=2, parameterName=WidthNoiseFrequency}
  MaterialExpressionRemap_0 = MaterialXRemap!(
   None=Sine(None=Multiply(A=@n23, B=@n24))
   InputLow
   InputHigh
   TargetLow=ScalarParameter() {defaultValue=1, parameterName=Param_5}
   TargetHigh=ScalarParameter() {defaultValue=1.3}
  ) {targetHighDefault=2, targetLowDefault=1}
  n25 = Multiply(A=@n1, B=@MaterialExpressionRemap_0)
  n26 = Multiply(A=@n25, B=ScalarParameter() {defaultValue=1, parameterName=Param_4})
  n27 = ScalarParameter() {defaultValue=0.001, parameterName=Param_2}
  n28 = ComponentMask(None=TextureCoordinate()) {g=true}
  n29 = SmoothStep(Min=@n27, Max=ScalarParameter() {defaultValue=0.05, parameterName=Param_1}, Value=OneMinus(None=Add(A=@n28, B=@n27)))
  n30 = ScalarParameter() {defaultValue=0.5, parameterName=Roudness}
  n31 = ScalarParameter() {defaultValue=50, parameterName=DepthWPO}
  n32 = ShadingPathSwitch(Default=OneMinus(None=DepthFade(FadeDistance=@n31)), Mobile=Constant()) {desc=Migrated from FeatureLevelSwitch}
  n33 = ScalarParameter() {defaultValue=-22, parameterName=OffsetAmount}
  n34 = Add(
   A=Multiply(
    A=MaterialFunctionCall(WidthBase=@n26, WidthTip=@n25) {materialFunction=&SplineThicken}.WorldPositionOffset
    B=Power(Base=@n29, Exp=@n30)
   )
   B=MaterialFunctionCall(Offset Amount=Multiply(A=@n32, B=@n33)) {materialFunction=/Engine/Functions/Engine_MaterialFunctions02/WorldPositionOffset/CameraOffset}.Clamped Camera Offset
  )
  n35 = Constant()
  n36 = DynamicParameter() {defaultValue=<rgba 0 1 1 1>, paramNames [Random Offset, Spline Width Mult, Dissolve, Dissolve Edge Intensity]}
  n37 = VectorParameter() {
   group=WPO
   parameterName=WPO Depth UV
   sortPriority=6
   channelNames {a=WPO Depth V Pan, b=WPO Depth U Pan, g=WPO Depth V, r=WPO Depth U}
   defaultValue=<rgba 0 0.5 0 0.25>
  }
  n38 = ScalarParameter() {defaultValue=1, group=Noise, parameterName=Global Scale U}
  n39 = ScalarParameter() {defaultValue=1, group=Noise, parameterName=Global Scale V}
  n40 = Multiply(A=TextureCoordinate(), B=Multiply(A=AppendVector(A=@n37.WPO Depth U, B=@n37.WPO Depth U), B=AppendVector(A=@n38, B=@n39)))
  n41 = AppendVector(A=@n37.WPO Depth U Pan, B=@n37.WPO Depth U Pan)
  n42 = TextureSample(UVs=Add(A=AppendVector(A=@n36.Random Offset, B=@n36.Random Offset), B=Panner(Coordinate=@n40, Speed=@n41))) {samplerType=SAMPLERTYPE_LinearColor, texture=/Game/Global/Textures/Default/Tileable_Patterns/T_PerlinNoise_Uncompressed}
  n43 = MaterialFunctionCall(
   R=@n35
   G=Multiply(A=LinearInterpolate(Alpha=@n42.R) {constB=-1}, B=ScalarParameter() {defaultValue=4, group=WPO, parameterName=WPO Depth Intensity}) {constB=6}
   B=@n35
  ) {materialFunction=&AppendMany}
  n44 = ComponentMask(None=TextureCoordinate()) {g=true}
  n45 = OneMinus(None=@n44)
  n46 = Constant() {r=2}
  n47 = Constant() {r=1}
  n48 = Multiply(A=Multiply(A=Power(Base=@n45, Exp=@n46) {constExponent=0.5}, B=Power(Base=@n44, Exp=@n47) {constExponent=0.5}), B=Constant() {r=6}) {constB=4}
  n49 = ScalarParameter() {parameterName=RotateWPO}
  n50 = DynamicParameter() {defaultValue=<rgba 0 1 1 1>, paramNames [Random Offset, Spline Width Mult, Dissolve, Dissolve Edge Intensity]}
  n51 = VectorParameter() {group=WPO, parameterName=WPO Noise UV, sortPriority=6, defaultValue=<rgba 0 0.65 0 0.4>}
  n52 = Multiply(
   A=AppendVector(A=@n51.R, B=@n51.R)
   B=AppendVector(
    A=ScalarParameter() {defaultValue=1, group=Noise, parameterName=Global Scale U}
    B=ScalarParameter() {defaultValue=1, group=Noise, parameterName=Global Scale V}
   )
  )
  n53 = TextureCoordinate()
  n54 = AppendVector(A=ComponentMask(None=@n53) {r=true}, B=Power(Base=ComponentMask(None=@n53) {g=true}) {constExponent=1})
  n55 = Add(
   A=AppendVector(A=@n50.Random Offset, B=@n50.Random Offset)
   B=Panner(Coordinate=Multiply(A=@n54, B=@n52), Speed=AppendVector(A=@n51.B, B=@n51.B))
  )
  n56 = LinearInterpolate(Alpha=TextureSample(UVs=MaterialFunctionCall(Angle=@n49, UV/2DVector=@n55) {materialFunction=&MF_2DRotate}.Result) {samplerType=SAMPLERTYPE_LinearColor, texture=&T_PerlinNoise_Uncompressed}.R) {constA=-1}
  n57 = DynamicParameter() {parameterIndex=1, defaultValue=<rgba 0 10 1 1>, paramNames [Random Offset, Noise Intensty, Dissolve, Dissolve Edge Intensity]}
  n58 = Multiply(
   A=MaterialFunctionCall(R=@n56, G=Constant(), B=Multiply(A=@n56) {constB=0.5}) {materialFunction=&AppendMany}.RGB
   B=LinearInterpolate(
    A=ScalarParameter() {defaultValue=5, group=WPO, parameterName=WPO Noise Intensity}
    B=@n57.Noise Intensty
    Alpha=ScalarParameter() {parameterName=UseDynNoiseIntensity}
   )
  )
  n59 = OneMinus(None=ComponentMask(None=TextureCoordinate()) {g=true})
  n60 = OneMinus(None=@n59)
  n61 = Constant() {r=1.8}
  n62 = Power(Base=@n59, Exp=Constant() {r=0.9}) {constExponent=0.5}
  n63 = ScalarParameter() {group=SplineThicken, parameterName=WPO Top Mask}
  n64 = CurveAtlasRowParameter(CurveTime=ComponentMask(None=TextureCoordinate()) {g=true}) {
   atlas=/Game/Athena/LinearColorCurves/A_Atlas_Turtleneck_Glider
   curve=/Game/Accessories/FORT_Backpacks/Backpack_M_MED_Calavera/FX/Curve_Backpack_Calavera_01
   defaultValue=39
   group=SplineThicken
   parameterName=Curve Shape
  }
  n65 = LinearInterpolate(
   A=Multiply(A=Power(Base=@n60, Exp=@n61) {constExponent=0.5}, B=LinearInterpolate(B=@n62, Alpha=@n63) {constA=1})
   B=@n64.G
   Alpha=ScalarParameter() {group=WPO, parameterName=Lerp WPO Curve}
  )
  n66 = Transform(None=Add(A=Multiply(A=@n43.RGB, B=@n48), B=Multiply(A=@n58, B=@n65))) {transformSourceType=TRANSFORMSOURCE_Local}
  n67 = DynamicParameter() {
   desc=Parameters set by Niagara
   parameterIndex=3
   defaultValue=<rgba 0 1 1 1>
   paramNames [ShapeTypeShift, Spline Width Mult, Dissolve, Dissolve Edge Intensity]
  }
  n68 = MaterialFunctionCall(Float3=Subtract(
    A=MaterialFunctionCall(X=@n67.ShapeTypeShift, Y=@n67.ShapeTypeShift, Z=@n67.ShapeTypeShift) {materialFunction=&MakeFloat3}.Result
    B=ParticlePositionWS()
   )) {materialFunction=&BreakOutFloat3Components}
  n69 = AppendVector(A=@n68.R, B=@n68.R)
  n70 = ScalarParameter() {defaultValue=100, group=Interpolate, parameterName=Interpolate Clamp XY}
  n71 = Multiply(A=@n70) {constB=-1}
  n72 = ScalarParameter() {defaultValue=20, group=Interpolate, parameterName=Interpolate Clamp Z}
  n73 = Multiply(A=@n72) {constB=-1}
  n74 = ComponentMask(None=OneMinus(None=@n20)) {g=true}
  n75 = ComponentMask(None=@n20) {g=true}
  n76 = DynamicParameter() {parameterIndex=1, defaultValue=<rgba 0 10 1 1>, paramNames [Random Offset, Noise Intensty, Dissolve, Dissolve Edge Intensity]}
  n77 = LinearInterpolate(
   A=ComponentMask(None=@n20) {g=true}
   B=Multiply(A=Multiply(A=@n74, B=@n75)) {constB=1.5}
   Alpha=LinearInterpolate(
    A=ScalarParameter() {group=Interpolate, parameterName=Lerp Interpolate Mask Type}
    B=@n76.Random Offset
    Alpha=ScalarParameter() {group=SplineThicken, parameterName=UseDynamicParameterForShapeType}
   )
  )
  n78 = ScalarParameter() {defaultValue=1.7, group=Interpolate, parameterName=Interpolate Power}
  n79 = LinearInterpolate(
   B=AppendVector(A=Clamp(None=@n69, Min=@n71, Max=@n70) {maxDefault=1000, minDefault=-1000}, B=Clamp(None=@n68.B, Min=@n73, Max=@n72) {minDefault=-1})
   Alpha=Multiply(A=@n67.Dissolve Edge Intensity, B=Power(Base=@n77, Exp=@n78))
  )
  n80 = Constant()
  ShadingPathSwitch(
   Default=Add(
    A=Add(A=@n34, B=@n66)
    B=StaticSwitchParameter(True=@n79, False=@n80) {defaultValue=true, group=Interpolate, parameterName=Use Interpolate WPO, sortPriority=0}
   )
   Mobile=Constant()
  ) {desc=Migrated from FeatureLevelSwitch}
  VertexColor()
 ]
}
