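# Translucent unlit Niagara mesh-particle material for the Apprentice skin's wisp effects, coloured by particle color and alpha with optional edge fading.
# World position offset widens the mesh across a camera-facing axis, shaped by tapered tips and a gradient mask driven by dynamic parameters, with particle random flipping its direction.
# A second offset uses MF_MoveToCameraFrustum to squeeze the mesh to an offset object depth, disabled inside the near-clip distance.
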Material /BRCosmetics/Characters/Player/Male/Medium/Bodies/M_MED_Apprentice/FX/Materials/M_Skin_Apprentice_Wisps_Mesh
graph {
 properties {
  bUsedWithNiagaraMeshParticles=true
  blendMode=BLEND_Translucent
  pixelDepthOffsetMode=PDOM_Legacy
  shadingModel=MSM_Unlit
  usedShadingModels=Unlit
 }
 nodes [
  n1 = CameraPositionWS()
  n2 = WorldPosition()
  n3 = Constant3Vector() {constant=<rgba 0 1 0 0>}
  n4 = CrossProduct(A=Normalize(VectorInput=Subtract(A=@n1, B=@n2.XYZ)), B=Normalize(VectorInput=Transform(None=@n3)))
  n5 = Subtract(A=ComponentMask(None=TextureCoordinate()) {r=true}) {constB=0.5}
  n6 = ComponentMask(None=TextureCoordinate()) {g=true}
  n7 = Abs(None=ConstantBiasScale(None=@n6) {bias=-0.5, scale=2})
  n8 = Saturate(None=Divide(A=OneMinus(None=@n7), B=ScalarParameter() {defaultValue=0.3, parameterName=Tips Taper Size}))
  n9 = DynamicParameter() {paramNames [Width 0-1, Width Multiplier, Width Mask Size, Param4]}
  n10 = ScalarParameter() {parameterName=Additional Width Multiplier}
  n11 = Subtract(A=ParticleRandom()) {constB=0.5}
  n12 = Subtract(A=@n6, B=Saturate(None=Sign(None=@n11)))
  n13 = Multiply(
   A=Multiply(A=SquareRoot(None=@n8), B=Add(A=@n9.Width Multiplier, B=@n10))
   B=MaterialFunctionCall!(Gradient=Abs(None=@n12), Mask Position=@n9.Width 0-1, Mask Size=@n9.Width 0-1) {materialFunction=/Game/Effects/Fort_Effects/MaterialFunctions/MF_GradientMask}.Linear
  )
  n14 = TransformPosition(None=Constant3Vector() {constant=<rgba 0 0 33 0>}) {transformSourceType=TRANSFORMPOSSOURCE_View, transformType=TRANSFORMPOSSOURCE_World}
  n15 = Subtract(
   A=ScreenPosition().ViewportUV
   B=Multiply(
    A=MaterialFunctionCall(World Position=@n14) {materialFunction=&TransformToClipSpace}.Clip Space XY
    B=Constant2Vector() {g=-0.5, r=0.5}
   )
  )
  n16 = Add(
   A=MaterialFunctionCall!(World Position=ObjectPositionWS()) {materialFunction=/BRCosmetics/Accessories/FORT_Backpacks/Backpack_M_MED_Rumble/FX/Materials/MF_DepthFromWorldPosition}.Depth
   B=ScalarParameter() {defaultValue=33, parameterName=Depth Offset}
  )
  n17 = ScalarParameter() {defaultValue=33, parameterName=Near Clip}
  n18 = Saturate(None=Ceil(None=Subtract(A=@n16, B=@n17)))
  n19 = Saturate(None=ScalarParameter() {defaultValue=1, parameterName=Mesh Squeeze})
  Add(
   A=Multiply(A=Normalize(VectorInput=@n4), B=Multiply(A=@n5, B=@n13))
   B=Multiply(
    A=MaterialFunctionCall!(XY View Position=@n15, Z View Depth=@n16) {materialFunction=/BRCosmetics/Accessories/FORT_Backpacks/Backpack_M_MED_Rumble/FX/Materials/MF_MoveToCameraFrustum}.World Position Offset
    B=Multiply(A=@n18, B=@n19)
   )
  )
  n20 = ParticleColor()
  MaterialFunctionCall(Emissive=Multiply(A=@n20.RGB, B=ScalarParameter() {defaultValue=1, parameterName=Color Multiplier})) {materialFunction=&MF_FX_EmissiveAdjustments}
  n21 = ScalarParameter() {parameterName=Fade Edges}
  n22 = ConstantBiasScale(None=ComponentMask(None=TextureCoordinate()) {r=true}) {bias=-0.5, scale=2}
  n23 = SquareRoot(None=DotProduct(A=@n22, B=@n22))
  Multiply(A=OneMinus(None=Multiply(A=@n21, B=@n23)), B=@n20.A)
  VectorParameter() {parameterName=Color, defaultValue=<rgba 0 0.05 1 0>}
 ]
}
