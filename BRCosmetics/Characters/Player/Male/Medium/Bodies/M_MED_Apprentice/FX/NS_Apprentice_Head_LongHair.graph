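# Niagara effect for the long-hair head of the male Apprentice body, combining mesh hair strands, sparks and a corona flare on the skeletal mesh.
# SplineHair re-bursts short-lived strands, re-sampling partition-selected bone/socket ranges each update, then orienting, offsetting, growing and fading them.
# Sparks and flare fire timed bursts, with curl-noise, time-of-day-dimmed sparks.
# VariantID picks hair and spark colors, bisHeadHidden drives visibility tags, and front end lowers hair counts.
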NiagaraSystem /BRCosmetics/Characters/Player/Male/Medium/Bodies/M_MED_Apprentice/FX/NS_Apprentice_Head_LongHair
dependencies {
 usedDataInterfaces [
  &NiagaraDataInterfaceCurve
  &NiagaraDataInterfaceArrayColor
  &NiagaraDataInterfaceArrayInt32
  &NiagaraDataInterfaceArrayFloat
  &NiagaraDataInterfaceColorCurve
 ]
 usedDynamicInputs [&Divide_Float]
 usedModules [
  &SystemState
  &self:SystemSpawnScript.SystemScriptSource.SystemScriptGraph.NiagaraNodeAssignment_0.SetVariables_#1
  &self:SystemSpawnScript.SystemScriptSource.SystemScriptGraph.NiagaraNodeAssignment_1.SetVariables_#2
 ]
}
emitters {
 Impact_Flare_and_Corona {
  emitterUpdateScript [
   &EmitterState()
   &SpawnBurst_Instantaneous(Spawn Count=1, Use Spawn Probability=false)
   &SpawnBurst_Instantaneous(Spawn Count=1, Spawn Time=5.5, Use Spawn Probability=false) {module=SpawnBurst_Instantaneous001}
  ]
  particleSpawnScript [
   /Niagara/Modules/Spawn/Initialization/InitializeParticle(
    Color=<rgba 0.076 1 0.8922 1>
    Lifetime=0.3
    Mesh Scale=&UniformRangedVector(Minimum=<1.25 1.25 1.25>, Maximum=<1.4 1.4 1.4>)
    Module.Write Color=true
    Module.Write Lifetime=true
    Module.Write Mass=true
    Module.Write Material Random=true
    Module.Write Position=false
    Module.Write Scale=true
    Module.Write SpriteRotation=false
    Module.Write SpriteSize=true
    Position=&SimulationPosition()
    Sprite Size=&Multiply_Vector2D_ByFloat(Float=1, Vector2D=<200 175>)
   )
   /SetVariables(Particles.CoronaSpriteSize=&Vector2DFromFloat(Value=110)) {isSetParameters=true}
   &CameraOffset(Camera Offset Amount=15)
   /SetVariables(Particles.ImpactColor:LinearColor=<rgba 1 0.975 0.85 1>) {isSetParameters=true}
   &SkeletalMeshLocation(Bone Sampling Mode=::NewEnumerator4, MaskSpawnGroup=false, Skeletal Mesh=$User.Skeletal Mesh)
  ]
  particleUpdateScript [
   &ParticleState()
   &DynamicMaterialParameters(
    Index 0 Param 1=&FloatFromCurve(FloatCurve=DataInterface() {propertyValues {Curve {keys [<at 0 1 RCIM_Cubic>,<at 0.8 0 RCIM_Cubic>]}}})
    Index 0 Param 2=2
    Index 0 Param 3=1
    Index 0 Param 4=1
    Index 1 Param 1=&FloatFromCurve(FloatCurve=DataInterface() {propertyValues {Curve {keys [
         {arriveTangent=-2.3163, interpMode=RCIM_Cubic, time=0, value=2}
         {arriveTangent=-1.9019, interpMode=RCIM_Cubic, leaveTangent=-1.9019, tangentMode=RCTM_User, time=1, value=0.4}
        ]}}})
    Index 1 Param 2=0.4
    Index 1 Param 3=0.6
    Index 1 Param 4=1
    Module.Param0WriteEnabled=true
    Module.Param0WriteEnabled1=false
    Module.Param1WriteEnabled=true
    Module.Param1WriteEnabled1=true
    Module.Param2WriteEnabled=true
    Module.Param2WriteEnabled1=true
    Module.Param3WriteEnabled=true
    Module.Param3WriteEnabled1=true
    Write Parameter Index 1=true
   )
   /SetVariables(Particles.CoronaSpriteColor:LinearColor=&MakeLinearColorFromVectorAndFloat(
     "Vector (RGB)"=&MakeVectorFromLinearColorRGB(Color=<rgba 0.7 1 0.9719 0.5>)
     "Float (Alpha)"=&FloatFromCurve(FloatCurve=DataInterface() {propertyValues {Curve {keys [<at 0.35 1.5 RCIM_Cubic>,<at 0.95 0>]}}})
    )) {isSetParameters=true}
   &ScaleColor(
    Scale Alpha=&FloatFromCurve(FloatCurve=DataInterface() {propertyValues {Curve {keys [<at 0 1>,<at 1 0>]}}}, Scale Curve=0.2)
    Scale RGB=&MakeVectorFromLinearColorRGB(Color=<rgba 0 0.5333 1 1>)
    ScaleA=true
    ScaleRGB=true
   )
   /SetVariables(Particles.VisibilityTag=&MakeCustomIntFromBool(Bool=$User.bisHeadHidden, True INT=1, False INT=0)) {isSetParameters=true}
  ]
  rendererClasses [&NiagaraSpriteRendererProperties]
  renderers [
   {rendererClass=&NiagaraSpriteRendererProperties}
   {rendererClass=&NiagaraSpriteRendererProperties, rendererIndex=1}
  ]
 }
 Spark001 {
  emitterUpdateScript [
   &EmitterState(Life Cycle Mode=::NewEnumerator1, Loop Behavior=::NewEnumerator1, Loop Duration=7, UseLoopDelay=true)
   &SpawnBurst_Instantaneous(Spawn Count=30, Use Spawn Probability=false)
   &SpawnBurst_Instantaneous(Spawn Count=30, Spawn Time=5.5, Use Spawn Probability=false) {module=SpawnBurst_Instantaneous001}
  ]
  particleSpawnScript [
   /Niagara/Modules/Spawn/Initialization/V2/InitializeParticle(
    Color=&SelectLinearColorFromArray(
     Array Sampling Mode=::NewEnumerator1
     Direct Array Index=&MakeIntFromFloat(Float=$User.VariantID)
     Color Selection Array=DataInterface() {propertyValues {ColorData [<rgba 0.098 0.502 0.4902 1>, <rgba 0.098 0.502 0.4902 1>, <rgba 1 0.323 0.7996 1>, <rgba 0 1 0.2588 1>]}}
     Random Seed=?
    )
    Lifetime Max=0.8
    Lifetime Min=0.4
    Lifetime Mode=::NewEnumerator1
    Sprite Size Max=<3 7>
    Sprite Size Min=<1.5 3>
    Sprite Size Mode=::NewEnumerator2
    UsePositionOffset=false
    Write Material Random=true
   )
   &ShapeLocation(MaskSpawnGroup=false, Sphere Distribution=::NewEnumerator3, Sphere Radius=50, Uniform Distribution=0.5, UseOwnerScale=true)
   &SkeletalMeshLocation(MaskSpawnGroup=false, Mesh Sampling Type=::NewEnumerator5, OverrideRandomSeed=false, Skeletal Mesh=$User.Skeletal Mesh)
   &AddVelocityFromPoint(
    Module.Use Falloff Distance=false
    Velocity Strength=&Multiply_Float(A=0.8, B=&RandomRangeFloat(Minimum=300, Maximum=1000))
   ) {enabled=false}
   &AddVelocityFromPoint(
    Module.Use Falloff Distance=false
    Velocity Strength=&Multiply_Float(A=0.8, B=&RandomRangeFloat(Minimum=300, Maximum=1000))
   ) {enabled=false, module=AddVelocityFromPoint001}
   &AddVelocity(Velocity=&Multiply_VectorByFloat(Vector=$Particles.ShapeLocation.ShapeNormal, Float=350))
  ]
  particleUpdateScript [
   &ParticleState()
   &ScaleColor(
    Linear Color Curve=DataInterface() {propertyValues {
      AlphaCurve {keys [<at 0.0074 1>,<at 0.8905 1>]}
      BlueCurve {keys [<at 0 5>,<at 0.8991 1>]}
      GreenCurve {keys [<at 0 5>,<at 0.8991 1>]}
      RedCurve {keys [<at 0 5>,<at 0.8991 1>]}
     }}
    Scale Mode=::NewEnumerator2
   ) {enabled=false}
   &ScaleColor(
    Scale Alpha=&FNTimeOfDayScalar(Night Scale=0.5, Morning Scale=0.3, Evening Scale=0.3)
    Scale RGB=&VectorFromFloat(Value=&FloatFromCurve(FloatCurve=DataInterface() {propertyValues {Curve {keys [<at 0 20>,<at 0.2 20>,<at 1 5>]}}}))
    ScaleA=true
    ScaleRGB=true
   ) {module=ScaleColor001}
   &Drag(Drag=&RandomRangeFloat(Minimum=8, Maximum=15), Use Linear Drag=true, Use Rotational Drag=false)
   &CurlNoiseForce(Noise Strength=&RandomRangeFloat(Minimum=200, Maximum=400), Pan Noise=true)
   &GravityForce(Gravity=&RandomRangeVector(Minimum=<0 0 0>, Maximum=<0 0 -100>))
   &SolveForcesAndVelocity(Clamp Velocity=false, Limit Acceleration=false)
   &ScaleSpriteSizeBySpeed(Max Scale Factor=<1 5>, Min Scale Factor=<0.5 0.05>)
   &ScaleSpriteSize(Scale Factor=&MakeVector2D(
     X=&FloatFromCurve(FloatCurve=DataInterface() {propertyValues {Curve {keys [<at 0 0.5>,<at 0.1 1>,<at 1 0.5>]}}})
     Y=&FloatFromCurve(FloatCurve=DataInterface() {propertyValues {Curve {keys [<at 0 1>,<at 1 0>]}}})
    ))
   /SetVariables(Particles.VisibilityTag=&MakeCustomIntFromBool(Bool=$User.bisHeadHidden, True INT=1, False INT=0)) {isSetParameters=true}
  ]
  rendererClasses [&NiagaraSpriteRendererProperties]
  renderers [{rendererClass=&NiagaraSpriteRendererProperties}]
 }
 SplineHair {
  emitterUpdateScript [
   &EmitterState(
    Life Cycle Mode=::NewEnumerator1
    Loop Behavior=::NewEnumerator2
    Loop Count=10
    Loop Duration=&RandomRangeFloat(Minimum=0.4, Maximum=0.6)
    Recalculate Duration Each Loop=true
    UseLoopDelay=false
   )
   &SpawnBurst_Instantaneous(
    Spawn Count=&MakeIntFromFloat(Float=&FN_FrontEndFloatModifier(Gameplay=30, Front End=10))
    Spawn Group=1
    Use Spawn Probability=false
    UseLoopCountLimit=false
   )
   &SpawnRate(
    Spawn Probability=0.8
    SpawnRate=&Multiply_Float(
     A=50
     B=&FloatFromCurve(
      FloatCurve=DataInterface() {propertyValues {Curve {keys [
          {interpMode=RCIM_Cubic, tangentMode=RCTM_User, time=0, value=0}
          {interpMode=RCIM_Cubic, tangentMode=RCTM_User, time=0.1, value=1}
          {arriveTangent=-0.0405, interpMode=RCIM_Cubic, leaveTangent=-0.0405, tangentMode=RCTM_User, time=0.25, value=1}
          {arriveTangent=-4.6029, interpMode=RCIM_Cubic, leaveTangent=-4.6029, tangentMode=RCTM_User, time=0.9, value=0}
         ]}}}
      CurveIndex=$System.NormalizedLoopAge
     )
    )
    Use Spawn Probability=true
   ) {enabled=false}
  ]
  particleSpawnScript [
   &InitializeParticle(
    Color=&SelectLinearColorFromArray(
     Array Sampling Mode=::NewEnumerator1
     Direct Array Index=&MakeIntFromFloat(Float=$User.VariantID)
     Color Selection Array=DataInterface() {propertyValues {ColorData [<rgba 0.098 0.502 0.4902 1>, <rgba 0.098 0.502 0.4902 1>, <rgba 1 0.323 0.7996 1>, <rgba 0 1 0.2588 1>]}}
     Random Seed=?
    )
    Lifetime=&Multiply_Float(
     A=&MakeCustomFloatFromBool(Bool=&SetBoolByFloatComparison(A=$System.NormalizedLoopAge, B=0.8), True Float=0, False Float=1)
     B=&Multiply_Float(
      A=&MakeCustomFloatFromBool(Bool=&RandomBool(), True Float=1, False Float=0)
      B=&RandomRangeFloat(Minimum=1, Maximum=0.7)
     )
    )
    Mesh Scale Max=<1 1 1>
    Mesh Scale Min=<1 1 0.5>
    Mesh Scale Mode=::NewEnumerator2
    Sprite Size Mode=::NewEnumerator3
    UsePositionOffset=false
    Write Material Random=true
   )
   &InitialMeshOrientation(
    Mesh Orientation Mode=::NewEnumerator7
    Rotation=&MakeVector(
     X=&RandomRangeFloat(Minimum=-0.05, Maximum=0.05)
     Y=&RandomRangeFloat(Minimum=-0.05, Maximum=0.05)
     Z=&RandomRangeFloat(Minimum=0, Maximum=1)
    )
    Rotation Coordinate Space=::NewEnumerator1
    Use Rotation Vector=true
   )
   &FN_SetSpawnGroup(Number of Spawn Groups:NiagaraInt32=&SelectIntFromArray(
     Direct Array Index=&RandomRangeInt(Minimum=0, Maximum=7)
     Int Selection Array=DataInterface() {propertyValues {IntData [0,1,2,3,4,6,8,9]}}
    )) {enabled=false}
   &FN_SetSpawnGroup(Number of Spawn Groups=8) {enabled=false, module=FN_SetSpawnGroup001}
   /SetVariables(Particles.BGColor:LinearColor=&SelectLinearColorFromArray(
     Array Sampling Mode=::NewEnumerator1
     Direct Array Index=&MakeIntFromFloat(Float=$User.VariantID)
     Color Selection Array=DataInterface() {propertyValues {ColorData [<rgba 0.0039 0.251 0.302 1>, <rgba 0.0039 0.251 0.302 1>, <rgba 0.8021 0.1007 0.5945 1>, <rgba 0 0.651 0.1685 1>]}}
     Random Seed=?
    )) {isSetParameters=true}
   /SetVariables(Particles.MeshIndex=8) {enabled=false, isSetParameters=true}
   &ShapeLocation(MaskSpawnGroup=false, Override Seed=false, Sphere Radius=50, UseOwnerScale=true)
   &PartitionParticles(Number Of Partitions=8)
   &PartitionParticles(Number Of Partitions=21) {module=PartitionParticles001}
   &PartitionParticles(
    Number Of Partitions=&MakeIntFromFloat(Float=&FN_FrontEndFloatModifier(Gameplay=3, Front End=1))
    Particles Per Partition=3
    Partition Mode:ENiagara_PartitionMode=::NewEnumerator1
   ) {module=PartitionParticles002}
   &CameraOffset()
   &PartitionParticles() {module=PartitionParticles003}
  ]
  particleUpdateScript [
   &ParticleState()
   &SkeletalMeshLocation(
    Bone / Socket Index=$Particles.PartitionParticles001.Partition
    MaskSpawnGroup=false
    Mesh Sampling Type=::NewEnumerator5
    Position Sampling=::NewEnumerator2
    Sampled Position Offset=&RandomRangeVector(Minimum=<0 0 0>, Maximum=<0 0 0>)
    Sampled Position Offset Coordinate Space:ENiagaraSkelMeshTransforms=::NewEnumerator2
    Skeletal Mesh=$User.Skeletal Mesh
    Socket Sampling Mode=::NewEnumerator5
   ) {enabled=false}
   &SkeletalMeshLocation(
    Bone / Socket Index=&MakeIntFromFloat(Float=/Niagara/DynamicInputs/Math/LerpMultipleFloats(
      Float Count:ENiagara_MultipleLerpCount=::NewEnumerator0
      1:NiagaraFloat=&RandomRangeFloat(Minimum=1, Maximum=7)
      2:NiagaraFloat=&RandomRangeFloat(Minimum=10, Maximum=12)
      3:NiagaraFloat=&RandomRangeFloat(Minimum=15, Maximum=25)
      4:NiagaraFloat=?
      5:NiagaraFloat=?
      Alpha:NiagaraFloat=&SelectFloatFromArray(Direct Array Index=$Particles.PartitionParticles002.Partition, Float Selection Array=DataInterface() {propertyValues {FloatData [0,0.5,1]}})
     ))
    MaskSpawnGroup=false
    Mesh Sampling Type=::NewEnumerator5
    Position Sampling=::NewEnumerator2
    Sampled Position Offset=/Niagara/DynamicInputs/Math/LerpMultipleVectors(
     Alpha:NiagaraFloat=&SelectFloatFromArray(Direct Array Index=$Particles.PartitionParticles002.Partition, Float Selection Array=DataInterface() {propertyValues {FloatData [0,0.5,1]}})
     V 1:Vector3f=&RandomRangeVector(Minimum=<-3 -1 -3>, Maximum=<2 1 1>)
     V 2:Vector3f=&RandomRangeVector(Minimum=<-5 10 -20>, Maximum=<-5 0 -30>)
     V 3:Vector3f=&RandomRangeVector(Minimum=<0 10 2>, Maximum=<-5 -10 -5>)
     V 4:Vector3f=?
     V 5:Vector3f=?
     Vector Count:ENiagara_MultipleLerpCount=::NewEnumerator0
    )
    Sampled Position Offset Coordinate Space:ENiagaraSkelMeshTransforms=::NewEnumerator2
    Skeletal Mesh=$User.Skeletal Mesh
    Socket Sampling Mode=::NewEnumerator5
   ) {module=SkeletalMeshLocation001}
   &ScaleMeshSize(Scale Factor=&MakeVector(
     X=&FloatFromCurve(
      FloatCurve=DataInterface() {propertyValues {Curve {keys [{tangentMode=RCTM_User, time=0, value=0}, {tangentMode=RCTM_User, time=1, value=1}]}}}
      Scale Curve=0.4
     )
     Y=1
     Z=&FloatFromCurve(
      FloatCurve=DataInterface() {propertyValues {Curve {keys [
          {arriveTangent=4.1785, interpMode=RCIM_Cubic, leaveTangent=4.1785, tangentMode=RCTM_User, time=0, value=0}
          {arriveTangent=1.1179, interpMode=RCIM_Cubic, leaveTangent=1.1179, tangentMode=RCTM_User, time=0.2042, value=0.6228}
          {arriveTangent=0.0224, interpMode=RCIM_Cubic, leaveTangent=0.0224, tangentMode=RCTM_User, time=1, value=1}
         ]}}}
      Scale Curve=&Multiply_Float(A=1, B=&RandomRangeFloat(Minimum=0.15, Maximum=0.25))
     )
    ))
   &ScaleMeshSize(Scale Factor=&VectorFromFloat(Value=&FloatFromCurve(
      FloatCurve=DataInterface() {propertyValues {Curve {keys [
          {tangentMode=RCTM_User, time=0, value=1}
          {arriveTangent=-0.0405, interpMode=RCIM_Cubic, leaveTangent=-0.0405, tangentMode=RCTM_User, time=0.7009, value=1}
          {arriveTangent=-8.0089, interpMode=RCIM_Cubic, leaveTangent=-8.0089, tangentMode=RCTM_User, time=1, value=0}
         ]}}}
      CurveIndex=$System.Progress
     ))) {enabled=false, module=ScaleMeshSize001}
   &DynamicMaterialParameters(
    Index 0 Param 1=&FloatFromCurve(
     FloatCurve=DataInterface() {propertyValues {Curve {keys [{tangentMode=RCTM_User, time=0, value=0}, {tangentMode=RCTM_User, time=1, value=1}]}}}
     Scale Curve=&Multiply_Float(A=1, B=1)
    )
    Index 0 Param 2=2
    Index 0 Param 3=0.666
    Module.Param0WriteEnabled=true
    Module.Param1WriteEnabled=true
    Module.Param2WriteEnabled=true
    Module.Param3WriteEnabled=true
   )
   /SetVariables(Particles.MeshOrientation=&LerpQuaternion(
     Quaternion A:Quat4f=$Particles.Initial.MeshOrientation
     Quaternion B:Quat4f=$Particles.SkeletalMeshLocation001.SampledRotation
     Lerp Factor:NiagaraFloat=-0.05
     Enable Spherical Linear Interpolation:NiagaraBool=false
    )) {isSetParameters=true}
   /SetVariables(Particles.Position=&AddVectorToPosition(
     Position=$Particles.SkeletalMeshLocation001.SampledPosition
     Vector=&Multiply_VectorByFloat(
      Vector=&ReturnMeshOrientationAxis(Mesh Orientation Axis=::ZAxis, Mesh Orientation Quat=$Particles.MeshOrientation)
      Float=&FloatFromCurve(
       FloatCurve=DataInterface() {propertyValues {Curve {keys [{tangentMode=RCTM_User, time=0, value=-0.2}, {tangentMode=RCTM_User, time=1, value=1}]}}}
       Scale Curve=10
      )
     )
    )) {isSetParameters=true}
   &Color(Color=&ColorFromCurve(ColorCurve=DataInterface() {propertyValues {
       AlphaCurve {keys []}
       BlueCurve {keys [<at 0 3 RCIM_Cubic>,<at 0.7945 0.2 RCIM_Cubic>]}
       GreenCurve {keys [<at 0 2 RCIM_Cubic>,<at 0.7945 0 RCIM_Cubic>]}
       RedCurve {keys [<at 0 0 RCIM_Cubic>,<at 0.7945 0 RCIM_Cubic>]}
      }})) {enabled=false}
   /SetVariables(
    Particles.BGColor:LinearColor=&MultiplyLinearColor(
     Linear Color=$Particles.Initial.BGColor
     Scale Factor=&MakeVector4_FromFloat(Float=&FloatFromCurve(FloatCurve=DataInterface() {propertyValues {Curve {keys [
           {tangentMode=RCTM_User, time=0, value=1}
           {arriveTangent=-0.0405, interpMode=RCIM_Cubic, leaveTangent=-0.0405, tangentMode=RCTM_User, time=0.7009, value=1}
           {arriveTangent=-8.0089, interpMode=RCIM_Cubic, leaveTangent=-8.0089, tangentMode=RCTM_User, time=1, value=0}
          ]}}}))
    )
    Particles.Color=&MultiplyLinearColor(
     Linear Color=$Particles.Initial.Color
     Scale Factor=&MakeVector4_FromFloat(Float=&FloatFromCurve(FloatCurve=DataInterface() {propertyValues {Curve {keys [
           {tangentMode=RCTM_User, time=0, value=1}
           {arriveTangent=-0.0405, interpMode=RCIM_Cubic, leaveTangent=-0.0405, tangentMode=RCTM_User, time=0.7009, value=1}
           {arriveTangent=-8.0089, interpMode=RCIM_Cubic, leaveTangent=-8.0089, tangentMode=RCTM_User, time=1, value=0}
          ]}}}))
    )
   ) {isSetParameters=true}
   /SetVariables(Particles.VisibilityTag=&MakeCustomIntFromBool(Bool=$User.bisHeadHidden, True INT=1, False INT=0)) {isSetParameters=true}
  ]
  rendererClasses [&NiagaraSpriteRendererProperties, &NiagaraMeshRendererProperties]
  renderers [
   {rendererClass=&NiagaraSpriteRendererProperties}
   {rendererClass=&NiagaraMeshRendererProperties, rendererIndex=1}
   {rendererClass=&NiagaraMeshRendererProperties, rendererIndex=2}
   {rendererClass=&NiagaraMeshRendererProperties, rendererIndex=3}
  ]
 }
}
summary {
 systemName=NS_Apprentice_Head_LongHair
 emitters [
  {emitterName=SplineHair, rendererClasses [&NiagaraSpriteRendererProperties, &NiagaraMeshRendererProperties]}
  {emitterName=Impact_Flare_and_Corona, rendererClasses [&NiagaraSpriteRendererProperties]}
  {emitterName=Spark001, rendererClasses [&NiagaraSpriteRendererProperties]}
 ]
 userVariables [
  User.bIsFrontEnd {type=NiagaraBool}
  User.bIsFrontEndPreview {type=NiagaraBool}
  User.bisHeadHidden {type=NiagaraBool}
  User.Dissolve Edge Intensity {type=NiagaraFloat, defaultValue=1, description=""}
  User.Dissolve Progress {type=NiagaraFloat, defaultValue=1, description=""}
  User.FadeInDuration {type=NiagaraFloat, defaultValue=1, description=""}
  User.Global Alpha {type=NiagaraFloat, defaultValue=1, description=""}
  User.Interpolate Intensity {type=NiagaraFloat, defaultValue=0.3, description=""}
  User.Skeletal Mesh {
   type=NiagaraDataInterfaceSkeletalMesh
   defaultValue {
    struct=NiagaraExt_VariableValue_DataInterface
    value {dataInterface=&self:NiagaraDataInterfaceSkeletalMesh_0}
   }
  }
  User.VariantID {type=NiagaraFloat, defaultValue=0, description=""}
 ]
}
