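# Material function for the Apprentice character's eyes, behind a static switch, that animates eyelids in response to timed triggers and damage.
# Per-eye hashed offsets vary kill and damage curves, which combine to shape procedural UV1 lid gradients.
# The lid mask blends incoming material attributes with LUT-tinted closed-eye shading and gradient-based roughness.
# It also builds a vertex-color-derived world position push and a noise-dissolve opacity mask for timed fade in and out.
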MaterialFunction /BRCosmetics/Characters/Player/Male/Medium/Bodies/M_MED_Apprentice/Materials/MF_Apprentice_EyeReactivity
graph {
 properties {previewBlendMode=BLEND_Masked}
 nodes [
  FunctionOutput(None=StaticSwitchParameter(True=NamedRerouteUsage(), False=NamedRerouteUsage()) {group=1-Apprentice_EyeReactivity, parameterName=UseApprentice_EyeReactivity, sortPriority=0})
  NamedRerouteDeclaration(None=FunctionInput() {inputName=MA, inputType=FunctionInput_MaterialAttributes}) {name=MA, nodeColor=<rgba 0 0.6484 1 1>}
  NamedRerouteDeclaration(None=ComponentMask(None=FunctionInput() {inputName=FX, inputType=FunctionInput_Vector4, sortPriority=1}) {r=true}) {name=FX_R, nodeColor=<rgba 1 0.0434 0 1>}
  n1 = ScalarParameter() {defaultValue=0.3, group=1-Apprentice_EyeReactivity, parameterName=OverlapRoughness_Upper_EyeLid, sliderMax=1, sortPriority=21}
  n2 = ScalarParameter() {defaultValue=0.7, group=1-Apprentice_EyeReactivity, parameterName=OverlapRoughness_Lower_EyeLid, sliderMax=1, sortPriority=22}
  n3 = LinearInterpolate(A=@n1, B=@n2, Alpha=NamedRerouteUsage()) {constB=0}
  n4 = ScalarParameter() {defaultValue=0.125, group=1-Apprentice_EyeReactivity, parameterName=OverlapRoughness_EyeLid_Background, sliderMax=1, sortPriority=19}
  n5 = ScalarParameter() {
   defaultValue=-0.1
   group=1-Apprentice_EyeReactivity
   parameterName=OverlapRoughness_EyeLidGradient_Bias
   sliderMax=1
   sliderMin=-1
   sortPriority=23
  }
  n6 = OneMinus(None=Saturate(None=NamedRerouteUsage()))
  n7 = Add(A=NamedRerouteUsage(), B=Add(A=@n5, B=Multiply(A=@n6) {constB=0.5})) {constB=0}
  n8 = ScalarParameter() {defaultValue=1.5, group=1-Apprentice_EyeReactivity, parameterName=OverlapRoughness_EyeLidGradient_Scale, sliderMax=16, sortPriority=24}
  n9 = Multiply(A=Power(Base=NamedRerouteUsage()) {constExponent=6}) {constB=4}
  n10 = Saturate(None=Add(A=Multiply(A=@n7, B=@n8) {constB=8}, B=Saturate(None=@n9)))
  NamedRerouteDeclaration(None=LinearInterpolate(
    A=LinearInterpolate(A=@n3, B=@n4, Alpha=@n10) {constB=0}
    B=ScalarParameter() {group=1-Apprentice_EyeReactivity, parameterName=OverlapRoughness_Default, sliderMax=1, sortPriority=20}
    Alpha=NamedRerouteUsage()
   ) {constB=0}) {name=OverlapRoughness, nodeColor=<rgba 0.0955 0.2229 1 1>}
  n11 = Subtract(A=Time(), B=ScalarParameter() {defaultValue=-99999, group=1-Apprentice_EyeReactivity_DontTouch, parameterName=Reactive_CT, sliderMin=-99999})
  n12 = Add(A=ScalarParameter() {defaultValue=6, group=1-Apprentice_EyeReactivity_DontTouch, parameterName=Reactive_Anim_Duration, sliderMin=0}) {constB=0}
  NamedRerouteDeclaration(None=Saturate(None=Divide(A=@n11, B=@n12))) {name=TriggeredBlend, nodeColor=<rgba 0.2422 0 1 1>}
  n13 = TextureCoordinate() {coordinateIndex=1}
  n14 = Multiply(A=Add(A=Frac(None=@n13)) {constB=-0.5}) {constB=2}
  NamedRerouteDeclaration(None=Abs(None=ComponentMask(None=@n14) {r=true})) {name=CaruncleGradient, nodeColor=<rgba 0.7109 0 1 1>}
  n15 = Subtract(
   A=Time()
   B=ScalarParameter() {defaultValue=-99999, group=1-Apprentice_EyeReactivity_DontTouch, parameterName=EyeFadeIn_CT, sliderMin=-99999}
  )
  n16 = Divide(A=Subtract(
    A=Time()
    B=ScalarParameter() {defaultValue=99999, group=1-Apprentice_EyeReactivity_DontTouch, parameterName=EyeFadeOut_CT, sliderMin=-99999}
   )) {constB=0.35}
  n17 = Multiply(A=Saturate(None=Divide(A=@n15) {constB=0.35}), B=OneMinus(None=Saturate(None=@n16)))
  n18 = TextureSample(UVs=Multiply(A=TextureCoordinate() {coordinateIndex=1}) {constB=0.035}) {samplerType=SAMPLERTYPE_LinearGrayscale, texture=/Game/TimeOfDay/CloudsAndStorm/Textures/LDR_LLL1_0}
  NamedRerouteDeclaration(None=Ceil(None=Subtract(A=@n17, B=@n18.R))) {name=OpacityMask, nodeColor=<rgba 0.1953 1 0 1>}
  n19 = VectorParameter() {group=1-Apprentice_EyeReactivity, parameterName=ClosedEye_ShadedColor, sortPriority=2, defaultValue=<rgba 0.0278 0.0593 0.0938 1>}
  n20 = VectorParameter() {group=1-Apprentice_EyeReactivity, parameterName=ClosedEye_LitColor, sortPriority=2, defaultValue=<rgba 0.0498 0.0932 0.1406 1>}
  n21 = LinearInterpolate(A=@n19.RGB, B=@n20.RGB, Alpha=NamedRerouteUsage())
  n22 = GetMaterialAttributes(None=NamedRerouteUsage()) {attributeGetTypes [#1, #2]}
  n23 = NamedRerouteUsage()
  n24 = SetMaterialAttributes!(
   MaterialAttributes=NamedRerouteUsage()
   BaseColor=Constant()
   EmissiveColor=LinearInterpolate(
    A=MaterialFunctionCall!(In=@n21, LUT, LUT Multiplier) {materialFunction=/Game/Packages/Fortress_SharedMaterials/Anime_Master_Material/MF_ApplyLUT}.Result
    B=@n22.EmissiveColor
    Alpha=@n23
   )
   Normal=LinearInterpolate(A=Constant3Vector() {constant=<rgba 0 0 1 0>}, B=@n22.Normal, Alpha=@n23)
   Roughness=NamedRerouteUsage()
   WorldPositionOffset
   OpacityMask=NamedRerouteUsage()
  ) {attributeSetTypes [
    #3
    #1
    #2
    #4
    #5
    #6
   ]}
  NamedRerouteDeclaration(None=BlendMaterialAttributes(A=NamedRerouteUsage(), B=@n24, Alpha=NamedRerouteUsage())) {name=Out, nodeColor=<rgba 0 1 0.2734 1>}
  n25 = Add(
   A=Distance(A=@n14, B=Constant())
   B=Multiply(A=ScalarParameter() {defaultValue=0.98, group=1-Apprentice_EyeReactivity, parameterName=SphereDiameter, sliderMax=2, sortPriority=15}) {constB=-1}
  ) {constB=-1}
  n26 = ComponentMask(None=@n14) {g=true}
  n27 = NamedRerouteUsage()
  n28 = NamedRerouteUsage()
  n29 = NamedRerouteDeclaration(None=OneMinus(None=Max(A=@n27, B=@n28))) {name=AnimGradient_Combined, nodeColor=<rgba 1 0 0.1464 1>}
  n30 = Multiply(A=@n29, B=@n29)
  n31 = Ceil(None=@n26)
  n32 = Multiply(A=LinearInterpolate(A=@n25, B=Multiply(A=@n26) {constB=-1}, Alpha=@n30), B=OneMinus(None=@n31))
  n33 = Multiply(A=LinearInterpolate(A=@n25, B=@n26, Alpha=@n30), B=NamedRerouteDeclaration(None=@n31) {name=LidMaskLower, nodeColor=<rgba 1 0.9844 0 1>})
  n34 = Ceil(None=NamedRerouteDeclaration(None=Max(A=@n32, B=@n33)) {name=LidGradients, nodeColor=<rgba 1 0.6328 0 1>})
  NamedRerouteDeclaration(None=Saturate(None=OneMinus(None=@n34))) {name=LidOpacity, nodeColor=<rgba 0 1 0.6484 1>}
  n35 = VertexColor()
  n36 = Normalize(VectorInput=MaterialFunctionCall(X=@n35.R, Y=@n35.R, Z=@n35.R) {materialFunction=&MakeFloat3}.Result)
  n37 = Transform(None=Add(A=Multiply(A=@n36) {constB=2}) {constB=-1}) {transformSourceType=TRANSFORMSOURCE_Local}
  n38 = Multiply(A=ScalarParameter() {defaultValue=5, group=1-Apprentice_EyeReactivity, parameterName=PushStrenght, sliderMax=5}) {constB=-1}
  NamedRerouteDeclaration(None=Multiply(A=NamedRerouteUsage(), B=Multiply(A=@n37, B=@n38))) {name=WPO, nodeColor=<rgba 0.3359 1 0 1>}
  n39 = NamedRerouteUsage()
  n40 = TextureCoordinate() {coordinateIndex=1}
  n41 = Divide(A=Ceil(None=ComponentMask(None=@n40) {r=true})) {constB=16}
  n42 = Saturate(None=Multiply(A=MaterialFunctionCall(Float=@n41) {materialFunction=&MF_Hash}.Float) {constB=0.1})
  n43 = Frac(None=Add(A=Add(A=@n39, B=@n42)) {constB=0.25})
  n44 = Add(A=Multiply(A=Sine(None=@n43)) {constB=0.5}) {constB=0.5}
  n45 = OneMinus(None=Multiply(A=@n44, B=@n44) {desc=bot curve pow})
  NamedRerouteDeclaration(None=Multiply(A=Multiply(A=@n45, B=@n45), B=@n45) {desc=top curve pow}) {name=AnimGradient_Kill, nodeColor=<rgba 1 0 0.3047 1>}
  n46 = Ceil(None=DotProduct(A=SkyAtmosphereLightDirection(), B=VertexNormalWS()))
  NamedRerouteDeclaration(None=Saturate(None=FunctionInput(Preview=@n46) {bUsePreviewValueAsDefault=true, inputName=ShadingCurve, inputType=FunctionInput_Scalar, sortPriority=2})) {name=ShadingCurve, nodeColor=<rgba 1 0.9549 0 1>}
  n47 = Saturate(None=ScalarParameter() {desc=CurrentDamage, group=1-Apprentice_EyeReactivity_DontTouch, parameterName=Reactive_CurrentDamage, sliderMax=1})
  n48 = Add(A=Add(A=@n42, B=NamedRerouteDeclaration(None=@n47) {name=DamageCurve, nodeColor=<rgba 0.8984 0 1 1>})) {constB=-0.1}
  n49 = OneMinus(None=Subtract(
    A=Time()
    B=ScalarParameter() {defaultValue=99999, group=1-Apprentice_EyeReactivity_DontTouch, parameterName=Reactive_Damage_CT, sliderMin=-99999}
   ))
  NamedRerouteDeclaration(None=Multiply(A=Saturate(None=@n48), B=Saturate(None=@n49))) {name=AnimGradient_Damage, nodeColor=<rgba 1 0 0.3047 1>}
  NamedRerouteUsage()
  n50 = Frac(None=TextureCoordinate() {coordinateIndex=1})
  n51 = OneMinus(None=@n50)
  n52 = MaterialFunctionCall(Float2=Multiply(A=Multiply(A=@n50, B=@n51)) {constB=2}) {materialFunction=&BreakOutFloat2Components}
  n53 = OneMinus(None=Saturate(None=@n52.R))
  Subtract(A=@n52.G, B=Add(A=Multiply(A=@n53) {constB=0.05}) {constB=-0.05})
 ]
}
