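# Material function for the Apprentice body materials that adds a colored Fresnel rim glow to the emissive channel.
# A saturated Fresnel with tunable exponent and base reflect fraction, scaled by a blend parameter, forms a mask that multiplies the squared RGB color.
# The result adds onto the incoming material attributes' emissive, with a static switch parameter toggling the effect.
# A Vector4 FX input is declared but unconnected.
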MaterialFunction /BRCosmetics/Characters/Player/Male/Medium/Bodies/M_MED_Apprentice/Materials/MF_Apprentice_FresnelFX
graph {
 properties {previewBlendMode=BLEND_Masked}
 nodes [
  FunctionInput() {bUsePreviewValueAsDefault=true, inputName=FX, inputType=FunctionInput_Vector4, sortPriority=1}
  FunctionOutput(None=StaticSwitchParameter(True=NamedRerouteUsage(), False=NamedRerouteUsage()) {group=1-Apprentice_FresmelFX, parameterName=UseApprentice_FresmelFX, sortPriority=0})
  NamedRerouteDeclaration(None=FunctionInput() {inputName=MA, inputType=FunctionInput_MaterialAttributes}) {name=MA, nodeColor=<rgba 0 0.6484 1 1>}
  n1 = VectorParameter() {group=1-Apprentice_FresmelFX, parameterName=Fresnel_Color, sortPriority=3, defaultValue=<rgba 0 1 1 3>}
  n2 = ScalarParameter() {defaultValue=3, group=1-Apprentice_FresmelFX, parameterName=Fresnel_Exponent, sortPriority=20}
  n3 = ScalarParameter() {defaultValue=0.01, group=1-Apprentice_FresmelFX, parameterName=Fresnel_BaseReflectFrac, sliderMax=1, sortPriority=21}
  n4 = Multiply(
   A=Saturate(None=Fresnel(ExponentIn=@n2, BaseReflectFractionIn=@n3) {baseReflectFraction=0, exponent=1})
   B=ScalarParameter() {defaultValue=1, group=1-Apprentice_FresmelFX, parameterName=Fresnel_Blend, sliderMax=1, sortPriority=15}
  )
  NamedRerouteDeclaration(None=Multiply(A=Multiply(A=@n1.RGB, B=@n1.RGB), B=NamedRerouteDeclaration(None=@n4) {name=FresnelMask, nodeColor=<rgba 0.5469 1 0 1>})) {name=FresnelColor, nodeColor=<rgba 0.9922 0 1 1>}
  n5 = GetMaterialAttributes(None=NamedRerouteUsage()) {attributeGetTypes [#1]}
  n6 = NamedRerouteUsage()
  NamedRerouteDeclaration(None=SetMaterialAttributes!(MaterialAttributes=@n5.MaterialAttributes, EmissiveColor=Add(A=@n5.EmissiveColor, B=@n6)) {attributeSetTypes [#1]}) {name=Out, nodeColor=<rgba 0 1 0.2734 1>}
 ]
}
