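# Material function for the Apprentice character that adds a switchable, stylized three-color highlight to emissive, masked by a selectable FX channel.
# A static switch picks Phong (reflection against directional light from a Custom node) or "Blinn" (camera vectors against normal, without light direction), with optional Z flattening.
# The powered result drives SmoothCeil ramps that band the colors, which pass through MF_ApplyLUT before blending into the material attributes.
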MaterialFunction /BRCosmetics/Characters/Player/Male/Medium/Bodies/M_MED_Apprentice/Materials/MF_Apprentice_VibrantHighlight
graph {
 properties {previewBlendMode=BLEND_Masked}
 nodes [
  CollectionParameter() {
   collection=&FortniteMaterialParameters
   parameterName=SunlightDirectionalVector
  }
  FunctionOutput(None=StaticSwitchParameter(True=NamedRerouteUsage(), False=NamedRerouteUsage()) {group=1-Apprentice_VibrantHighlight, parameterName=UseApprentice_VibrantHighlight, sortPriority=0})
  NamedRerouteDeclaration(None=FunctionInput() {inputName=MA, inputType=FunctionInput_MaterialAttributes}) {name=MA, nodeColor=<rgba 0 0.6484 1 1>}
  n1 = ChannelMaskParameter(None=FunctionInput() {bUsePreviewValueAsDefault=true, inputName=FX, inputType=FunctionInput_Vector4, sortPriority=1}) {
   group=1-Apprentice_VibrantHighlight
   maskChannel=Blue
   parameterName=FX_VibrantHighlight_Channel
   sortPriority=4
   defaultValue=<rgba 0 0 1 0>
  }
  NamedRerouteDeclaration(None=@n1.R) {name=FX_Mask, nodeColor=<rgba 1 0 0.6562 1>}
  n2 = VectorParameter() {group=1-Apprentice_VibrantHighlight, parameterName=ColorC, sortPriority=3, defaultValue=<rgba 0.1395 0 0.8698 1>}
  n3 = VectorParameter() {group=1-Apprentice_VibrantHighlight, parameterName=ColorB, sortPriority=3, defaultValue=<rgba 1 0.12 1 1>}
  n4 = ScalarParameter() {defaultValue=1, group=1-Apprentice_VibrantHighlight, parameterName=VH_IgnoreZ, sliderMax=1}
  n5 = AppendVector(A=Constant2Vector() {g=1, r=1}, B=Saturate(None=OneMinus(None=@n4)))
  n6 = Multiply(A=Custom() {code="return View.DirectionalLightDirection;", description=DirectionalLightDirection, inputs []}, B=@n5)
  n7 = MaterialFunctionCall() {materialFunction=&CameraDirectionVector}
  n8 = Multiply(A=Add(A=CameraVectorWS(), B=Multiply(A=@n7.Result) {constB=-1}), B=@n5)
  n9 = Multiply(A=@n5, B=PixelNormalWS())
  n10 = StaticSwitchParameter(
   True=DotProduct(A=ReflectionVectorWS(), B=Normalize(VectorInput=@n6)) {desc=Phong}
   False=DotProduct(A=Normalize(VectorInput=@n8), B=Normalize(VectorInput=@n9)) {desc=Blinn}
  ) {desc=Switches between Blinn and Phong Highlight., group=1-Apprentice_VibrantHighlight, parameterName=UseVH_PhongHighlight, sortPriority=0}
  n11 = Saturate(None=Power(
    Base=@n10
    Exp=ScalarParameter() {defaultValue=8, group=1-Apprentice_VibrantHighlight, parameterName=VH_Fresnel_Exponent, sliderMax=8, sortPriority=20}
   ) {constExponent=3})
  n12 = ScalarParameter() {defaultValue=0.45, group=1-Apprentice_VibrantHighlight, parameterName=ColorB_Adjust, sliderMax=1}
  n13 = ScalarParameter() {defaultValue=0.2, group=1-Apprentice_VibrantHighlight, parameterName=RampWidth, sliderMax=1}
  n14 = LinearInterpolate(
   A=@n2.RGB
   B=@n3.RGB
   Alpha=MaterialFunctionCall!(In=Subtract(A=@n11, B=@n12), Ramp Width=@n13) {materialFunction=/Engine/Functions/Engine_MaterialFunctions01/Math/SmoothCeil}.Result
  )
  n15 = VectorParameter() {group=1-Apprentice_VibrantHighlight, parameterName=ColorA, sortPriority=3, defaultValue=<rgba 1 1 0.073 1>}
  n16 = MaterialFunctionCall(
   In=Subtract(A=@n11, B=ScalarParameter() {defaultValue=0.98, group=1-Apprentice_VibrantHighlight, parameterName=ColorA_Adjust, sliderMax=1})
   Ramp Width=@n13
  ) {materialFunction=&SmoothCeil}
  NamedRerouteDeclaration(None=MaterialFunctionCall!(In=LinearInterpolate(A=@n14, B=@n15.RGB, Alpha=@n16.Result), LUT, LUT Multiplier) {materialFunction=/Game/Packages/Fortress_SharedMaterials/Anime_Master_Material/MF_ApplyLUT}.Result) {name=Highlight, nodeColor=<rgba 0.9922 0 1 1>}
  n17 = NamedRerouteUsage()
  n18 = GetMaterialAttributes(None=@n17) {attributeGetTypes [#1]}
  n19 = LinearInterpolate(A=@n18.EmissiveColor, B=NamedRerouteUsage(), Alpha=NamedRerouteUsage())
  NamedRerouteDeclaration(None=BlendMaterialAttributes(
    A=@n17
    B=SetMaterialAttributes!(MaterialAttributes=@n18.MaterialAttributes, EmissiveColor=@n19) {attributeSetTypes [#1]}
    Alpha=NamedRerouteUsage()
   )) {name=Out, nodeColor=<rgba 0 1 0.2734 1>}
  n20 = ScalarParameter() {defaultValue=0.2, group=1-Apprentice_VibrantHighlight, parameterName=ColorC_Adjust, sliderMax=1}
  NamedRerouteDeclaration(None=MaterialFunctionCall(In=Subtract(A=@n11, B=@n20), Ramp Width=@n13) {materialFunction=&SmoothCeil}.Result) {name=Fresnel_Mask, nodeColor=<rgba 0.8047 1 0 1>}
  SkyAtmosphereLightDirection()
 ]
}
