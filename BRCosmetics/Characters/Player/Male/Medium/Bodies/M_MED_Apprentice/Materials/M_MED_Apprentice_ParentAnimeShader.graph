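# Parent anime-style material for the Apprentice male medium player body, using Cloth shading and material attributes on skeletal meshes with morph target and mesh deformer support.
# The shared character PreFX function feeds MF_AnimeShading, whose result and auxiliary outputs drive Apprentice-specific animated triangle and eye reactivity layers.
# Vibrant highlight and Fresnel FX layers follow in the chain before the shared character PostFX function, which also takes PreFX outputs.
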Material /BRCosmetics/Characters/Player/Male/Medium/Bodies/M_MED_Apprentice/Materials/M_MED_Apprentice_ParentAnimeShader
graph {
 properties {
  bAutomaticallySetUsageInEditor=false
  bUseMaterialAttributes=true
  bUsedWithMeshDeformer=true
  bUsedWithMorphTargets=true
  bUsedWithSkeletalMesh=true
  pixelDepthOffsetMode=PDOM_Legacy
  shadingModel=MSM_Cloth
  usedShadingModels=Cloth
 }
 nodes [
  n1 = MaterialFunctionCall() {materialFunction=&BaseFN_CharacterFunction_PreFX}
  n2 = MaterialFunctionCall!(
   Material=@n1.Result
   LitDiffuseOverride
   SSC_AlphaOverride
   Tint Mask
   DF InkLine Visibility Mask
   DF Inkline Injection
   DF Inkline Color Override
   DF Inkline Color Override Mask
   CustomUVs
  ) {materialFunction=/Game/Packages/Fortress_SharedMaterials/Anime_Master_Material/MF_AnimeShading}
  n3 = MaterialFunctionCall!(
   MA=MaterialFunctionCall!(MA=@n2.Result, Shading Curve=@n2.Result, InkLine Mask=@n2.Result, FX=@n2.Result) {materialFunction=/Game/Characters/Player/Male/Medium/Bodies/M_MED_Apprentice/Materials/MF_Apprentice_AnimatedTriangles}.Result
   FX=@n2.FX
   ShadingCurve=@n2.Shading
  ) {materialFunction=&/MF_Apprentice_EyeReactivity}
  MaterialFunctionCall(
   AO=@n1.AO
   Distance Blend=@n1.AO
   Fresnel=@n1.AO
   Gmap=@n1.AO
   Pre Skinned Local Position=@n1.AO
   Material=MaterialFunctionCall!(
    MA=MaterialFunctionCall!(MA=@n3.Result, FX=@n2.FX) {materialFunction=&/MF_Apprentice_VibrantHighlight}.Result
    FX
   ) {materialFunction=&/MF_Apprentice_FresnelFX}.Result
  ) {materialFunction=&BaseFN_CharacterFunction_PostFX}
 ]
}
