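# Physics asset for the Apprentice (male, medium) player body's dynamic jacket, simulating its hanging panels against the body.
# Six capsule chains (left/right front, side and back) hang from MSTR anchors on spine_02, with limited swing and tight twist.
# Looser cross-links join fronts to sides, both backs, and right side to back.
# Capsules on lower spine, pelvis, legs and feet provide collision, plus an oversized, lowered root capsule named for crouch collision.
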PhysicsAsset /BRCosmetics/Characters/Player/Male/Medium/Bodies/M_MED_Apprentice/Meshes/M_MED_Apprentice_Dyn_Jacket
bodies {
 calf_l [{shapeType=Capsule, length=29.5782, radius=12.6538, center=<-19.349 -1.0815 0.594>, rotation=<rot 88.1624 174.4553 174.7609>}]
 calf_r [{shapeType=Capsule, length=32.3282, radius=13.1538, center=<19.3248 2.4812 0.3937>, rotation=<rot 88.1624 174.4553 -5.2391>}]
 dyn_jacket_l_01 [{shapeType=Capsule, length=17.0803, radius=10.7915, shapeName=dyn_jacket_l_01_capsule, center=<7 0 6>, rotation=<rot 0 0 95>}]
 dyn_jacket_l_02 [{shapeType=Capsule, length=18, radius=17, shapeName=dyn_jacket_l_02_capsule, center=<9 0 10>, rotation=<rot 0 0 95>}]
 dyn_jacket_l_03 [{shapeType=Capsule, length=27.9553, radius=16.9165, shapeName=dyn_jacket_l_02_capsule, center=<9 0 10>, rotation=<rot 0 -15 95>}]
 dyn_jacket_l_MSTR [{shapeType=Capsule, length=1, radius=1}]
 dyn_jacket_l_bk_01 [{shapeType=Capsule, length=10.8303, radius=10.9165, shapeName=dyn_jacket_l_bk_01_capsule, center=<6 0 6>, rotation=<rot -0.0001 0 85>}]
 dyn_jacket_l_bk_02 [{shapeType=Capsule, length=22.3303, radius=12.1665, shapeName=dyn_jacket_l_bk_02_capsule, center=<8 0 8>, rotation=<rot -0.0001 0 85>}]
 dyn_jacket_l_bk_03 [{shapeType=Capsule, length=26.9553, radius=17.6665, shapeName=dyn_jacket_l_bk_02_capsule, center=<12 0 10>, rotation=<rot -0.0001 -5 90>}]
 dyn_jacket_l_bk_MSTR [{shapeType=Capsule, length=1, radius=1}]
 dyn_jacket_l_fr_01 [{shapeType=Capsule, length=6.375, radius=9.875, shapeName=dyn_jacket_l_fr_01_capsule, center=<10 -5 6>, rotation=<rot 0.0002 0.0002 85>}]
 dyn_jacket_l_fr_02 [{shapeType=Capsule, length=8, radius=15.5, shapeName=dyn_jacket_l_fr_02_capsule, center=<12 -7 12>, rotation=<rot 0.0004 -0.0003 80>}]
 dyn_jacket_l_fr_03 [{shapeType=Capsule, length=10.75, radius=19.75, shapeName=dyn_jacket_l_fr_02_capsule, center=<10 -12 15>, rotation=<rot 0 0 80>}]
 dyn_jacket_l_fr_04 [{shapeType=Capsule, length=17.375, radius=21.375, shapeName=dyn_jacket_l_fr_02_capsule, center=<15 -15 15>, rotation=<rot 0 0 80>}]
 dyn_jacket_l_fr_MSTR [{shapeType=Capsule, length=1, radius=1}]
 dyn_jacket_r_01 [{shapeType=Capsule, length=17.0803, radius=10.7915, shapeName=dyn_jacket_l_01_capsule, center=<7 0 6>, rotation=<rot -0.0001 -0.0002 85>}]
 dyn_jacket_r_02 [{shapeType=Capsule, length=18, radius=17, shapeName=dyn_jacket_l_02_capsule, center=<9 0 10>, rotation=<rot -0.0001 -0.0002 85>}]
 dyn_jacket_r_03 [{shapeType=Capsule, length=27.9553, radius=16.9165, shapeName=dyn_jacket_l_02_capsule, center=<9 0 10>, rotation=<rot 1.2924 14.9451 85.1696>}]
 dyn_jacket_r_MSTR [{shapeType=Capsule, length=1, radius=1}]
 dyn_jacket_r_bk_01 [{shapeType=Capsule, length=10.8303, radius=10.9165, shapeName=dyn_jacket_l_bk_01_capsule, center=<6 0 6>, rotation=<rot -0.0001 0 95>}]
 dyn_jacket_r_bk_02 [{shapeType=Capsule, length=22.3303, radius=12.1665, shapeName=dyn_jacket_l_bk_02_capsule, center=<8 0 8>, rotation=<rot -0.0001 0 95>}]
 dyn_jacket_r_bk_03 [{shapeType=Capsule, length=26.9553, radius=17.6665, shapeName=dyn_jacket_l_bk_02_capsule, center=<12 0 10>, rotation=<rot -0.0001 5 90>}]
 dyn_jacket_r_bk_MSTR [{shapeType=Capsule, length=1, radius=1}]
 dyn_jacket_r_fr_01 [{shapeType=Capsule, length=4, radius=9.5, shapeName=dyn_jacket_l_fr_01_capsule, center=<10 5 6>, rotation=<rot 0.0001 0 95>}]
 dyn_jacket_r_fr_02 [{shapeType=Capsule, length=8, radius=15.5, shapeName=dyn_jacket_l_fr_02_capsule, center=<12 7 12>, rotation=<rot 0.0004 -0.0003 100.0001>}]
 dyn_jacket_r_fr_03 [{
   shapeType=Capsule
   length=10.75
   radius=19.75
   shapeName=dyn_jacket_l_fr_02_capsule
   center=<10 12 15>
   rotation=<rot 0.0005 -0.0004 110.0001>
  }]
 dyn_jacket_r_fr_04 [{shapeType=Capsule, length=17.375, radius=21.375, shapeName=dyn_jacket_l_fr_02_capsule, center=<15 15 15>, rotation=<rot 0.0004 -0.0003 110>}]
 dyn_jacket_r_fr_MSTR [{shapeType=Capsule, length=1, radius=1}]
 foot_l [{shapeType=Capsule, length=11.5782, radius=14.5288, center=<-3.0823 4.4745 0>, rotation=<rot 0 0 -90>}]
 foot_r [{shapeType=Capsule, length=11.5782, radius=14.5288, center=<3.0823 -4.4745 0>, rotation=<rot 0 0 90>}]
 pelvis [{shapeType=Capsule, length=8.6, radius=13.5872, center=<-4.1085 -0.5916 0>}]
 root [{shapeType=Capsule, length=0, radius=176.125, shapeName=dyn_crouch_collision_capsule, center=<-0.0003 -5.5608 -160.3587>}]
 spine_01 [{shapeType=Capsule, length=9.9006, radius=10.4006, center=<0 0.9137 0>}]
 spine_02 [{shapeType=Capsule, length=6.5256, radius=10.4006, center=<0 0.6443 0>}]
 thigh_twist_01_l [{shapeType=Capsule, length=28.5147, radius=12.7173, center=<-1.7397 0.872 2.4783>, rotation=<rot 89.5696 156.7542 157.0616>}]
 thigh_twist_01_r [{shapeType=Capsule, length=28.5147, radius=12.7173, center=<1.7342 -0.8697 -1.682>, rotation=<rot 89.3737 15.2286 -164.4624>}]
}
constraints [
 {bone1Name=dyn_jacket_r_bk_01, bone2Name=dyn_jacket_r_bk_MSTR, swing1LimitDegrees=15, swing2LimitDegrees=15, twistLimitDegrees=2}
 {bone1Name=dyn_jacket_r_bk_02, bone2Name=dyn_jacket_r_bk_01, swing1LimitDegrees=15, swing2LimitDegrees=15, twistLimitDegrees=2}
 {bone1Name=dyn_jacket_l_bk_01, bone2Name=dyn_jacket_l_bk_MSTR, swing1LimitDegrees=15, swing2LimitDegrees=15, twistLimitDegrees=2}
 {bone1Name=dyn_jacket_l_bk_02, bone2Name=dyn_jacket_l_bk_01, swing1LimitDegrees=15, swing2LimitDegrees=15, twistLimitDegrees=2}
 {bone1Name=dyn_jacket_l_fr_01, bone2Name=dyn_jacket_l_fr_MSTR, swing1LimitDegrees=15, swing2LimitDegrees=30, twistLimitDegrees=2}
 {bone1Name=dyn_jacket_l_fr_02, bone2Name=dyn_jacket_l_fr_01, swing1LimitDegrees=15, swing2LimitDegrees=30, twistLimitDegrees=2}
 {bone1Name=dyn_jacket_r_fr_01, bone2Name=dyn_jacket_r_fr_MSTR, swing1LimitDegrees=10, swing2LimitDegrees=30, twistLimitDegrees=2}
 {bone1Name=dyn_jacket_r_fr_02, bone2Name=dyn_jacket_r_fr_01, swing1LimitDegrees=10, swing2LimitDegrees=30, twistLimitDegrees=2}
 {bone1Name=dyn_jacket_l_02, bone2Name=dyn_jacket_l_01, swing1LimitDegrees=20, swing2LimitDegrees=45, twistLimitDegrees=2}
 {bone1Name=dyn_jacket_l_01, bone2Name=dyn_jacket_l_MSTR, swing1LimitDegrees=20, swing2LimitDegrees=45, twistLimitDegrees=2}
 {bone1Name=dyn_jacket_r_01, bone2Name=dyn_jacket_r_MSTR, swing1LimitDegrees=20, swing2LimitDegrees=10, twistLimitDegrees=2}
 {bone1Name=dyn_jacket_r_02, bone2Name=dyn_jacket_r_01, swing1LimitDegrees=20, swing2LimitDegrees=45, twistLimitDegrees=2}
 {
  bone1Name=dyn_jacket_r_bk_MSTR
  bone2Name=spine_02
  swing1LimitDegrees=45
  swing1Motion=Free
  swing2LimitDegrees=45
  swing2Motion=Free
  twistLimitDegrees=45
  twistMotion=Free
 }
 {
  bone1Name=dyn_jacket_l_bk_MSTR
  bone2Name=spine_02
  swing1LimitDegrees=45
  swing1Motion=Free
  swing2LimitDegrees=45
  swing2Motion=Free
  twistLimitDegrees=45
  twistMotion=Free
 }
 {
  bone1Name=dyn_jacket_r_MSTR
  bone2Name=spine_02
  swing1LimitDegrees=45
  swing1Motion=Free
  swing2LimitDegrees=45
  swing2Motion=Free
  twistLimitDegrees=45
  twistMotion=Free
 }
 {
  bone1Name=dyn_jacket_l_MSTR
  bone2Name=spine_02
  swing1LimitDegrees=45
  swing1Motion=Free
  swing2LimitDegrees=45
  swing2Motion=Free
  twistLimitDegrees=45
  twistMotion=Free
 }
 {
  bone1Name=dyn_jacket_l_fr_MSTR
  bone2Name=spine_02
  swing1LimitDegrees=45
  swing1Motion=Free
  swing2LimitDegrees=45
  swing2Motion=Free
  twistLimitDegrees=45
  twistMotion=Free
 }
 {
  bone1Name=dyn_jacket_r_fr_MSTR
  bone2Name=spine_02
  swing1LimitDegrees=45
  swing1Motion=Free
  swing2LimitDegrees=45
  swing2Motion=Free
  twistLimitDegrees=45
  twistMotion=Free
 }
 {
  bone1Name=thigh_twist_01_r
  bone2Name=pelvis
  swing1LimitDegrees=45
  swing1Motion=Free
  swing2LimitDegrees=45
  swing2Motion=Free
  twistLimitDegrees=45
  twistMotion=Free
 }
 {
  bone1Name=spine_01
  bone2Name=pelvis
  swing1LimitDegrees=45
  swing1Motion=Free
  swing2LimitDegrees=45
  swing2Motion=Free
  twistLimitDegrees=45
  twistMotion=Free
 }
 {
  bone1Name=spine_02
  bone2Name=spine_01
  swing1LimitDegrees=45
  swing1Motion=Free
  swing2LimitDegrees=45
  swing2Motion=Free
  twistLimitDegrees=45
  twistMotion=Free
 }
 {bone1Name=dyn_jacket_l_03, bone2Name=dyn_jacket_l_02, swing1LimitDegrees=20, swing2LimitDegrees=45, twistLimitDegrees=2}
 {bone1Name=dyn_jacket_l_bk_03, bone2Name=dyn_jacket_l_bk_02, swing1LimitDegrees=15, swing2LimitDegrees=15, twistLimitDegrees=2}
 {bone1Name=dyn_jacket_l_fr_04, bone2Name=dyn_jacket_l_fr_03, swing1LimitDegrees=20, swing2LimitDegrees=45, twistLimitDegrees=2}
 {bone1Name=dyn_jacket_l_fr_03, bone2Name=dyn_jacket_l_fr_02, swing1LimitDegrees=20, swing2LimitDegrees=45, twistLimitDegrees=2}
 {bone1Name=dyn_jacket_r_03, bone2Name=dyn_jacket_r_02, swing1LimitDegrees=20, swing2LimitDegrees=45, twistLimitDegrees=2}
 {bone1Name=dyn_jacket_r_bk_03, bone2Name=dyn_jacket_r_bk_02, swing1LimitDegrees=15, swing2LimitDegrees=15, twistLimitDegrees=2}
 {bone1Name=dyn_jacket_r_fr_03, bone2Name=dyn_jacket_r_fr_02, swing1LimitDegrees=20, swing2LimitDegrees=30, twistLimitDegrees=2}
 {bone1Name=dyn_jacket_r_fr_04, bone2Name=dyn_jacket_r_fr_03, swing1LimitDegrees=20, swing2LimitDegrees=45, twistLimitDegrees=2}
 {
  bone1Name=calf_r
  bone2Name=pelvis
  swing1LimitDegrees=45
  swing1Motion=Free
  swing2LimitDegrees=45
  swing2Motion=Free
  twistLimitDegrees=45
  twistMotion=Free
 }
 {
  bone1Name=foot_r
  bone2Name=calf_r
  swing1LimitDegrees=45
  swing1Motion=Free
  swing2LimitDegrees=45
  swing2Motion=Free
  twistLimitDegrees=45
  twistMotion=Free
 }
 {
  bone1Name=pelvis
  bone2Name=root
  swing1LimitDegrees=45
  swing1Motion=Free
  swing2LimitDegrees=45
  swing2Motion=Free
  twistLimitDegrees=45
  twistMotion=Free
 }
 {
  bone1Name=calf_l
  bone2Name=pelvis
  swing1LimitDegrees=45
  swing1Motion=Free
  swing2LimitDegrees=45
  swing2Motion=Free
  twistLimitDegrees=45
  twistMotion=Free
 }
 {
  bone1Name=foot_l
  bone2Name=calf_l
  swing1LimitDegrees=45
  swing1Motion=Free
  swing2LimitDegrees=45
  swing2Motion=Free
  twistLimitDegrees=45
  twistMotion=Free
 }
 {
  bone1Name=dyn_jacket_r_fr_03
  bone2Name=dyn_jacket_r_02
  swing1LimitDegrees=15
  swing2LimitDegrees=45
  swing2Motion=Free
  twistLimitDegrees=45
  twistMotion=Free
 }
 {
  bone1Name=thigh_twist_01_l
  bone2Name=pelvis
  swing1LimitDegrees=45
  swing1Motion=Free
  swing2LimitDegrees=45
  swing2Motion=Free
  twistLimitDegrees=45
  twistMotion=Free
 }
 {
  bone1Name=dyn_jacket_l_fr_03
  bone2Name=dyn_jacket_l_02
  swing1LimitDegrees=15
  swing1Motion=Free
  swing2LimitDegrees=45
  swing2Motion=Free
  twistLimitDegrees=45
  twistMotion=Free
 }
 {
  bone1Name=dyn_jacket_l_bk_03
  bone2Name=dyn_jacket_r_bk_03
  swing1LimitDegrees=45
  swing1Motion=Free
  swing2LimitDegrees=45
  swing2Motion=Free
  twistLimitDegrees=45
  twistMotion=Free
 }
 {
  bone1Name=dyn_jacket_r_03
  bone2Name=dyn_jacket_r_bk_03
  swing1LimitDegrees=45
  swing1Motion=Free
  swing2LimitDegrees=45
  swing2Motion=Free
  twistLimitDegrees=45
  twistMotion=Free
 }
]
