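# Physics asset for the jacket flap on the medium male Apprentice cosmetic body, giving the flap bones bodies for secondary motion.
# Five capsules form a constrained chain from dyn_jacket_flaps_MSTR, with tight twist and swing limits near the root that loosen toward the tip.
# The master bone and a large dyn_scarf_collision capsule attach to the pelvis with free angular motion.
# The oversized pelvis and scarf capsules appear to act as collision volumes for the flap.
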PhysicsAsset /BRCosmetics/Characters/Player/Male/Medium/Bodies/M_MED_Apprentice/Meshes/M_MED_Apprentice_Dyn_Jacket_Flap
bodies {
 dyn_jacket_flaps_01 [{
   shapeType=Capsule
   length=9.3303
   radius=11.2915
   shapeName=dyn_jacket_l_01_capsule
   center=<8 0 6>
   rotation=<rot 85.0001 -179.9998 -179.9998>
  }]
 dyn_jacket_flaps_02 [{shapeType=Capsule, length=9.0803, radius=11.7915, shapeName=dyn_jacket_l_02_capsule, center=<9 0 5>, rotation=<rot 74.9997 -0.0004 -0.0004>}]
 dyn_jacket_flaps_03 [{shapeType=Capsule, length=5.8303, radius=10.7915, shapeName=dyn_jacket_l_02_capsule, center=<8 0 3>, rotation=<rot 90 0 0>}]
 dyn_jacket_flaps_04 [{shapeType=Capsule, length=5.8303, radius=10.7915, shapeName=dyn_jacket_l_02_capsule, center=<8 0 3>, rotation=<rot 90 0 0>}]
 dyn_jacket_flaps_05 [{shapeType=Capsule, length=5.8303, radius=10.7915, shapeName=dyn_jacket_l_02_capsule, center=<8 0 3>, rotation=<rot 90 0 0>}]
 dyn_jacket_flaps_MSTR [{shapeType=Capsule, length=1, radius=1}]
 dyn_scarf_collision [{shapeType=Capsule, length=17.25, radius=46.3372, center=<21.277 0 -42.3113>, rotation=<rot 85.0007 180 180>}]
 pelvis [{shapeType=Capsule, length=7.75, radius=32.0872, center=<-15.4231 14.3248 0>}]
}
constraints [
 {
  bone1Name=dyn_jacket_flaps_MSTR
  bone2Name=pelvis
  swing1LimitDegrees=45
  swing1Motion=Free
  swing2LimitDegrees=45
  swing2Motion=Free
  twistLimitDegrees=45
  twistMotion=Free
 }
 {bone1Name=dyn_jacket_flaps_01, bone2Name=dyn_jacket_flaps_MSTR, swing1LimitDegrees=30, swing2LimitDegrees=30, twistLimitDegrees=2}
 {bone1Name=dyn_jacket_flaps_02, bone2Name=dyn_jacket_flaps_01, swing1LimitDegrees=35, swing2LimitDegrees=45, twistLimitDegrees=2}
 {bone1Name=dyn_jacket_flaps_03, bone2Name=dyn_jacket_flaps_02, swing1LimitDegrees=60, swing2LimitDegrees=45, twistLimitDegrees=15}
 {bone1Name=dyn_jacket_flaps_04, bone2Name=dyn_jacket_flaps_03, swing1LimitDegrees=60, swing2LimitDegrees=45, twistLimitDegrees=15}
 {
  bone1Name=dyn_scarf_collision
  bone2Name=pelvis
  swing1LimitDegrees=45
  swing1Motion=Free
  swing2LimitDegrees=45
  swing2Motion=Free
  twistLimitDegrees=45
  twistMotion=Free
 }
 {bone1Name=dyn_jacket_flaps_05, bone2Name=dyn_jacket_flaps_04, swing1LimitDegrees=60, swing2LimitDegrees=45, twistLimitDegrees=15}
]
