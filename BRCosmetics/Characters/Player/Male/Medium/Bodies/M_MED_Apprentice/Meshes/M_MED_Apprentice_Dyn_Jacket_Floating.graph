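# Physics asset for the M_MED_Apprentice male body's floating jacket, defining capsule bodies and joints for its dynamic cloth bones.
# Side, back and front chains on both sides, plus a flaps chain, hang from MSTR bodies linked to spine_02 with free angular motion.
# Chain joints have mostly narrow twist and swing1 limits and wider swing2.
# Pelvis, spine, leg and large scarf-collision capsules supply body shapes around the chains.
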PhysicsAsset /BRCosmetics/Characters/Player/Male/Medium/Bodies/M_MED_Apprentice/Meshes/M_MED_Apprentice_Dyn_Jacket_Floating
bodies {
 calf_l [{shapeType=Capsule, length=41.3282, radius=15.4038, center=<-26.3787 -1.0443 0.3694>, rotation=<rot 88.1624 174.4553 174.7609>}]
 calf_r [{shapeType=Capsule, length=41.3282, radius=15.4038, center=<26.3787 1.0443 -0.3694>, rotation=<rot 88.1624 174.4553 -5.2391>}]
 dyn_jacket_flaps_01 [{
   shapeType=Capsule
   length=4.8303
   radius=15.7915
   shapeName=dyn_jacket_l_01_capsule
   center=<8 0 6>
   rotation=<rot 85.0001 -179.9998 -179.9998>
  }]
 dyn_jacket_flaps_02 [{shapeType=Capsule, length=5.3303, radius=15.5415, shapeName=dyn_jacket_l_02_capsule, center=<9 0 5>, rotation=<rot 74.9997 -0.0004 -0.0004>}]
 dyn_jacket_flaps_03 [{shapeType=Capsule, length=8.8303, radius=12.0415, shapeName=dyn_jacket_l_02_capsule, center=<8 0 3>, rotation=<rot 90 0 0>}]
 dyn_jacket_flaps_04 [{shapeType=Capsule, length=8.8303, radius=12.0415, shapeName=dyn_jacket_l_02_capsule, center=<8 0 3>, rotation=<rot 90 0 0>}]
 dyn_jacket_flaps_05 [{shapeType=Capsule, length=8.8303, radius=12.0415, shapeName=dyn_jacket_l_02_capsule, center=<8 0 3>, rotation=<rot 90 0 0>}]
 dyn_jacket_flaps_MSTR [{shapeType=Capsule, length=1, radius=1}]
 dyn_jacket_l_01 [{shapeType=Capsule, length=17.0803, radius=10.7915, shapeName=dyn_jacket_l_01_capsule, center=<7 0 6>, rotation=<rot 0 0 95>}]
 dyn_jacket_l_02 [{shapeType=Capsule, length=16, radius=17, shapeName=dyn_jacket_l_02_capsule, center=<9 0 10>, rotation=<rot 0 0 95>}]
 dyn_jacket_l_03 [{shapeType=Capsule, length=20.7053, radius=16.9165, shapeName=dyn_jacket_l_02_capsule, center=<9 0 10>, rotation=<rot 0 -15 95>}]
 dyn_jacket_l_MSTR [{shapeType=Capsule, length=1, radius=1}]
 dyn_jacket_l_bk_01 [{shapeType=Capsule, length=13.3303, radius=8.4165, shapeName=dyn_jacket_l_bk_01_capsule, center=<6 0 6>, rotation=<rot 90 0 0>}]
 dyn_jacket_l_bk_02 [{shapeType=Capsule, length=22.3303, radius=12.1665, shapeName=dyn_jacket_l_bk_02_capsule, center=<8 0 8>, rotation=<rot -0.0001 0 85>}]
 dyn_jacket_l_bk_03 [{shapeType=Capsule, length=26.2053, radius=17.1665, shapeName=dyn_jacket_l_bk_02_capsule, center=<12 0 10>, rotation=<rot -0.0001 -5 90>}]
 dyn_jacket_l_bk_MSTR [{shapeType=Capsule, length=1, radius=1}]
 dyn_jacket_l_fr_01 [{shapeType=Capsule, length=6.375, radius=9.875, shapeName=dyn_jacket_l_fr_01_capsule, center=<10 -5 6>, rotation=<rot 0.0002 0.0002 85>}]
 dyn_jacket_l_fr_02 [{shapeType=Capsule, length=8, radius=15.5, shapeName=dyn_jacket_l_fr_02_capsule, center=<12 -7 12>, rotation=<rot 0.0004 -0.0003 80>}]
 dyn_jacket_l_fr_03 [{shapeType=Capsule, length=10.75, radius=19.75, shapeName=dyn_jacket_l_fr_02_capsule, center=<10 -12 15>, rotation=<rot 0 0 80>}]
 dyn_jacket_l_fr_04 [{shapeType=Capsule, length=17.375, radius=21.375, shapeName=dyn_jacket_l_fr_02_capsule, center=<15 -15 15>, rotation=<rot 0 0 80>}]
 dyn_jacket_l_fr_MSTR [{shapeType=Capsule, length=1, radius=1}]
 dyn_jacket_r_01 [{shapeType=Capsule, length=17.0803, radius=10.7915, shapeName=dyn_jacket_l_01_capsule, center=<7 0 6>, rotation=<rot -0.0001 -0.0002 85>}]
 dyn_jacket_r_02 [{shapeType=Capsule, length=16, radius=17, shapeName=dyn_jacket_l_02_capsule, center=<9 0 10>, rotation=<rot -0.0001 -0.0002 85>}]
 dyn_jacket_r_03 [{shapeType=Capsule, length=20.7053, radius=16.9165, shapeName=dyn_jacket_l_02_capsule, center=<9 0 10>, rotation=<rot 1.2924 14.9451 85.1696>}]
 dyn_jacket_r_MSTR [{shapeType=Capsule, length=1, radius=1}]
 dyn_jacket_r_bk_01 [{shapeType=Capsule, length=13.3303, radius=8.4165, shapeName=dyn_jacket_l_bk_01_capsule, center=<6 0 6>, rotation=<rot 90 0 0>}]
 dyn_jacket_r_bk_02 [{shapeType=Capsule, length=22.3303, radius=12.1665, shapeName=dyn_jacket_l_bk_02_capsule, center=<8 0 8>, rotation=<rot -0.0001 0 95>}]
 dyn_jacket_r_bk_03 [{shapeType=Capsule, length=26.2053, radius=17.1665, shapeName=dyn_jacket_l_bk_02_capsule, center=<12 0 10>, rotation=<rot -0.0001 5 90>}]
 dyn_jacket_r_bk_MSTR [{shapeType=Capsule, length=1, radius=1}]
 dyn_jacket_r_fr_01 [{shapeType=Capsule, length=4, radius=9.5, shapeName=dyn_jacket_l_fr_01_capsule, center=<10 5 6>, rotation=<rot 0.0001 0 95>}]
 dyn_jacket_r_fr_02 [{shapeType=Capsule, length=8, radius=15.5, shapeName=dyn_jacket_l_fr_02_capsule, center=<12 7 12>, rotation=<rot 0.0004 -0.0003 100.0001>}]
 dyn_jacket_r_fr_03 [{
   shapeType=Capsule
   length=10.75
   radius=19.75
   shapeName=dyn_jacket_l_fr_02_capsule
   center=<10 12 15>
   rotation=<rot 0.0004 -0.0003 100.0001>
  }]
 dyn_jacket_r_fr_04 [{shapeType=Capsule, length=17.375, radius=21.375, shapeName=dyn_jacket_l_fr_02_capsule, center=<15 15 15>, rotation=<rot 0.0002 -0.0002 100>}]
 dyn_jacket_r_fr_MSTR [{shapeType=Capsule, length=1, radius=1}]
 dyn_scarf_collision [{shapeType=Capsule, length=28, radius=46.3372, center=<53.8035 0 -30.9017>, rotation=<rot 0 0 90>}]
 pelvis [{shapeType=Capsule, length=20, radius=22.0872, center=<-27.9255 7.6132 0>, rotation=<rot 90 178.2101 173.2102>}]
 spine_01 [{shapeType=Capsule, length=9.9006, radius=10.4006, center=<0 0.9137 0>}]
 spine_02 [{shapeType=Capsule, length=6.5256, radius=10.4006, center=<0 0.6443 0>}]
 thigh_twist_01_l [{shapeType=Capsule, length=24.5782, radius=12.1538, center=<1.4194 -1.7641 2.0202>, rotation=<rot 85.6084 176.2951 179.1272>}]
 thigh_twist_01_r [{shapeType=Capsule, length=24.5782, radius=12.1538, center=<-1.4194 1.7641 -2.0202>, rotation=<rot 85.6084 176.2951 -0.8728>}]
}
constraints [
 {bone1Name=dyn_jacket_r_bk_01, bone2Name=dyn_jacket_r_bk_MSTR, swing1LimitDegrees=2, swing2LimitDegrees=15, twistLimitDegrees=2}
 {bone1Name=dyn_jacket_r_bk_02, bone2Name=dyn_jacket_r_bk_01, swing1LimitDegrees=2, swing2LimitDegrees=15, twistLimitDegrees=2}
 {bone1Name=dyn_jacket_l_bk_01, bone2Name=dyn_jacket_l_bk_MSTR, swing1LimitDegrees=2, swing2LimitDegrees=15, twistLimitDegrees=2}
 {bone1Name=dyn_jacket_l_bk_02, bone2Name=dyn_jacket_l_bk_01, swing1LimitDegrees=2, swing2LimitDegrees=15, twistLimitDegrees=2}
 {bone1Name=dyn_jacket_l_fr_01, bone2Name=dyn_jacket_l_fr_MSTR, swing1LimitDegrees=2, swing2LimitDegrees=30, twistLimitDegrees=2}
 {bone1Name=dyn_jacket_l_fr_02, bone2Name=dyn_jacket_l_fr_01, swing1LimitDegrees=2, swing2LimitDegrees=30, twistLimitDegrees=2}
 {bone1Name=dyn_jacket_r_fr_01, bone2Name=dyn_jacket_r_fr_MSTR, swing1LimitDegrees=2, swing2LimitDegrees=30, twistLimitDegrees=2}
 {bone1Name=dyn_jacket_r_fr_02, bone2Name=dyn_jacket_r_fr_01, swing1LimitDegrees=2, swing2LimitDegrees=30, twistLimitDegrees=2}
 {bone1Name=dyn_jacket_l_02, bone2Name=dyn_jacket_l_01, swing1LimitDegrees=2, swing2LimitDegrees=45, twistLimitDegrees=2}
 {bone1Name=dyn_jacket_l_01, bone2Name=dyn_jacket_l_MSTR, swing1LimitDegrees=2, swing2LimitDegrees=45, twistLimitDegrees=2}
 {bone1Name=dyn_jacket_r_01, bone2Name=dyn_jacket_r_MSTR, swing1LimitDegrees=2, swing2LimitDegrees=10, twistLimitDegrees=2}
 {bone1Name=dyn_jacket_r_02, bone2Name=dyn_jacket_r_01, swing1LimitDegrees=2, swing2LimitDegrees=45, twistLimitDegrees=2}
 {
  bone1Name=dyn_jacket_r_bk_MSTR
  bone2Name=spine_02
  swing1LimitDegrees=45
  swing1Motion=Free
  swing2LimitDegrees=45
  swing2Motion=Free
  twistLimitDegrees=45
  twistMotion=Free
 }
 {
  bone1Name=dyn_jacket_l_bk_MSTR
  bone2Name=spine_02
  swing1LimitDegrees=45
  swing1Motion=Free
  swing2LimitDegrees=45
  swing2Motion=Free
  twistLimitDegrees=45
  twistMotion=Free
 }
 {
  bone1Name=dyn_jacket_r_MSTR
  bone2Name=spine_02
  swing1LimitDegrees=45
  swing1Motion=Free
  swing2LimitDegrees=45
  swing2Motion=Free
  twistLimitDegrees=45
  twistMotion=Free
 }
 {
  bone1Name=dyn_jacket_l_MSTR
  bone2Name=spine_02
  swing1LimitDegrees=45
  swing1Motion=Free
  swing2LimitDegrees=45
  swing2Motion=Free
  twistLimitDegrees=45
  twistMotion=Free
 }
 {
  bone1Name=dyn_jacket_l_fr_MSTR
  bone2Name=spine_02
  swing1LimitDegrees=45
  swing1Motion=Free
  swing2LimitDegrees=45
  swing2Motion=Free
  twistLimitDegrees=45
  twistMotion=Free
 }
 {
  bone1Name=dyn_jacket_r_fr_MSTR
  bone2Name=spine_02
  swing1LimitDegrees=45
  swing1Motion=Free
  swing2LimitDegrees=45
  swing2Motion=Free
  twistLimitDegrees=45
  twistMotion=Free
 }
 {
  bone1Name=spine_01
  bone2Name=pelvis
  swing1LimitDegrees=45
  swing1Motion=Free
  swing2LimitDegrees=45
  swing2Motion=Free
  twistLimitDegrees=45
  twistMotion=Free
 }
 {
  bone1Name=spine_02
  bone2Name=spine_01
  swing1LimitDegrees=45
  swing1Motion=Free
  swing2LimitDegrees=45
  swing2Motion=Free
  twistLimitDegrees=45
  twistMotion=Free
 }
 {bone1Name=dyn_jacket_l_03, bone2Name=dyn_jacket_l_02, swing1LimitDegrees=2, swing2LimitDegrees=45, twistLimitDegrees=2}
 {bone1Name=dyn_jacket_l_bk_03, bone2Name=dyn_jacket_l_bk_02, swing1LimitDegrees=2, swing2LimitDegrees=15, twistLimitDegrees=2}
 {bone1Name=dyn_jacket_l_fr_04, bone2Name=dyn_jacket_l_fr_03, swing1LimitDegrees=2, swing2LimitDegrees=45, twistLimitDegrees=2}
 {bone1Name=dyn_jacket_l_fr_03, bone2Name=dyn_jacket_l_fr_02, swing1LimitDegrees=2, swing2LimitDegrees=45, twistLimitDegrees=2}
 {bone1Name=dyn_jacket_r_03, bone2Name=dyn_jacket_r_02, swing1LimitDegrees=2, swing2LimitDegrees=45, twistLimitDegrees=2}
 {bone1Name=dyn_jacket_r_bk_03, bone2Name=dyn_jacket_r_bk_02, swing1LimitDegrees=2, swing2LimitDegrees=15, twistLimitDegrees=2}
 {bone1Name=dyn_jacket_r_fr_03, bone2Name=dyn_jacket_r_fr_02, swing1LimitDegrees=2, swing2LimitDegrees=30, twistLimitDegrees=2}
 {bone1Name=dyn_jacket_r_fr_04, bone2Name=dyn_jacket_r_fr_03, swing1LimitDegrees=2, swing2LimitDegrees=45, twistLimitDegrees=2}
 {
  bone1Name=dyn_jacket_flaps_MSTR
  bone2Name=spine_02
  swing1LimitDegrees=45
  swing1Motion=Free
  swing2LimitDegrees=45
  swing2Motion=Free
  twistLimitDegrees=45
  twistMotion=Free
 }
 {bone1Name=dyn_jacket_flaps_01, bone2Name=dyn_jacket_flaps_MSTR, swing1LimitDegrees=15, swing2LimitDegrees=30, twistLimitDegrees=2}
 {bone1Name=dyn_jacket_flaps_02, bone2Name=dyn_jacket_flaps_01, swing1LimitDegrees=15, swing2LimitDegrees=30, twistLimitDegrees=2}
 {bone1Name=dyn_jacket_flaps_03, bone2Name=dyn_jacket_flaps_02, swing1LimitDegrees=45, swing2LimitDegrees=30, twistLimitDegrees=2}
 {bone1Name=dyn_jacket_flaps_04, bone2Name=dyn_jacket_flaps_03, swing1LimitDegrees=45, swing2LimitDegrees=30, twistLimitDegrees=2}
 {bone1Name=dyn_jacket_flaps_05, bone2Name=dyn_jacket_flaps_04, swing1LimitDegrees=45, swing2LimitDegrees=30, twistLimitDegrees=2}
 {
  bone1Name=dyn_scarf_collision
  bone2Name=pelvis
  swing1LimitDegrees=45
  swing1Motion=Free
  swing2LimitDegrees=45
  swing2Motion=Free
  twistLimitDegrees=45
  twistMotion=Free
 }
 {
  bone1Name=thigh_twist_01_l
  bone2Name=pelvis
  swing1LimitDegrees=45
  swing1Motion=Free
  swing2LimitDegrees=45
  swing2Motion=Free
  twistLimitDegrees=45
  twistMotion=Free
 }
 {
  bone1Name=calf_r
  bone2Name=pelvis
  swing1LimitDegrees=45
  swing1Motion=Free
  swing2LimitDegrees=45
  swing2Motion=Free
  twistLimitDegrees=45
  twistMotion=Free
 }
 {
  bone1Name=calf_l
  bone2Name=pelvis
  swing1LimitDegrees=45
  swing1Motion=Free
  swing2LimitDegrees=45
  swing2Motion=Free
  twistLimitDegrees=45
  twistMotion=Free
 }
 {
  bone1Name=thigh_twist_01_r
  bone2Name=pelvis
  swing1LimitDegrees=45
  swing1Motion=Free
  swing2LimitDegrees=45
  swing2Motion=Free
  twistLimitDegrees=45
  twistMotion=Free
 }
]
