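# Physics asset for the medium male Apprentice body's dynamic jacket, by its name the variant for sliding.
# Capsule bone chains cover left and right front, side and back panels plus a flaps chain, each with limited swing and tight twist between links.
# Each chain's MSTR bone joins spine_02 with free angular motion, linked through spine_01 and pelvis to root.
# Root carries a body capsule and a huge, offset zero-length crouch collision capsule, effectively a sphere.
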PhysicsAsset /BRCosmetics/Characters/Player/Male/Medium/Bodies/M_MED_Apprentice/Meshes/M_MED_Apprentice_Dyn_Jacket_Sliding
bodies {
 dyn_jacket_flaps_01 [{
   shapeType=Capsule
   length=9.8303
   radius=10.7915
   shapeName=dyn_jacket_l_01_capsule
   center=<8 0 6>
   rotation=<rot 85.0001 -179.9998 -179.9998>
  }]
 dyn_jacket_flaps_02 [{
   shapeType=Capsule
   length=7.8303
   radius=16.2915
   shapeName=dyn_jacket_l_02_capsule
   center=<11.1069 0 4.4354>
   rotation=<rot 79.9996 -0.0008 -0.0008>
  }]
 dyn_jacket_flaps_03 [{shapeType=Capsule, length=8.0803, radius=14.2915, shapeName=dyn_jacket_l_02_capsule, center=<8 0 3>, rotation=<rot 90 0 0>}]
 dyn_jacket_flaps_04 [{shapeType=Capsule, length=5.8303, radius=18.7915, shapeName=dyn_jacket_l_02_capsule, center=<8 0 3>, rotation=<rot 90 0 0>}]
 dyn_jacket_flaps_05 [{shapeType=Capsule, length=5.8303, radius=18.7915, shapeName=dyn_jacket_l_02_capsule, center=<8 0 3>, rotation=<rot 90 0 0>}]
 dyn_jacket_flaps_MSTR [{shapeType=Capsule, length=1, radius=1}]
 dyn_jacket_l_01 [{shapeType=Capsule, length=17.0803, radius=10.7915, shapeName=dyn_jacket_l_01_capsule, center=<7 0 6>, rotation=<rot 0 0 95>}]
 dyn_jacket_l_02 [{shapeType=Capsule, length=16, radius=17, shapeName=dyn_jacket_l_02_capsule, center=<9 0 10>, rotation=<rot 0 0 95>}]
 dyn_jacket_l_03 [{shapeType=Capsule, length=17.2053, radius=20.4165, shapeName=dyn_jacket_l_02_capsule, center=<9 0 10>, rotation=<rot 0 -15 95>}]
 dyn_jacket_l_MSTR [{shapeType=Capsule, length=1, radius=1}]
 dyn_jacket_l_bk_01 [{shapeType=Capsule, length=10.8303, radius=10.9165, shapeName=dyn_jacket_l_bk_01_capsule, center=<6 0 6>, rotation=<rot -0.0001 0 85>}]
 dyn_jacket_l_bk_02 [{shapeType=Capsule, length=20.8303, radius=13.6665, shapeName=dyn_jacket_l_bk_02_capsule, center=<8 0 8>, rotation=<rot -0.0001 0 85>}]
 dyn_jacket_l_bk_03 [{shapeType=Capsule, length=14.4553, radius=17.1665, shapeName=dyn_jacket_l_bk_02_capsule, center=<12 0 10>, rotation=<rot 0 0 90>}]
 dyn_jacket_l_bk_MSTR [{shapeType=Capsule, length=1, radius=1}]
 dyn_jacket_l_fr_01 [{shapeType=Capsule, length=6.375, radius=9.875, shapeName=dyn_jacket_l_fr_01_capsule, center=<10 -5 6>, rotation=<rot 0.0002 0.0002 85>}]
 dyn_jacket_l_fr_02 [{shapeType=Capsule, length=8, radius=15.5, shapeName=dyn_jacket_l_fr_02_capsule, center=<12 -7 12>, rotation=<rot 0.0004 -0.0003 80>}]
 dyn_jacket_l_fr_03 [{shapeType=Capsule, length=9, radius=21.5, shapeName=dyn_jacket_l_fr_02_capsule, center=<10 -12 15>, rotation=<rot 0 0 80>}]
 dyn_jacket_l_fr_04 [{shapeType=Capsule, length=13.125, radius=25.625, shapeName=dyn_jacket_l_fr_02_capsule, center=<15 -15 15>, rotation=<rot 0 0 80>}]
 dyn_jacket_l_fr_MSTR [{shapeType=Capsule, length=1, radius=1}]
 dyn_jacket_r_01 [{shapeType=Capsule, length=17.0803, radius=10.7915, shapeName=dyn_jacket_l_01_capsule, center=<7 0 6>, rotation=<rot -0.0001 -0.0002 85>}]
 dyn_jacket_r_02 [{shapeType=Capsule, length=16, radius=17, shapeName=dyn_jacket_l_02_capsule, center=<9 0 10>, rotation=<rot -0.0001 -0.0002 85>}]
 dyn_jacket_r_03 [{shapeType=Capsule, length=17.2053, radius=20.4165, shapeName=dyn_jacket_l_02_capsule, center=<9 0 10>, rotation=<rot 1.2924 14.9451 85.1696>}]
 dyn_jacket_r_MSTR [{shapeType=Capsule, length=1, radius=1}]
 dyn_jacket_r_bk_01 [{shapeType=Capsule, length=10.8303, radius=10.9165, shapeName=dyn_jacket_l_bk_01_capsule, center=<6 0 6>, rotation=<rot -0.0001 0 95>}]
 dyn_jacket_r_bk_02 [{shapeType=Capsule, length=20.8303, radius=13.6665, shapeName=dyn_jacket_l_bk_02_capsule, center=<8 0 8>, rotation=<rot -0.0001 0 95>}]
 dyn_jacket_r_bk_03 [{shapeType=Capsule, length=14.4553, radius=17.1665, shapeName=dyn_jacket_l_bk_02_capsule, center=<12 0 10>, rotation=<rot 0 0 90>}]
 dyn_jacket_r_bk_MSTR [{shapeType=Capsule, length=1, radius=1}]
 dyn_jacket_r_fr_01 [{shapeType=Capsule, length=4, radius=9.5, shapeName=dyn_jacket_l_fr_01_capsule, center=<10 5 6>, rotation=<rot 0.0001 0 95>}]
 dyn_jacket_r_fr_02 [{shapeType=Capsule, length=8, radius=15.5, shapeName=dyn_jacket_l_fr_02_capsule, center=<12 7 12>, rotation=<rot 0.0004 -0.0003 100.0001>}]
 dyn_jacket_r_fr_03 [{shapeType=Capsule, length=9.5, radius=21, shapeName=dyn_jacket_l_fr_02_capsule, center=<10 12 15>, rotation=<rot 0.0004 -0.0003 100.0001>}]
 dyn_jacket_r_fr_04 [{shapeType=Capsule, length=14.625, radius=24.125, shapeName=dyn_jacket_l_fr_02_capsule, center=<15 15 15>, rotation=<rot 0.0002 -0.0002 100>}]
 dyn_jacket_r_fr_MSTR [{shapeType=Capsule, length=1, radius=1}]
 pelvis [{shapeType=Capsule, length=8.6, radius=13.5872, center=<-4.1085 -0.5916 0>}]
 root [
  {shapeType=Capsule, length=0, radius=240.625, shapeName=dyn_crouch_collision_capsule, center=<-0.0003 6.7341 -228.7847>}
  {
   shapeType=Capsule
   length=18.5
   radius=30.375
   shapeName=root_capsule
   center=<-0.9483 6.6358 -9.4844>
   rotation=<rot -0.4349 20.0189 85.018>
  }
 ]
 spine_01 [{shapeType=Capsule, length=9.9006, radius=10.4006, center=<0 0.9137 0>}]
 spine_02 [{shapeType=Capsule, length=6.5256, radius=10.4006, center=<0 0.6443 0>}]
}
constraints [
 {bone1Name=dyn_jacket_r_bk_01, bone2Name=dyn_jacket_r_bk_MSTR, swing1LimitDegrees=10, swing2LimitDegrees=45, twistLimitDegrees=2}
 {bone1Name=dyn_jacket_r_bk_02, bone2Name=dyn_jacket_r_bk_01, swing1LimitDegrees=5, swing2LimitDegrees=45, twistLimitDegrees=2}
 {bone1Name=dyn_jacket_l_bk_01, bone2Name=dyn_jacket_l_bk_MSTR, swing1LimitDegrees=10, swing2LimitDegrees=45, twistLimitDegrees=2}
 {bone1Name=dyn_jacket_l_bk_02, bone2Name=dyn_jacket_l_bk_01, swing1LimitDegrees=5, swing2LimitDegrees=45, twistLimitDegrees=2}
 {bone1Name=dyn_jacket_l_fr_01, bone2Name=dyn_jacket_l_fr_MSTR, swing1LimitDegrees=5, swing2LimitDegrees=30, twistLimitDegrees=2}
 {bone1Name=dyn_jacket_l_fr_02, bone2Name=dyn_jacket_l_fr_01, swing1LimitDegrees=5, swing2LimitDegrees=45, twistLimitDegrees=2}
 {bone1Name=dyn_jacket_r_fr_01, bone2Name=dyn_jacket_r_fr_MSTR, swing1LimitDegrees=5, swing2LimitDegrees=30, twistLimitDegrees=2}
 {bone1Name=dyn_jacket_r_fr_02, bone2Name=dyn_jacket_r_fr_01, swing1LimitDegrees=5, swing2LimitDegrees=45, twistLimitDegrees=2}
 {bone1Name=dyn_jacket_l_02, bone2Name=dyn_jacket_l_01, swing1LimitDegrees=5, swing2LimitDegrees=45, twistLimitDegrees=2}
 {bone1Name=dyn_jacket_l_01, bone2Name=dyn_jacket_l_MSTR, swing1LimitDegrees=5, swing2LimitDegrees=45, twistLimitDegrees=2}
 {bone1Name=dyn_jacket_r_01, bone2Name=dyn_jacket_r_MSTR, swing1LimitDegrees=5, swing2LimitDegrees=45, twistLimitDegrees=2}
 {bone1Name=dyn_jacket_r_02, bone2Name=dyn_jacket_r_01, swing1LimitDegrees=5, swing2LimitDegrees=45, twistLimitDegrees=2}
 {
  bone1Name=dyn_jacket_r_bk_MSTR
  bone2Name=spine_02
  swing1LimitDegrees=45
  swing1Motion=Free
  swing2LimitDegrees=45
  swing2Motion=Free
  twistLimitDegrees=45
  twistMotion=Free
 }
 {
  bone1Name=dyn_jacket_l_bk_MSTR
  bone2Name=spine_02
  swing1LimitDegrees=45
  swing1Motion=Free
  swing2LimitDegrees=45
  swing2Motion=Free
  twistLimitDegrees=45
  twistMotion=Free
 }
 {
  bone1Name=dyn_jacket_r_MSTR
  bone2Name=spine_02
  swing1LimitDegrees=45
  swing1Motion=Free
  swing2LimitDegrees=45
  swing2Motion=Free
  twistLimitDegrees=45
  twistMotion=Free
 }
 {
  bone1Name=dyn_jacket_l_MSTR
  bone2Name=spine_02
  swing1LimitDegrees=45
  swing1Motion=Free
  swing2LimitDegrees=45
  swing2Motion=Free
  twistLimitDegrees=45
  twistMotion=Free
 }
 {
  bone1Name=dyn_jacket_l_fr_MSTR
  bone2Name=spine_02
  swing1LimitDegrees=45
  swing1Motion=Free
  swing2LimitDegrees=45
  swing2Motion=Free
  twistLimitDegrees=45
  twistMotion=Free
 }
 {
  bone1Name=dyn_jacket_r_fr_MSTR
  bone2Name=spine_02
  swing1LimitDegrees=45
  swing1Motion=Free
  swing2LimitDegrees=45
  swing2Motion=Free
  twistLimitDegrees=45
  twistMotion=Free
 }
 {
  bone1Name=spine_01
  bone2Name=pelvis
  swing1LimitDegrees=45
  swing1Motion=Free
  swing2LimitDegrees=45
  swing2Motion=Free
  twistLimitDegrees=45
  twistMotion=Free
 }
 {
  bone1Name=spine_02
  bone2Name=spine_01
  swing1LimitDegrees=45
  swing1Motion=Free
  swing2LimitDegrees=45
  swing2Motion=Free
  twistLimitDegrees=45
  twistMotion=Free
 }
 {bone1Name=dyn_jacket_l_03, bone2Name=dyn_jacket_l_02, swing1LimitDegrees=5, swing2LimitDegrees=45, twistLimitDegrees=2}
 {bone1Name=dyn_jacket_l_bk_03, bone2Name=dyn_jacket_l_bk_02, swing1LimitDegrees=5, swing2LimitDegrees=45, twistLimitDegrees=2}
 {bone1Name=dyn_jacket_l_fr_04, bone2Name=dyn_jacket_l_fr_03, swing1LimitDegrees=5, swing2LimitDegrees=45, twistLimitDegrees=2}
 {bone1Name=dyn_jacket_l_fr_03, bone2Name=dyn_jacket_l_fr_02, swing1LimitDegrees=5, swing2LimitDegrees=45, twistLimitDegrees=2}
 {bone1Name=dyn_jacket_r_03, bone2Name=dyn_jacket_r_02, swing1LimitDegrees=5, swing2LimitDegrees=45, twistLimitDegrees=2}
 {bone1Name=dyn_jacket_r_bk_03, bone2Name=dyn_jacket_r_bk_02, swing1LimitDegrees=5, swing2LimitDegrees=45, twistLimitDegrees=2}
 {bone1Name=dyn_jacket_r_fr_03, bone2Name=dyn_jacket_r_fr_02, swing1LimitDegrees=5, swing2LimitDegrees=45, twistLimitDegrees=2}
 {bone1Name=dyn_jacket_r_fr_04, bone2Name=dyn_jacket_r_fr_03, swing1LimitDegrees=5, swing2LimitDegrees=45, twistLimitDegrees=2}
 {
  bone1Name=pelvis
  bone2Name=root
  swing1LimitDegrees=45
  swing1Motion=Free
  swing2LimitDegrees=45
  swing2Motion=Free
  twistLimitDegrees=45
  twistMotion=Free
 }
 {
  bone1Name=dyn_jacket_flaps_MSTR
  bone2Name=spine_02
  swing1LimitDegrees=45
  swing1Motion=Free
  swing2LimitDegrees=45
  swing2Motion=Free
  twistLimitDegrees=45
  twistMotion=Free
 }
 {bone1Name=dyn_jacket_flaps_01, bone2Name=dyn_jacket_flaps_MSTR, swing1LimitDegrees=45, swing2LimitDegrees=45, twistLimitDegrees=2}
 {bone1Name=dyn_jacket_flaps_02, bone2Name=dyn_jacket_flaps_01, swing1LimitDegrees=45, swing2LimitDegrees=45, twistLimitDegrees=2}
 {bone1Name=dyn_jacket_flaps_03, bone2Name=dyn_jacket_flaps_02, swing1LimitDegrees=45, swing2LimitDegrees=45, twistLimitDegrees=2}
 {bone1Name=dyn_jacket_flaps_04, bone2Name=dyn_jacket_flaps_03, swing1LimitDegrees=45, swing2LimitDegrees=45, twistLimitDegrees=2}
 {bone1Name=dyn_jacket_flaps_05, bone2Name=dyn_jacket_flaps_04, swing1LimitDegrees=45, swing2LimitDegrees=45, twistLimitDegrees=2}
]
