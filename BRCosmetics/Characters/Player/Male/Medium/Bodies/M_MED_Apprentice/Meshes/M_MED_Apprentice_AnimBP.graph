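# Animation Blueprint for the M_MED_Apprentice cosmetic body, adding deformation and secondary motion to the pose copied from its parent mesh.
# Two Control Rigs feed bone constraints, corrective offsets and AnimDynamics for drawstrings, sleeves and belt, cached as prePose.
# RigidBody nodes simulate the jacket with different physics assets, selected by floating/diving, parachute and glider flags.
# A custom-emote flag switches jacket setups and toggles final arm, hand and back-jacket offsets.
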AnimBlueprint /BRCosmetics/Characters/Player/Male/Medium/Bodies/M_MED_Apprentice/Meshes/M_MED_Apprentice_AnimBP
target_skeleton=&/M_MED_Apprentice_Skeleton
graphs [
 AnimGraph {nodes [
   n1 = "Transform (Modify) Bone - Bone: lowerarm_l"(ComponentPose=Local To Component(LocalPose=Use cached pose 'FinalSetup'()), Alpha=1, Translation="-1.7,0,0", Rotation="0,0,0", Scale="1,1,1") {
    rotationMode=BMM_Additive
    rotationSpace=BCS_BoneSpace
    translationMode=BMM_Additive
    translationSpace=BCS_BoneSpace
    alphaBoolBlend {blendOption=Linear}
    boneToModify {boneName=lowerarm_l}
    translation {x=-1.7}
   }
   n2 = "Transform (Modify) Bone - Bone: lowerarm_r"(ComponentPose=@n1, Alpha=1, Translation="3.026,0,0", Rotation="0,0,0", Scale="1,1,1") {
    rotationMode=BMM_Additive
    rotationSpace=BCS_BoneSpace
    translationMode=BMM_Additive
    translationSpace=BCS_BoneSpace
    alphaBoolBlend {blendOption=Linear}
    boneToModify {boneName=lowerarm_r}
    translation {x=3.026}
   }
   n3 = "Transform (Modify) Bone - Bone: hand_l"(ComponentPose=@n2, Alpha=1, Translation="-0.000,1.500,-0.000", Rotation="-1.273,6.370,9.991", Scale="1,1,1") {
    rotationMode=BMM_Additive
    rotationSpace=BCS_BoneSpace
    translationMode=BMM_Additive
    translationSpace=BCS_BoneSpace
    alphaBoolBlend {blendOption=Linear}
    boneToModify {boneName=hand_l}
    rotation=<rot -1.273 6.37 9.991>
    translation {y=1.5}
   }
   n4 = "Transform (Modify) Bone - Bone: hand_r"(ComponentPose=@n3, Alpha=1, Translation="0,0,0", Rotation="0.004,4.308,1.306", Scale="1,1,1") {
    rotationMode=BMM_Additive
    rotationSpace=BCS_BoneSpace
    translationMode=BMM_Additive
    translationSpace=BCS_BoneSpace
    alphaBoolBlend {blendOption=Linear}
    boneToModify {boneName=hand_r}
    rotation=<rot 0.004 4.308 1.306>
   }
   n5 = "Transform (Modify) Bone - Bone: dyn_jacket_r_bk_01"(ComponentPose=@n4, Alpha=1, Translation="-2.8,0.6,-7.5", Rotation="6.676,-0.763,-2.812", Scale="1,1,1") {
    rotationMode=BMM_Additive
    rotationSpace=BCS_BoneSpace
    translationMode=BMM_Additive
    translationSpace=BCS_BoneSpace
    alphaBoolBlend {blendOption=Linear}
    boneToModify {boneName=dyn_jacket_r_bk_01}
    rotation=<rot 6.6759 -0.763 -2.812>
    translation=<-2.8 0.6 -7.5>
   }
   n6 = "Transform (Modify) Bone - Bone: dyn_jacket_l_bk_01"(ComponentPose=@n5, Alpha=1, Translation="-2.800,0.675,-7.570", Rotation="8.050,-0.522,0.429", Scale="1,1,1") {
    rotationMode=BMM_Additive
    rotationSpace=BCS_BoneSpace
    translationMode=BMM_Additive
    translationSpace=BCS_BoneSpace
    alphaBoolBlend {blendOption=Linear}
    boneToModify {boneName=dyn_jacket_l_bk_01}
    rotation=<rot 8.05 -0.522 0.429>
    translation=<-2.8 0.675 -7.57>
   }
   Output Pose(Result=Blend Poses by bool(
     BlendPose_0=Component To Local(ComponentPose=@n6)
     BlendPose_1=Use cached pose 'FinalSetup'()
     bActiveValue=Get isCustomEmotePlaying_0() -> (isCustomEmotePlaying_0:bool)
    ) {blendType=Linear}) {layerGroup=None}
   n7 = Control Rig!(Source=Copy Pose From Mesh() {bUseAttachedParent=true}, Alpha, Sit:double=Get isCustomEmotePlaying() -> (isCustomEmotePlaying:double)) -> (Pose) {
    alphaBoolBlend {blendOption=Linear}
    controlRigAssetReference {blueprintRigClass=&/M_MED_Apprentice_Deform_CtrlRig_twists.M_MED_Apprentice_Deform_CtrlRig_twists_C}
   }
   n8 = Control Rig!(Source=@n7, Alpha) -> (Pose) {
    alphaBoolBlend {blendOption=Linear}
    controlRigAssetReference {blueprintRigClass=&/M_MED_Apprentice_DeformRig_ControlRig.M_MED_Apprentice_DeformRig_ControlRig_C}
   }
   n9 = "Constraint - attach_backpack (PPP)"!(
    ComponentPose=Local To Component(LocalPose=@n8)
    Alpha=1
    ConstraintWeights_0=0.568
    ConstraintWeights_1:float=0.292
    ConstraintWeights_2:float=0.14
   ) -> (Pose) {
    alphaBoolBlend {blendOption=Linear}
    boneToModify {boneName=attach_backpack}
    constraintSetup [{targetBone {boneName=spine_04}}, {targetBone {boneName=spine_05}}, {targetBone {boneName=spine_03}}]
    constraintWeights [1,1,1]
   }
   n10 = "Constraint - attach_cape (PPPP)"!(
    ComponentPose=@n9
    Alpha=1
    ConstraintWeights_0=0.1325
    ConstraintWeights_1:float=0.8385
    ConstraintWeights_2:float=0.016
    ConstraintWeights_3:float=0.013
   ) -> (Pose) {
    alphaBoolBlend {blendOption=Linear}
    boneToModify {boneName=attach_cape}
    constraintSetup [{targetBone {boneName=spine_04}}, {targetBone {boneName=spine_05}}, {targetBone {boneName=clavicle_r}}, {targetBone {boneName=neck_01}}]
    constraintWeights [1,1,1,1]
   }
   n11 = "Transform (Modify) Bone - Bone: attach_backpack"!(ComponentPose=@n10, bAlphaBoolEnabled:bool=True, Translation="0,-30,0", Rotation="0,0,0", Scale="1,1,1") -> (Pose) {
    alphaInputType=Bool
    rotationMode=BMM_Additive
    rotationSpace=BCS_BoneSpace
    translationMode=BMM_Additive
    translationSpace=BCS_BoneSpace
    alphaBoolBlend {blendOption=Linear}
    boneToModify {boneName=attach_backpack}
    translation {y=-30}
   }
   n12 = "Transform (Modify) Bone - Bone: upperarm_twist_02_l"(ComponentPose=@n11, Alpha=1, Translation="0,0,0", Rotation="0,0,0", Scale="1,1.5,1.5") {
    rotationMode=BMM_Additive
    rotationSpace=BCS_BoneSpace
    scaleMode=BMM_Additive
    scaleSpace=BCS_BoneSpace
    translationMode=BMM_Additive
    translationSpace=BCS_BoneSpace
    alphaBoolBlend {blendOption=Linear}
    boneToModify {boneName=upperarm_twist_02_l}
    scale {y=1.5, z=1.5}
   }
   n13 = "Transform (Modify) Bone - Bone: def_collar_ft_r"(ComponentPose=@n12, Alpha=1, Translation="0,0,-7", Rotation="0,0,15", Scale="1,1,1") {
    rotationMode=BMM_Additive
    rotationSpace=BCS_BoneSpace
    scaleMode=BMM_Additive
    scaleSpace=BCS_BoneSpace
    translationMode=BMM_Additive
    translationSpace=BCS_BoneSpace
    alphaBoolBlend {blendOption=Linear}
    boneToModify {boneName=def_collar_ft_r}
    rotation {roll=15}
    translation {z=-7}
   }
   n14 = "Transform (Modify) Bone - Bone: def_collar_r"(ComponentPose=@n13, Alpha=1, Translation="-2,0,-4", Rotation="-40,0,60", Scale="1,1,1") {
    rotationMode=BMM_Additive
    rotationSpace=BCS_BoneSpace
    scaleMode=BMM_Additive
    scaleSpace=BCS_BoneSpace
    translationMode=BMM_Additive
    translationSpace=BCS_BoneSpace
    alphaBoolBlend {blendOption=Linear}
    boneToModify {boneName=def_collar_r}
    rotation {pitch=-40, roll=60}
    translation {x=-2, z=-4}
   }
   n15 = "Transform (Modify) Bone - Bone: def_jacket_collar_r"(ComponentPose=@n14, Alpha=0, Translation="-1,2,0", Rotation="90,80,65", Scale="1,1,1") {
    rotationMode=BMM_Additive
    rotationSpace=BCS_BoneSpace
    scaleMode=BMM_Additive
    scaleSpace=BCS_BoneSpace
    translationMode=BMM_Additive
    translationSpace=BCS_BoneSpace
    alphaBoolBlend {blendOption=Linear}
    alphaScaleBiasClamp {bInterpResult=true, interpSpeedDecreasing=15, interpSpeedIncreasing=15}
    boneToModify {boneName=def_jacket_collar_r}
    rotation=<rot 90 80 65>
    translation=<-1 2>
   }
   n16 = "Transform (Modify) Bone - Bone: def_jacket_collar_r"(ComponentPose=@n15, Alpha=0, Translation="0,0,0", Rotation="120,0,15", Scale="1,1,1") {
    rotationMode=BMM_Additive
    rotationSpace=BCS_BoneSpace
    scaleMode=BMM_Additive
    scaleSpace=BCS_BoneSpace
    translationMode=BMM_Additive
    translationSpace=BCS_BoneSpace
    alphaBoolBlend {blendOption=Linear}
    alphaScaleBiasClamp {bInterpResult=true, interpSpeedDecreasing=15, interpSpeedIncreasing=15}
    boneToModify {boneName=def_jacket_collar_r}
    rotation {pitch=120, roll=15}
   }
   n17 = "Transform (Modify) Bone - Bone: def_jacket_collar_l"(ComponentPose=@n16, Alpha=1, Translation="0,0,0", Rotation="30,0,10", Scale="1,1,1") {
    rotationMode=BMM_Additive
    rotationSpace=BCS_BoneSpace
    scaleMode=BMM_Additive
    scaleSpace=BCS_BoneSpace
    translationMode=BMM_Additive
    translationSpace=BCS_BoneSpace
    alphaBoolBlend {blendOption=Linear}
    alphaScaleBiasClamp {bInterpResult=true, interpSpeedDecreasing=15, interpSpeedIncreasing=15}
    boneToModify {boneName=def_jacket_collar_l}
    rotation {pitch=30, roll=10}
   }
   n18 = "Transform (Modify) Bone - Bone: def_jacket_collar_l"(ComponentPose=@n17, Alpha=1, Translation="-2,0,0", Rotation="80,0,-20", Scale="1,1,1") {
    rotationMode=BMM_Additive
    rotationSpace=BCS_BoneSpace
    scaleMode=BMM_Additive
    scaleSpace=BCS_BoneSpace
    translationMode=BMM_Additive
    translationSpace=BCS_BoneSpace
    alphaBoolBlend {blendOption=Linear}
    alphaScaleBiasClamp {bInterpResult=true, interpSpeedDecreasing=15, interpSpeedIncreasing=15}
    boneToModify {boneName=def_jacket_collar_l}
    rotation {pitch=80, roll=-20}
    translation {x=-2}
   }
   n19 = Anim Dynamics - Bone: dyn_drawstring_l!(ComponentPose=@n18, Alpha=1, GravityScale:float=1, ExternalForce="0,0,0") -> (Pose) {
    angularDampingOverride=0.99
    angularSpringConstant=1000
    bGravityOverrideInSimSpace=true
    bOverrideAngularDamping=true
    bUseGravityOverride=true
    simulationSpace=RootRelative
    alphaBoolBlend {blendOption=Linear}
    boundBone {boneName=dyn_drawstring_l}
    componentAppliedLinearAccClamp=<3000 3000 3000>
    componentLinearAccScale=<0.2 0.2 0.2>
    componentLinearVelScale=<0.3 0.3 0.3>
    gravityOverride {z=-980}
    physicsBodyDefinitions [{
      sphereCollisionRadius=0
      boundBone {boneName=dyn_drawstring_l}
      boxExtents=<10 20 20>
      constraintSetup {angularLimitsMax=<0 0 30>, angularLimitsMin=<0 -30 -30>, angularTarget=<0 0 0>}
      localJointOffset=<3 0 0>
     }]
    relativeSpaceBone {boneName=lowerarm_r}
   }
   n20 = Anim Dynamics - Bone: dyn_drawstring_r!(ComponentPose=@n19, Alpha=1, GravityScale:float=1, ExternalForce="0,0,0") -> (Pose) {
    angularDampingOverride=0.99
    angularSpringConstant=1000
    bGravityOverrideInSimSpace=true
    bOverrideAngularDamping=true
    bUseGravityOverride=true
    simulationSpace=RootRelative
    alphaBoolBlend {blendOption=Linear}
    boundBone {boneName=dyn_drawstring_r}
    componentAppliedLinearAccClamp=<3000 3000 3000>
    componentLinearAccScale=<0.2 0.2 0.2>
    componentLinearVelScale=<0.3 0.3 0.3>
    gravityOverride {z=-980}
    physicsBodyDefinitions [{
      sphereCollisionRadius=0
      boundBone {boneName=dyn_drawstring_r}
      boxExtents=<10 20 20>
      constraintSetup {angularLimitsMax=<0 0 30>, angularLimitsMin=<0 -30 -30>, angularTarget=<0 0 0>}
      localJointOffset=<5 0 0>
     }]
    relativeSpaceBone {boneName=lowerarm_r}
   }
   n21 = Anim Dynamics - Bone: dyn_sleeves_r_MSTR!(ComponentPose=@n20, Alpha=1, GravityScale:float=1, ExternalForce="0,0,0") -> (Pose) {
    angularDampingOverride=1
    angularSpringConstant=500
    bGravityOverrideInSimSpace=true
    bOverrideAngularDamping=true
    bUseGravityOverride=true
    simulationSpace=BoneRelative
    alphaBoolBlend {blendOption=Linear}
    boundBone {boneName=dyn_sleeves_r_MSTR}
    componentAppliedLinearAccClamp=<3000 3000 3000>
    componentLinearAccScale {y=0.05}
    componentLinearVelScale {y=0.1}
    gravityOverride=<300 500 500>
    physicsBodyDefinitions [{
      sphereCollisionRadius=0
      boundBone {boneName=dyn_sleeves_r_MSTR}
      boxExtents=<15 20 20>
      constraintSetup {angularLimitsMax=<0 8 30>, angularLimitsMin=<0 -2 -30>, angularTarget=<0 0 0>}
      localJointOffset=<5 0 2>
     }]
    relativeSpaceBone {boneName=lowerarm_r}
   }
   n22 = Anim Dynamics - Bone: dyn_sleeves_r_up!(ComponentPose=@n21, Alpha=1, GravityScale:float=1, ExternalForce="0,0,0") -> (Pose) {
    angularDampingOverride=0.999
    angularSpringConstant=500
    bGravityOverrideInSimSpace=true
    bOverrideAngularDamping=true
    bUseGravityOverride=true
    simulationSpace=BoneRelative
    alphaBoolBlend {blendOption=Linear}
    boundBone {boneName=dyn_sleeves_r_up}
    componentAppliedLinearAccClamp=<3000 3000 3000>
    componentLinearAccScale=<0.2 0.2 0.2>
    componentLinearVelScale=<0.2 0.2 0.2>
    gravityOverride {y=600}
    physicsBodyDefinitions [{
      sphereCollisionRadius=0
      boundBone {boneName=dyn_sleeves_r_up}
      boxExtents=<15 20 20>
      constraintSetup {angularLimitsMax=<0 10 2>, angularLimitsMin=<0 -5 -2>, angularTarget=<0 0 0>}
      localJointOffset=<4 0 0>
     }]
    relativeSpaceBone {boneName=lowerarm_r}
   }
   n23 = Anim Dynamics - Bone: dyn_sleeves_r_dn!(ComponentPose=@n22, Alpha=1, GravityScale:float=1, ExternalForce="0,0,0") -> (Pose) {
    angularDampingOverride=0.99
    angularSpringConstant=1000
    bAngularSpring=true
    bGravityOverrideInSimSpace=true
    bOverrideAngularDamping=true
    bUseGravityOverride=true
    simulationSpace=RootRelative
    alphaBoolBlend {blendOption=Linear}
    boundBone {boneName=dyn_sleeves_r_dn}
    componentAppliedLinearAccClamp=<3000 3000 3000>
    componentLinearAccScale=<0.2 0.2 0.2>
    componentLinearVelScale=<0.3 0.3 0.3>
    gravityOverride {z=-800}
    physicsBodyDefinitions [{
      sphereCollisionRadius=0
      boundBone {boneName=dyn_sleeves_r_dn}
      boxExtents=<20 20 10>
      constraintSetup {angularLimitsMax=<0 5 0>, angularLimitsMin=<0 -8 0>, angularTarget=<0 0 0>}
      localJointOffset=<5 0 2>
     }]
    relativeSpaceBone {boneName=lowerarm_r}
   }
   n24 = Anim Dynamics - Bone: dyn_sleeves_r_ft!(ComponentPose=@n23, Alpha=1, GravityScale:float=1, ExternalForce="0,0,0") -> (Pose) {
    angularDampingOverride=1
    angularSpringConstant=500
    bGravityOverrideInSimSpace=true
    bOverrideAngularDamping=true
    bUseGravityOverride=true
    simulationSpace=BoneRelative
    alphaBoolBlend {blendOption=Linear}
    boundBone {boneName=dyn_sleeves_r_ft}
    componentAppliedLinearAccClamp=<3000 3000 3000>
    componentLinearAccScale=<0.05 0.15 0.05>
    componentLinearVelScale=<0.3 0.3 0.3>
    gravityOverride=<-500 300>
    physicsBodyDefinitions [{
      sphereCollisionRadius=0
      boundBone {boneName=dyn_sleeves_r_ft}
      boxExtents=<10 20 20>
      constraintSetup {angularLimitsMax=<0 15 8>, angularLimitsMin=<0 -2 -8>, angularTarget=<0 0 0>}
      localJointOffset=<5 0 2>
     }]
    relativeSpaceBone {boneName=lowerarm_r}
   }
   n25 = Anim Dynamics - Bone: dyn_sleeves_r_bk!(ComponentPose=@n24, Alpha=1, GravityScale:float=1, ExternalForce="0,0,0") -> (Pose) {
    angularDampingOverride=1
    angularSpringConstant=500
    bGravityOverrideInSimSpace=true
    bOverrideAngularDamping=true
    bUseGravityOverride=true
    simulationSpace=BoneRelative
    alphaBoolBlend {blendOption=Linear}
    boundBone {boneName=dyn_sleeves_r_bk}
    componentAppliedLinearAccClamp=<3000 3000 3000>
    componentLinearAccScale=<0.05 0.15 0.05>
    componentLinearVelScale=<0.3 0.3 0.3>
    gravityOverride=<-500 300>
    physicsBodyDefinitions [{
      sphereCollisionRadius=0
      boundBone {boneName=dyn_sleeves_r_bk}
      boxExtents=<10 20 20>
      constraintSetup {angularLimitsMax=<0 15 8>, angularLimitsMin=<0 -2 -8>, angularTarget=<0 0 0>}
      localJointOffset=<5 -1 2>
     }]
    relativeSpaceBone {boneName=lowerarm_r}
   }
   n26 = Anim Dynamics - Bone: dyn_sleeves_l_MSTR!(ComponentPose=@n25, Alpha=1, GravityScale:float=1, ExternalForce="0,0,0") -> (Pose) {
    angularDampingOverride=1
    angularSpringConstant=500
    bGravityOverrideInSimSpace=true
    bOverrideAngularDamping=true
    bUseGravityOverride=true
    simulationSpace=BoneRelative
    alphaBoolBlend {blendOption=Linear}
    boundBone {boneName=dyn_sleeves_l_MSTR}
    componentAppliedLinearAccClamp=<3000 3000 3000>
    componentLinearAccScale {y=0.05}
    componentLinearVelScale {y=0.1}
    gravityOverride=<-300 -500 -300>
    physicsBodyDefinitions [{
      sphereCollisionRadius=0
      boundBone {boneName=dyn_sleeves_l_MSTR}
      boxExtents=<15 20 20>
      constraintSetup {angularLimitsMax=<0 12 30>, angularLimitsMin=<0 -2 -30>, angularTarget=<0 0 0>}
      localJointOffset=<5 0 0>
     }]
    relativeSpaceBone {boneName=lowerarm_l}
   }
   n27 = Anim Dynamics - Bone: dyn_sleeves_l_up!(ComponentPose=@n26, Alpha=1, GravityScale:float=1, ExternalForce="0,0,0") -> (Pose) {
    angularDampingOverride=0.999
    angularSpringConstant=500
    bGravityOverrideInSimSpace=true
    bOverrideAngularDamping=true
    bUseGravityOverride=true
    simulationSpace=BoneRelative
    alphaBoolBlend {blendOption=Linear}
    boundBone {boneName=dyn_sleeves_l_up}
    componentAppliedLinearAccClamp=<3000 3000 3000>
    componentLinearAccScale=<0.2 0.2 0.2>
    componentLinearVelScale=<0.2 0.2 0.2>
    gravityOverride {y=-500, z=100}
    physicsBodyDefinitions [{
      sphereCollisionRadius=0
      boundBone {boneName=dyn_sleeves_l_up}
      boxExtents=<15 20 20>
      constraintSetup {angularLimitsMax=<0 10 2>, angularLimitsMin=<0 -5 -2>, angularTarget=<0 0 0>}
      localJointOffset=<5 0 0>
     }]
    relativeSpaceBone {boneName=lowerarm_l}
   }
   n28 = Anim Dynamics - Bone: dyn_sleeves_l_dn!(ComponentPose=@n27, Alpha=1, GravityScale:float=1, ExternalForce="0,0,0") -> (Pose) {
    angularDampingOverride=0.99
    angularSpringConstant=1000
    bAngularSpring=true
    bGravityOverrideInSimSpace=true
    bOverrideAngularDamping=true
    bUseGravityOverride=true
    simulationSpace=RootRelative
    alphaBoolBlend {blendOption=Linear}
    boundBone {boneName=dyn_sleeves_l_dn}
    componentAppliedLinearAccClamp=<3000 3000 3000>
    componentLinearAccScale=<0.3 0.3 0.3>
    componentLinearVelScale=<0.3 0.3 0.3>
    gravityOverride {z=-500}
    physicsBodyDefinitions [{
      sphereCollisionRadius=0
      boundBone {boneName=dyn_sleeves_l_dn}
      boxExtents=<20 20 10>
      constraintSetup {angularLimitsMax=<0 5 0>, angularLimitsMin=<0 -8 0>, angularTarget=<0 0 0>}
      localJointOffset=<5 0 2>
     }]
    relativeSpaceBone {boneName=lowerarm_r}
   }
   n29 = Anim Dynamics - Bone: dyn_sleeves_l_ft!(ComponentPose=@n28, Alpha=1, GravityScale:float=1, ExternalForce="0,0,0") -> (Pose) {
    angularDampingOverride=1
    angularSpringConstant=500
    bGravityOverrideInSimSpace=true
    bOverrideAngularDamping=true
    bUseGravityOverride=true
    simulationSpace=BoneRelative
    alphaBoolBlend {blendOption=Linear}
    boundBone {boneName=dyn_sleeves_l_ft}
    componentAppliedLinearAccClamp=<3000 3000 3000>
    componentLinearAccScale=<0.05 0.15 0.05>
    componentLinearVelScale=<0.3 0.3 0.3>
    gravityOverride=<-500 -300>
    physicsBodyDefinitions [{
      sphereCollisionRadius=0
      boundBone {boneName=dyn_sleeves_l_ft}
      boxExtents=<10 20 20>
      constraintSetup {angularLimitsMax=<0 15 10>, angularLimitsMin=<0 5 -10>, angularTarget=<0 0 0>}
      localJointOffset=<5 1 2>
     }]
    relativeSpaceBone {boneName=lowerarm_l}
   }
   n30 = Anim Dynamics - Bone: dyn_sleeves_l_bk!(ComponentPose=@n29, Alpha=1, GravityScale:float=1, ExternalForce="0,0,0") -> (Pose) {
    angularDampingOverride=1
    angularSpringConstant=500
    bGravityOverrideInSimSpace=true
    bOverrideAngularDamping=true
    bUseGravityOverride=true
    simulationSpace=BoneRelative
    alphaBoolBlend {blendOption=Linear}
    boundBone {boneName=dyn_sleeves_l_bk}
    componentAppliedLinearAccClamp=<3000 3000 3000>
    componentLinearAccScale=<0.05 0.15 0.05>
    componentLinearVelScale=<0.3 0.3 0.3>
    gravityOverride=<-500 -300>
    physicsBodyDefinitions [{
      sphereCollisionRadius=0
      boundBone {boneName=dyn_sleeves_l_bk}
      boxExtents=<10 20 20>
      constraintSetup {angularLimitsMax=<0 15 8>, angularLimitsMin=<0 5 -8>, angularTarget=<0 0 0>}
      localJointOffset=<5 -1 2>
     }]
    relativeSpaceBone {boneName=lowerarm_l}
   }
   n31 = Look At - Bone: def_vest_l(ComponentPose=@n30, Alpha=1) {
    alphaBoolBlend {blendOption=Linear}
    boneToModify {boneName=def_vest_l}
    lookAtLocation=<0 16.852 0.5626>
    lookAtTarget {boneReference {boneName=calf_l}}
    lookAt_Axis {axis=<1 0>}
   }
   n32 = "Transform (Modify) Bone - Bone: def_vest_l"(ComponentPose=@n31, Alpha=1, Translation="0,0,0", Rotation="10,0,10", Scale="0.7,1,1") {
    rotationMode=BMM_Additive
    rotationSpace=BCS_BoneSpace
    scaleMode=BMM_Additive
    scaleSpace=BCS_BoneSpace
    translationMode=BMM_Additive
    translationSpace=BCS_BoneSpace
    alphaBoolBlend {blendOption=Linear}
    boneToModify {boneName=def_vest_l}
    rotation {pitch=10, roll=10}
    scale {x=0.7}
   }
   n33 = "Transform (Modify) Bone - Bone: def_vest_r"(ComponentPose=@n32, Alpha=1, Translation="2.121,1.400,0.576", Rotation="75.000,0.000,-0.000", Scale="1,1,1") {
    rotationMode=BMM_Additive
    rotationSpace=BCS_BoneSpace
    translationMode=BMM_Additive
    translationSpace=BCS_BoneSpace
    alphaBoolBlend {blendOption=Linear}
    boneToModify {boneName=def_vest_r}
    rotation=<rot 75.0001 0 0>
    translation=<2.121 1.4 0.576>
   }
   n34 = "Constraint - dyn_belt_MSTR (PPPPP)"!(
    ComponentPose=@n33
    Alpha=1
    ConstraintWeights_0=0.1426
    ConstraintWeights_1:float=0.0485
    ConstraintWeights_2:float=0.0218
    ConstraintWeights_3:float=0.031
    ConstraintWeights_4:float=0.756
   ) -> (Pose) {
    alphaBoolBlend {blendOption=Linear}
    boneToModify {boneName=dyn_belt_MSTR}
    constraintSetup [
     {targetBone {boneName=spine_01}}
     {targetBone {boneName=spine_02}}
     {targetBone {boneName=pelvis}}
     {targetBone {boneName=spine_03}}
     {targetBone {boneName=def_vest_l}}
    ]
    constraintWeights [1,1,1,1,1]
   }
   n35 = "Transform (Modify) Bone - Bone: dyn_belt_01"(ComponentPose=@n34, Alpha=1, Translation="0,0,1", Rotation="35,15,0", Scale="1,1,1") {
    rotationMode=BMM_Additive
    rotationSpace=BCS_BoneSpace
    scaleMode=BMM_Additive
    scaleSpace=BCS_BoneSpace
    translationMode=BMM_Additive
    translationSpace=BCS_BoneSpace
    alphaBoolBlend {blendOption=Linear}
    boneToModify {boneName=dyn_belt_01}
    rotation {pitch=35, yaw=15}
    translation {z=1}
   }
   n36 = "Anim Dynamics - Chain: dyn_belt_01 -> dyn_belt_02"!(ComponentPose=@n35, Alpha=1, GravityScale:float=1, ExternalForce="0,0,0") -> (Pose) {
    angularDampingOverride=0.99
    angularSpringConstant=1000
    bChain=true
    bGravityOverrideInSimSpace=true
    bOverrideAngularDamping=true
    bUseGravityOverride=true
    simulationSpace=RootRelative
    alphaBoolBlend {blendOption=Linear}
    boundBone {boneName=dyn_belt_01}
    chainEnd {boneName=dyn_belt_02}
    componentAppliedLinearAccClamp=<3000 3000 3000>
    componentLinearAccScale=<0.2 0.2 0.2>
    componentLinearVelScale=<0.3 0.3 0.3>
    gravityOverride {z=-980}
    physicsBodyDefinitions [
     {
      sphereCollisionRadius=0
      boundBone {boneName=dyn_belt_01}
      boxExtents=<10 20 20>
      constraintSetup {angularLimitsMax=<0 0 10>, angularLimitsMin=<0 -30 -10>, angularTarget=<0 0 0>}
      localJointOffset=<2 0 0>
     }
     {
      sphereCollisionRadius=0
      boundBone {boneName=dyn_belt_02}
      boxExtents=<10 20 20>
      constraintSetup {angularLimitsMax=<0 0 10>, angularLimitsMin=<0 -30 -10>, angularTarget=<0 0 0>}
      localJointOffset=<5 0 0>
     }
    ]
    relativeSpaceBone {boneName=lowerarm_r}
   }
   Save cached pose 'prePose'(Pose=Component To Local(ComponentPose=@n36)) {cacheName=prePose}
   n37 = "Constraint - dyn_sleeves_l_MSTR (PP)"!(ComponentPose=Local To Component(LocalPose=Use cached pose 'prePose'()), Alpha=1, ConstraintWeights_0=0.7, ConstraintWeights_1:float=0.3) -> (Pose) {
    alphaBoolBlend {blendOption=Linear}
    boneToModify {boneName=dyn_sleeves_l_MSTR}
    constraintSetup [{targetBone {boneName=upperarm_twist_02_l}}, {targetBone {boneName=lowerarm_l}}]
    constraintWeights [1,1]
   }
   n38 = "Constraint - dyn_sleeves_l_MSTR (PP)"(ComponentPose=@n37, Alpha=1, ConstraintWeights_0=0, ConstraintWeights_1=1) {
    alphaBoolBlend {blendOption=Linear}
    boneToModify {boneName=dyn_sleeves_l_MSTR}
    constraintSetup [{targetBone {boneName=upperarm_twist_02_l}}, {targetBone {boneName=lowerarm_l}}]
    constraintWeights [1,1]
   }
   n39 = "Constraint - dyn_sleeves_r_MSTR (PP)"!(ComponentPose=@n38, Alpha=1, ConstraintWeights_0=0.7, ConstraintWeights_1:float=0.3) -> (Pose) {
    alphaBoolBlend {blendOption=Linear}
    boneToModify {boneName=dyn_sleeves_r_MSTR}
    constraintSetup [{targetBone {boneName=upperarm_twist_02_r}}, {targetBone {boneName=lowerarm_r}}]
    constraintWeights [1,1]
   }
   n40 = "Constraint - dyn_sleeves_r_MSTR (PP)"(ComponentPose=@n39, Alpha=1, ConstraintWeights_0=0, ConstraintWeights_1=1) {
    alphaBoolBlend {blendOption=Linear}
    boneToModify {boneName=dyn_sleeves_r_MSTR}
    constraintSetup [{targetBone {boneName=upperarm_twist_02_r}}, {targetBone {boneName=lowerarm_r}}]
    constraintWeights [1,1]
   }
   n41 = Look At - Bone: dyn_jacket_flaps_MSTR(ComponentPose=@n40, Alpha=1) {
    alphaBoolBlend {blendOption=Linear}
    boneToModify {boneName=dyn_jacket_flaps_MSTR}
    lookAtLocation {x=0}
    lookAtTarget {boneReference {boneName=dyn_jacket_flaps_TGT}}
    lookAt_Axis {axis=<1 0>}
   }
   n42 = "Constraint - dyn_scarf_collision (PP)"!(ComponentPose=@n41, Alpha=1, ConstraintWeights_0=0.5, ConstraintWeights_1:float=0.5) -> (Pose) {
    alphaBoolBlend {blendOption=Linear}
    boneToModify {boneName=dyn_scarf_collision}
    constraintSetup [{targetBone {boneName=dyn_jacket_l_bk_02}}, {targetBone {boneName=dyn_jacket_r_bk_02}}]
    constraintWeights [1,1]
   }
   n43 = "Constraint - dyn_sleeves_l_MSTR (PP)"(ComponentPose=Local To Component(LocalPose=Use cached pose 'prePose'()), Alpha=1, ConstraintWeights_0=0.7, ConstraintWeights_1=0.3) {
    alphaBoolBlend {blendOption=Linear}
    boneToModify {boneName=dyn_sleeves_l_MSTR}
    constraintSetup [{targetBone {boneName=upperarm_twist_02_l}}, {targetBone {boneName=lowerarm_l}}]
    constraintWeights [1,1]
   }
   n44 = "Constraint - dyn_sleeves_l_MSTR (PP)"(ComponentPose=@n43, Alpha=1, ConstraintWeights_0=0, ConstraintWeights_1=1) {
    alphaBoolBlend {blendOption=Linear}
    boneToModify {boneName=dyn_sleeves_l_MSTR}
    constraintSetup [{targetBone {boneName=upperarm_twist_02_l}}, {targetBone {boneName=lowerarm_l}}]
    constraintWeights [1,1]
   }
   n45 = "Constraint - dyn_sleeves_r_MSTR (PP)"(ComponentPose=@n44, Alpha=1, ConstraintWeights_0=0.7, ConstraintWeights_1=0.3) {
    alphaBoolBlend {blendOption=Linear}
    boneToModify {boneName=dyn_sleeves_r_MSTR}
    constraintSetup [{targetBone {boneName=upperarm_twist_02_r}}, {targetBone {boneName=lowerarm_r}}]
    constraintWeights [1,1]
   }
   n46 = "Constraint - dyn_sleeves_r_MSTR (PP)"(ComponentPose=@n45, Alpha=1, ConstraintWeights_0=0, ConstraintWeights_1=1) {
    alphaBoolBlend {blendOption=Linear}
    boneToModify {boneName=dyn_sleeves_r_MSTR}
    constraintSetup [{targetBone {boneName=upperarm_twist_02_r}}, {targetBone {boneName=lowerarm_r}}]
    constraintWeights [1,1]
   }
   n47 = Look At - Bone: dyn_jacket_flaps_MSTR(ComponentPose=@n46, Alpha=1) {
    alphaBoolBlend {blendOption=Linear}
    boneToModify {boneName=dyn_jacket_flaps_MSTR}
    lookAtLocation {x=0}
    lookAtTarget {boneReference {boneName=dyn_jacket_flaps_TGT}}
    lookAt_Axis {axis=<1 0>}
   }
   n48 = "Transform (Modify) Bone - Bone: dyn_jacket_l_fr_MSTR"(ComponentPose=@n47, Alpha=1, Translation="-3.000,-12.000,0.000", Rotation="25,0,-70", Scale="1,1,1") {
    rotationMode=BMM_Additive
    rotationSpace=BCS_BoneSpace
    translationMode=BMM_Additive
    translationSpace=BCS_BoneSpace
    alphaBoolBlend {blendOption=Linear}
    boneToModify {boneName=dyn_jacket_l_fr_MSTR}
    rotation {pitch=25, roll=-70}
    translation=<-3 -12>
   }
   n49 = "Transform (Modify) Bone - Bone: dyn_jacket_l_MSTR"(ComponentPose=@n48, Alpha=1, Translation="0,-4,2", Rotation="20,0,-20", Scale="1,1,1") {
    rotationMode=BMM_Additive
    rotationSpace=BCS_BoneSpace
    translationMode=BMM_Additive
    translationSpace=BCS_BoneSpace
    alphaBoolBlend {blendOption=Linear}
    boneToModify {boneName=dyn_jacket_l_MSTR}
    rotation {pitch=20, roll=-20}
    translation {y=-4, z=2}
   }
   n50 = "Transform (Modify) Bone - Bone: dyn_jacket_r_fr_MSTR"(ComponentPose=@n49, Alpha=1, Translation="-4.545,3.355,-6.308", Rotation="20,-15,40", Scale="1,1,1") {
    rotationMode=BMM_Additive
    rotationSpace=BCS_BoneSpace
    translationMode=BMM_Additive
    translationSpace=BCS_BoneSpace
    alphaBoolBlend {blendOption=Linear}
    boneToModify {boneName=dyn_jacket_r_fr_MSTR}
    rotation=<rot 20 -15 40>
    translation=<-4.545 3.355 -6.308>
   }
   n51 = "Transform (Modify) Bone - Bone: dyn_jacket_r_fr_MSTR"(ComponentPose=@n50, Alpha=1, Translation="0,4.5,0", Rotation="0,20,45", Scale="1,1,1") {
    rotationMode=BMM_Additive
    rotationSpace=BCS_BoneSpace
    translationMode=BMM_Additive
    translationSpace=BCS_BoneSpace
    alphaBoolBlend {blendOption=Linear}
    alphaScaleBiasClamp {bInterpResult=true, interpSpeedDecreasing=5, interpSpeedIncreasing=20}
    boneToModify {boneName=dyn_jacket_r_fr_MSTR}
    rotation {roll=45, yaw=20}
    translation {y=4.5}
   }
   n52 = "Transform (Modify) Bone - Bone: dyn_jacket_r_fr_02"(ComponentPose=@n51, Alpha=1, Translation="0,0,6", Rotation="0,0,0", Scale="1,1,1") {
    rotationMode=BMM_Additive
    rotationSpace=BCS_BoneSpace
    translationMode=BMM_Additive
    translationSpace=BCS_BoneSpace
    alphaBoolBlend {blendOption=Linear}
    alphaScaleBiasClamp {bInterpResult=true, interpSpeedDecreasing=5, interpSpeedIncreasing=20}
    boneToModify {boneName=dyn_jacket_r_fr_02}
    translation {z=6}
   }
   n53 = "Transform (Modify) Bone - Bone: dyn_jacket_r_fr_03"(ComponentPose=@n52, Alpha=1, Translation="0,0,6", Rotation="-10,0,0", Scale="1,1,1") {
    rotationMode=BMM_Additive
    rotationSpace=BCS_BoneSpace
    translationMode=BMM_Additive
    translationSpace=BCS_BoneSpace
    alphaBoolBlend {blendOption=Linear}
    alphaScaleBiasClamp {bInterpResult=true, interpSpeedDecreasing=5, interpSpeedIncreasing=20}
    boneToModify {boneName=dyn_jacket_r_fr_03}
    rotation {pitch=-10}
    translation {z=6}
   }
   n54 = "Transform (Modify) Bone - Bone: dyn_jacket_r_02"(ComponentPose=@n53, Alpha=0, Translation="0,0,5", Rotation="0,0,0", Scale="1,1,1") {
    rotationMode=BMM_Additive
    rotationSpace=BCS_BoneSpace
    translationMode=BMM_Additive
    translationSpace=BCS_BoneSpace
    alphaBoolBlend {blendOption=Linear}
    alphaScaleBiasClamp {bInterpResult=true, interpSpeedDecreasing=5, interpSpeedIncreasing=20}
    boneToModify {boneName=dyn_jacket_r_02}
    translation {z=5}
   }
   n55 = RigidBody!(ComponentPose=@n54, Alpha, OverrideWorldGravity:Vector=Get GravityOverride(), ExternalForce) -> (Pose) {
    bClampLinearTranslationLimitToRefPose=true
    bOverrideWorldGravity=true
    overridePhysicsAsset=&/M_MED_Apprentice_Dyn_Jacket
    simulationSpace=BaseBoneSpace
    alphaBoolBlend {blendOption=Linear}
    baseBoneRef {boneName=spine_01}
    componentAppliedLinearAccClamp=<3000 1000 1000>
    componentLinearAccScale {x=0.1}
    componentLinearVelScale {x=0.2}
    overrideWorldGravity {z=-980}
    simSpaceSettings {maxAngularAcceleration=50, maxAngularVelocity=5, maxLinearAcceleration=500, maxLinearVelocity=100, velocityScaleZ=0, worldAlpha=1}
   }
   n56 = "Constraint - dyn_scarf_collision (PP)"(ComponentPose=@n55, Alpha=1, ConstraintWeights_0=0.5, ConstraintWeights_1=0.5) {
    alphaBoolBlend {blendOption=Linear}
    boneToModify {boneName=dyn_scarf_collision}
    constraintSetup [{targetBone {boneName=dyn_jacket_l_bk_02}}, {targetBone {boneName=dyn_jacket_r_bk_02}}]
    constraintWeights [1,1]
   }
   Save cached pose 'baseJacketRigidPose'(Pose=Blend Poses by bool(
     BlendPose_0=Component To Local(ComponentPose=@n42)
     BlendPose_1=Component To Local(ComponentPose=@n56)
     BlendTime_0=0.5
     BlendTime_1=0.5
     bActiveValue=Get isCustomEmotePlaying_0()
    ) {blendType=QuarticInOut}) {cacheName=baseJacketRigidPose}
   n57 = RigidBody(ComponentPose=Local To Component(LocalPose=Use cached pose 'baseJacketRigidPose'()), OverrideWorldGravity="0,0,-500") {
    bClampLinearTranslationLimitToRefPose=true
    bOverrideWorldGravity=true
    overridePhysicsAsset=&/M_MED_Apprentice_Dyn_Jacket_Flap
    simulationSpace=BaseBoneSpace
    alphaBoolBlend {blendOption=Linear}
    baseBoneRef {boneName=spine_03}
    componentAppliedLinearAccClamp=<3000 3000 3000>
    componentLinearAccScale=<0.3 0.3 0.5>
    componentLinearVelScale=<0.3 0.3 0.3>
    overrideWorldGravity {z=-980}
    simSpaceSettings {maxAngularAcceleration=100, maxAngularVelocity=5, maxLinearAcceleration=500, maxLinearVelocity=100}
   }
   n58 = "Constraint - dyn_scarf_collision (PP)"(ComponentPose=Local To Component(LocalPose=Use cached pose 'prePose'()), Alpha=1, ConstraintWeights_0=0.5, ConstraintWeights_1=0.5) {
    alphaBoolBlend {blendOption=Linear}
    boneToModify {boneName=dyn_scarf_collision}
    constraintSetup [{targetBone {boneName=calf_l}}, {targetBone {boneName=calf_r}}]
    constraintWeights [1,1]
   }
   n59 = RigidBody(ComponentPose=@n58, OverrideWorldGravity=Get GravityOverride()) {
    bClampLinearTranslationLimitToRefPose=true
    bOverrideWorldGravity=true
    overridePhysicsAsset=&/M_MED_Apprentice_Dyn_Jacket_Floating
    simulationSpace=BaseBoneSpace
    alphaBoolBlend {blendOption=Linear}
    baseBoneRef {boneName=spine_01}
    componentAppliedLinearAccClamp=<3000 3000 3000>
    componentLinearAccScale=<0.2 0.2 0.2>
    componentLinearVelScale=<0.2 0.2 0.2>
    overrideWorldGravity {z=-980}
    simSpaceSettings {maxAngularAcceleration=100, maxAngularVelocity=5, maxLinearAcceleration=500, maxLinearVelocity=100, velocityScaleZ=0}
   }
   n60 = Blend Poses by bool(
    BlendPose_0=Component To Local(ComponentPose=@n59)
    BlendPose_1=Component To Local(ComponentPose=@n57)
    bActiveValue=Get has_hand_glider() -> (has_hand_glider:bool)
   ) {blendType=Linear}
   n61 = RigidBody(ComponentPose=Local To Component(LocalPose=Use cached pose 'baseJacketRigidPose'()), OverrideWorldGravity="0,0,-980") {
    bClampLinearTranslationLimitToRefPose=true
    bOverrideWorldGravity=true
    overridePhysicsAsset=&/M_MED_Apprentice_Dyn_Jacket_Flap_Spec_Glider
    simulationSpace=BaseBoneSpace
    alphaBoolBlend {blendOption=Linear}
    baseBoneRef {boneName=spine_03}
    componentAppliedLinearAccClamp=<1000 1000 1000>
    componentLinearAccScale {x=0.1, z=0.1}
    componentLinearVelScale=<0.1 0.1 0.1>
    overrideWorldGravity {z=-980}
    simSpaceSettings {maxAngularAcceleration=100, maxAngularVelocity=5, maxLinearAcceleration=500, maxLinearVelocity=100}
   }
   n62 = Component To Local(ComponentPose=@n61)
   n63 = Get isApprenticeGlider() -> (isApprenticeGlider:bool)
   n64 = RigidBody(ComponentPose=Local To Component(LocalPose=Use cached pose 'baseJacketRigidPose'()), OverrideWorldGravity="0,0,-500") {
    bClampLinearTranslationLimitToRefPose=true
    bOverrideWorldGravity=true
    overridePhysicsAsset=&M_MED_Apprentice_Dyn_Jacket_Flap
    simulationSpace=BaseBoneSpace
    alphaBoolBlend {blendOption=Linear}
    baseBoneRef {boneName=spine_03}
    componentAppliedLinearAccClamp=<3000 3000 3000>
    componentLinearAccScale=<0.3 0.3 0.5>
    componentLinearVelScale=<0.3 0.3 0.3>
    overrideWorldGravity {z=-980}
    simSpaceSettings {maxAngularAcceleration=100, maxAngularVelocity=5, maxLinearAcceleration=500, maxLinearVelocity=100}
   }
   n65 = "Transform (Modify) Bone - Bone: dyn_jacket_r_bk_02"!(ComponentPose=@n64, bAlphaBoolEnabled:bool=True, Translation="0,0,0", Rotation="10,0,0", Scale="1,1,1") -> (Pose) {
    alphaInputType=Bool
    rotationMode=BMM_Additive
    rotationSpace=BCS_BoneSpace
    translationMode=BMM_Additive
    translationSpace=BCS_BoneSpace
    alphaBoolBlend {blendInTime=0.3, blendOption=Linear}
    boneToModify {boneName=dyn_jacket_r_bk_02}
    rotation {pitch=10}
   }
   n66 = "Transform (Modify) Bone - Bone: dyn_jacket_l_bk_02"!(ComponentPose=@n65, bAlphaBoolEnabled:bool=True, Translation="0,0,0", Rotation="10,0,0", Scale="1,1,1") -> (Pose) {
    alphaInputType=Bool
    rotationMode=BMM_Additive
    rotationSpace=BCS_BoneSpace
    translationMode=BMM_Additive
    translationSpace=BCS_BoneSpace
    alphaBoolBlend {blendInTime=0.3, blendOption=Linear}
    boneToModify {boneName=dyn_jacket_l_bk_02}
    rotation {pitch=10}
   }
   n67 = "Constraint - dyn_scarf_collision (PP)"(ComponentPose=Local To Component(LocalPose=Use cached pose 'prePose'()), Alpha=1, ConstraintWeights_0=0.5, ConstraintWeights_1=0.5) {
    alphaBoolBlend {blendOption=Linear}
    boneToModify {boneName=dyn_scarf_collision}
    constraintSetup [{targetBone {boneName=calf_l}}, {targetBone {boneName=calf_r}}]
    constraintWeights [1,1]
   }
   n68 = RigidBody(ComponentPose=@n67, OverrideWorldGravity=Get GravityOverride()) {
    bClampLinearTranslationLimitToRefPose=true
    bOverrideWorldGravity=true
    overridePhysicsAsset=&M_MED_Apprentice_Dyn_Jacket_Floating
    simulationSpace=BaseBoneSpace
    alphaBoolBlend {blendOption=Linear}
    baseBoneRef {boneName=spine_01}
    componentAppliedLinearAccClamp=<1000 1000 1000>
    componentLinearAccScale=<0.2 0.2 0.2>
    componentLinearVelScale=<0.2 0.2 0.2>
    overrideWorldGravity {z=-980}
    simSpaceSettings {maxAngularAcceleration=100, maxAngularVelocity=5, maxLinearAcceleration=500, maxLinearVelocity=100, velocityScaleZ=0}
   }
   n69 = "Transform (Modify) Bone - Bone: dyn_jacket_flaps_MSTR"(ComponentPose=@n68, Alpha=1, Translation="0,0,3", Rotation="15,0,0", Scale="1,1,1") {
    rotationMode=BMM_Additive
    rotationSpace=BCS_BoneSpace
    translationMode=BMM_Additive
    translationSpace=BCS_BoneSpace
    alphaBoolBlend {blendOption=Linear}
    boneToModify {boneName=dyn_jacket_flaps_MSTR}
    rotation {pitch=15}
    translation {z=3}
   }
   n70 = Blend Poses by bool(
    BlendPose_0=Component To Local(ComponentPose=@n69)
    BlendPose_1=Blend Poses by bool(
     BlendPose_0=Blend Poses by bool(BlendPose_0=@n62, BlendPose_1=@n60, bActiveValue=@n63) {blendType=Linear}
     BlendPose_1=Component To Local(ComponentPose=@n66)
     bActiveValue=Get bIsParachuteOpened()
    ) {blendType=Linear}
    bActiveValue=Get isFloatingOrDiving() -> (isFloatingOrDiving:bool)
   ) {blendType=Linear}
   n71 = "Transform (Modify) Bone - Bone: dyn_jacket_l_fr_MSTR"(
    ComponentPose=Local To Component(LocalPose=Use cached pose 'prePose'())
    Alpha=1
    Translation="1.296,-2.849,0.854"
    Rotation="-10,-20,-75"
    Scale="1,1,1"
   ) {
    rotationMode=BMM_Additive
    rotationSpace=BCS_BoneSpace
    translationMode=BMM_Additive
    translationSpace=BCS_BoneSpace
    alphaBoolBlend {blendOption=Linear}
    boneToModify {boneName=dyn_jacket_l_fr_MSTR}
    rotation=<rot -10 -20 -75>
    translation=<1.296 -2.849 0.854>
   }
   n72 = "Transform (Modify) Bone - Bone: dyn_jacket_l_MSTR"(ComponentPose=@n71, Alpha=1, Translation="-5,-6,2", Rotation="10.028,-23.782,-26.270", Scale="1,1,1") {
    rotationMode=BMM_Additive
    rotationSpace=BCS_BoneSpace
    translationMode=BMM_Additive
    translationSpace=BCS_BoneSpace
    alphaBoolBlend {blendOption=Linear}
    boneToModify {boneName=dyn_jacket_l_MSTR}
    rotation=<rot 10.028 -23.7819 -26.27>
    translation=<-5 -6 2>
   }
   n73 = "Transform (Modify) Bone - Bone: dyn_jacket_l_bk_MSTR"(ComponentPose=@n72, Alpha=1, Translation="0,0,2", Rotation="60,-10,0", Scale="1,1,1") {
    rotationMode=BMM_Additive
    rotationSpace=BCS_BoneSpace
    translationMode=BMM_Additive
    translationSpace=BCS_BoneSpace
    alphaBoolBlend {blendOption=Linear}
    boneToModify {boneName=dyn_jacket_l_bk_MSTR}
    rotation {pitch=60, yaw=-10}
    translation {z=2}
   }
   n74 = "Transform (Modify) Bone - Bone: dyn_jacket_r_fr_MSTR"(ComponentPose=@n73, Alpha=1, Translation="2,0,0", Rotation="-10,0,80", Scale="1,1,1") {
    rotationMode=BMM_Additive
    rotationSpace=BCS_BoneSpace
    translationMode=BMM_Additive
    translationSpace=BCS_BoneSpace
    alphaBoolBlend {blendOption=Linear}
    boneToModify {boneName=dyn_jacket_r_fr_MSTR}
    rotation {pitch=-10, roll=80}
    translation {x=2}
   }
   n75 = "Transform (Modify) Bone - Bone: dyn_jacket_r_MSTR"(ComponentPose=@n74, Alpha=1, Translation="-4.136,1.267,-0.666", Rotation="8.458,22.728,41.576", Scale="1,1,1") {
    rotationMode=BMM_Additive
    rotationSpace=BCS_BoneSpace
    translationMode=BMM_Additive
    translationSpace=BCS_BoneSpace
    alphaBoolBlend {blendOption=Linear}
    boneToModify {boneName=dyn_jacket_r_MSTR}
    rotation=<rot 8.458 22.7279 41.576>
    translation=<-4.136 1.267 -0.666>
   }
   n76 = "Transform (Modify) Bone - Bone: dyn_jacket_r_bk_MSTR"(ComponentPose=@n75, Alpha=1, Translation="0,0,2", Rotation="60,10,0", Scale="1,1,1") {
    rotationMode=BMM_Additive
    rotationSpace=BCS_BoneSpace
    translationMode=BMM_Additive
    translationSpace=BCS_BoneSpace
    alphaBoolBlend {blendOption=Linear}
    boneToModify {boneName=dyn_jacket_r_bk_MSTR}
    rotation {pitch=60, yaw=10}
    translation {z=2}
   }
   n77 = "Transform (Modify) Bone - Bone: dyn_jacket_flaps_01"(ComponentPose=@n76, Alpha=1, Translation="-5,0,5", Rotation="60,10,0", Scale="1,1,1") {
    rotationMode=BMM_Additive
    rotationSpace=BCS_BoneSpace
    translationMode=BMM_Additive
    translationSpace=BCS_BoneSpace
    alphaBoolBlend {blendOption=Linear}
    boneToModify {boneName=dyn_jacket_flaps_01}
    rotation {pitch=60, yaw=10}
    translation {x=-5, z=5}
   }
   n78 = RigidBody(ComponentPose=@n77, OverrideWorldGravity="0,0,-980") {
    bClampLinearTranslationLimitToRefPose=true
    bOverrideWorldGravity=true
    overridePhysicsAsset=&/M_MED_Apprentice_Dyn_Jacket_Sliding
    simulationSpace=BaseBoneSpace
    alphaBoolBlend {blendOption=Linear}
    baseBoneRef {boneName=spine_01}
    componentAppliedLinearAccClamp=<3000 1000 1000>
    componentLinearAccScale=<0.2 0.2 0.2>
    componentLinearVelScale=<0.3 0.3 0.3>
    overrideWorldGravity {z=-980}
    simSpaceSettings {maxAngularAcceleration=100, maxAngularVelocity=5, maxLinearAcceleration=500, maxLinearVelocity=100, velocityScaleZ=0}
   }
   n79 = Component To Local(ComponentPose=@n78)
   n80 = Get bIsSliding()
   n81 = Look At - Bone: dyn_jacket_r_bk_MSTR(ComponentPose=Local To Component(LocalPose=Use cached pose 'prePose'()), Alpha=1) {
    alphaBoolBlend {blendOption=Linear}
    boneToModify {boneName=dyn_jacket_r_bk_MSTR}
    lookAtLocation=<-15 -17.3 12>
    lookAtTarget {boneReference {boneName=pelvis}}
    lookAt_Axis {axis=<1 0>}
   }
   n82 = Look At - Bone: dyn_jacket_l_bk_MSTR(ComponentPose=@n81, Alpha=1) {
    alphaBoolBlend {blendOption=Linear}
    boneToModify {boneName=dyn_jacket_l_bk_MSTR}
    lookAtLocation=<-15 -17.3 -12>
    lookAtTarget {boneReference {boneName=pelvis}}
    lookAt_Axis {axis=<1 0>}
   }
   n83 = Look At - Bone: dyn_jacket_r_fr_MSTR(ComponentPose=@n82, Alpha=1) {
    alphaBoolBlend {blendOption=Linear}
    boneToModify {boneName=dyn_jacket_r_fr_MSTR}
    lookAtLocation=<-16.3811 10.4887 18.4092>
    lookAtTarget {boneReference {boneName=pelvis}}
    lookAt_Axis {axis=<1 0>}
   }
   n84 = Look At - Bone: dyn_jacket_l_fr_MSTR(ComponentPose=@n83, Alpha=1) {
    alphaBoolBlend {blendOption=Linear}
    boneToModify {boneName=dyn_jacket_l_fr_MSTR}
    lookAtLocation=<-16.381 10.4887 -18.4092>
    lookAtTarget {boneReference {boneName=pelvis}}
    lookAt_Axis {axis=<1 0>}
   }
   n85 = "Constraint - dyn_jacket_r_fr_02 (P)"(ComponentPose=@n84, Alpha=1, ConstraintWeights_0=1) {
    alphaBoolBlend {blendOption=Linear}
    boneToModify {boneName=dyn_jacket_r_fr_02}
    constraintSetup [{targetBone {boneName=thigh_twist_01_r}}]
    constraintWeights [1]
   }
   n86 = "Constraint - dyn_jacket_r_01 (P)"(ComponentPose=@n85, Alpha=1, ConstraintWeights_0=1) {
    alphaBoolBlend {blendOption=Linear}
    alphaScaleBiasClamp {outRange {max=0.3, min=1}}
    boneToModify {boneName=dyn_jacket_r_01}
    constraintSetup [{targetBone {boneName=thigh_twist_01_r}}]
    constraintWeights [1]
   }
   n87 = "Constraint - dyn_jacket_r_bk_02 (P)"(ComponentPose=@n86, Alpha=0.8, ConstraintWeights_0=1) {
    alphaBoolBlend {blendOption=Linear}
    alphaScaleBiasClamp {bMapRange=true, outRange {max=0.3, min=1}}
    boneToModify {boneName=dyn_jacket_r_bk_02}
    constraintSetup [{targetBone {boneName=thigh_twist_01_r}}]
    constraintWeights [1]
   }
   n88 = "Constraint - dyn_jacket_l_fr_02 (P)"(ComponentPose=@n87, Alpha=1, ConstraintWeights_0=1) {
    alphaBoolBlend {blendOption=Linear}
    boneToModify {boneName=dyn_jacket_l_fr_02}
    constraintSetup [{targetBone {boneName=thigh_twist_01_l}}]
    constraintWeights [1]
   }
   n89 = "Constraint - dyn_jacket_l_01 (P)"(ComponentPose=@n88, Alpha=1, ConstraintWeights_0=1) {
    alphaBoolBlend {blendOption=Linear}
    alphaScaleBiasClamp {outRange {max=0.3, min=1}}
    boneToModify {boneName=dyn_jacket_l_01}
    constraintSetup [{targetBone {boneName=thigh_twist_01_l}}]
    constraintWeights [1]
   }
   n90 = "Constraint - dyn_jacket_l_bk_02 (P)"(ComponentPose=@n89, Alpha=0.8, ConstraintWeights_0=1) {
    alphaBoolBlend {blendOption=Linear}
    alphaScaleBiasClamp {bMapRange=true, outRange {max=0.3, min=1}}
    boneToModify {boneName=dyn_jacket_l_bk_02}
    constraintSetup [{targetBone {boneName=thigh_twist_01_l}}]
    constraintWeights [1]
   }
   n91 = "Transform (Modify) Bone - Bone: dyn_jacket_r_fr_02"(ComponentPose=@n90, Alpha=1, Translation="20,0,-1", Rotation="-7,0,0", Scale="0.750,1.000,1.000") {
    rotationMode=BMM_Additive
    rotationSpace=BCS_BoneSpace
    scaleMode=BMM_Additive
    scaleSpace=BCS_BoneSpace
    translationMode=BMM_Additive
    translationSpace=BCS_BoneSpace
    alphaBoolBlend {blendOption=Linear}
    boneToModify {boneName=dyn_jacket_r_fr_02}
    rotation {pitch=-7}
    scale {x=0.75}
    translation {x=20, z=-1}
   }
   n92 = "Transform (Modify) Bone - Bone: dyn_jacket_r_01"(ComponentPose=@n91, Alpha=1, Translation="5.947,-6.472,0.650", Rotation="-0.81,25,5.723", Scale="0.750,1.000,1.000") {
    rotationMode=BMM_Additive
    rotationSpace=BCS_BoneSpace
    scaleMode=BMM_Additive
    scaleSpace=BCS_BoneSpace
    translationMode=BMM_Additive
    translationSpace=BCS_BoneSpace
    alphaBoolBlend {blendOption=Linear}
    boneToModify {boneName=dyn_jacket_r_01}
    rotation=<rot -0.81 25 5.723>
    scale {x=0.75}
    translation=<5.947 -6.472 0.65>
   }
   n93 = "Transform (Modify) Bone - Bone: dyn_jacket_r_bk_01"(ComponentPose=@n92, Alpha=1, Translation="0,0,0", Rotation="15,0,0", Scale="1,1,1") {
    rotationMode=BMM_Additive
    rotationSpace=BCS_BoneSpace
    scaleMode=BMM_Additive
    scaleSpace=BCS_BoneSpace
    translationMode=BMM_Additive
    translationSpace=BCS_BoneSpace
    alphaBoolBlend {blendOption=Linear}
    alphaScaleBiasClamp {interpSpeedIncreasing=30}
    boneToModify {boneName=dyn_jacket_r_bk_01}
    rotation {pitch=15}
   }
   n94 = "Transform (Modify) Bone - Bone: dyn_jacket_r_bk_02"!(ComponentPose=@n93, Alpha:float=1, Translation="-7,0,5", Rotation="65,0,0", Scale="1,1,1") -> (Pose) {
    rotationMode=BMM_Additive
    rotationSpace=BCS_BoneSpace
    scaleMode=BMM_Additive
    scaleSpace=BCS_BoneSpace
    translationMode=BMM_Additive
    translationSpace=BCS_BoneSpace
    alphaBoolBlend {blendOption=Linear}
    alphaScaleBiasClamp {interpSpeedIncreasing=30}
    boneToModify {boneName=dyn_jacket_r_bk_02}
    rotation {pitch=65}
    translation {x=-7, z=5}
   }
   n95 = "Transform (Modify) Bone - Bone: dyn_jacket_l_bk_02"!(ComponentPose=@n94, Alpha:float=1, Translation="0,0,0", Rotation="25,0,0", Scale="1,1,1") -> (Pose) {
    rotationMode=BMM_Additive
    rotationSpace=BCS_BoneSpace
    scaleMode=BMM_Additive
    scaleSpace=BCS_BoneSpace
    translationMode=BMM_Additive
    translationSpace=BCS_BoneSpace
    alphaBoolBlend {blendOption=Linear}
    alphaScaleBiasClamp {interpSpeedIncreasing=30}
    boneToModify {boneName=dyn_jacket_l_bk_02}
    rotation {pitch=25}
   }
   n96 = "Transform (Modify) Bone - Bone: dyn_jacket_l_fr_02"(ComponentPose=@n95, Alpha=1, Translation="20,0,-1", Rotation="0,0,0", Scale="0.75,1,1") {
    rotationMode=BMM_Additive
    rotationSpace=BCS_BoneSpace
    scaleMode=BMM_Additive
    scaleSpace=BCS_BoneSpace
    translationMode=BMM_Additive
    translationSpace=BCS_BoneSpace
    alphaBoolBlend {blendOption=Linear}
    boneToModify {boneName=dyn_jacket_l_fr_02}
    scale {x=0.75}
    translation {x=20, z=-1}
   }
   n97 = "Transform (Modify) Bone - Bone: dyn_jacket_l_01"(ComponentPose=@n96, Alpha=1, Translation="10.455,4.556,-0.034", Rotation="16.660,-44.354,-2.081", Scale="0.75,1,1") {
    rotationMode=BMM_Additive
    rotationSpace=BCS_BoneSpace
    scaleMode=BMM_Additive
    scaleSpace=BCS_BoneSpace
    translationMode=BMM_Additive
    translationSpace=BCS_BoneSpace
    alphaBoolBlend {blendOption=Linear}
    boneToModify {boneName=dyn_jacket_l_01}
    rotation=<rot 16.66 -44.354 -2.081>
    scale {x=0.75}
    translation=<10.455 4.556 -0.034>
   }
   n98 = "Transform (Modify) Bone - Bone: dyn_jacket_l_bk_01"(ComponentPose=@n97, Alpha=1, Translation="0,0,0", Rotation="15,0,0", Scale="1,1,1") {
    rotationMode=BMM_Additive
    rotationSpace=BCS_BoneSpace
    scaleMode=BMM_Additive
    scaleSpace=BCS_BoneSpace
    translationMode=BMM_Additive
    translationSpace=BCS_BoneSpace
    alphaBoolBlend {blendOption=Linear}
    alphaScaleBiasClamp {interpSpeedIncreasing=30}
    boneToModify {boneName=dyn_jacket_l_bk_01}
    rotation {pitch=15}
   }
   n99 = "Transform (Modify) Bone - Bone: dyn_jacket_l_bk_02"(ComponentPose=@n98, Alpha=1, Translation="-7,0,5", Rotation="65,0,0", Scale="1,1,1") {
    rotationMode=BMM_Additive
    rotationSpace=BCS_BoneSpace
    scaleMode=BMM_Additive
    scaleSpace=BCS_BoneSpace
    translationMode=BMM_Additive
    translationSpace=BCS_BoneSpace
    alphaBoolBlend {blendOption=Linear}
    alphaScaleBiasClamp {interpSpeedIncreasing=30}
    boneToModify {boneName=dyn_jacket_l_bk_02}
    rotation {pitch=65}
    translation {x=-7, z=5}
   }
   n100 = "Transform (Modify) Bone - Bone: dyn_jacket_r_bk_02"(ComponentPose=@n99, Alpha=1, Translation="0,0,0", Rotation="25,0,0", Scale="1,1,1") {
    rotationMode=BMM_Additive
    rotationSpace=BCS_BoneSpace
    scaleMode=BMM_Additive
    scaleSpace=BCS_BoneSpace
    translationMode=BMM_Additive
    translationSpace=BCS_BoneSpace
    alphaBoolBlend {blendOption=Linear}
    alphaScaleBiasClamp {interpSpeedIncreasing=30}
    boneToModify {boneName=dyn_jacket_r_bk_02}
    rotation {pitch=25}
   }
   n101 = "Transform (Modify) Bone - Bone: dyn_jacket_r_bk_02"!(ComponentPose=@n100, bAlphaBoolEnabled:bool=True, Translation="0,0,2", Rotation="15,0,0", Scale="1,1,1") -> (Pose) {
    alphaInputType=Bool
    rotationMode=BMM_Additive
    rotationSpace=BCS_BoneSpace
    translationMode=BMM_Additive
    translationSpace=BCS_BoneSpace
    alphaBoolBlend {blendInTime=0.3, blendOption=Linear}
    boneToModify {boneName=dyn_jacket_r_bk_02}
    rotation {pitch=15}
    translation {z=2}
   }
   n102 = "Transform (Modify) Bone - Bone: dyn_jacket_l_bk_02"!(ComponentPose=@n101, bAlphaBoolEnabled:bool=True, Translation="0,0,2", Rotation="15,0,0", Scale="1,1,1") -> (Pose) {
    alphaInputType=Bool
    rotationMode=BMM_Additive
    rotationSpace=BCS_BoneSpace
    translationMode=BMM_Additive
    translationSpace=BCS_BoneSpace
    alphaBoolBlend {blendInTime=0.3, blendOption=Linear}
    boneToModify {boneName=dyn_jacket_l_bk_02}
    rotation {pitch=15}
    translation {z=2}
   }
   n103 = "Transform (Modify) Bone - Bone: dyn_jacket_flaps_01"!(ComponentPose=@n102, bAlphaBoolEnabled:bool=True, Translation="0,0,2", Rotation="10,0,0", Scale="1,1,1") -> (Pose) {
    alphaInputType=Bool
    rotationMode=BMM_Additive
    rotationSpace=BCS_BoneSpace
    translationMode=BMM_Additive
    translationSpace=BCS_BoneSpace
    alphaBoolBlend {blendInTime=0.3, blendOption=Linear, blendOutTime=0.2}
    boneToModify {boneName=dyn_jacket_flaps_01}
    rotation {pitch=10}
    translation {z=2}
   }
   n104 = "Transform (Modify) Bone - Bone: dyn_jacket_r_bk_01"!(ComponentPose=@n103, bAlphaBoolEnabled:bool=True, Translation="0,0,0", Rotation="-15,0,0", Scale="1,1,1") -> (Pose) {
    alphaInputType=Bool
    rotationMode=BMM_Additive
    rotationSpace=BCS_BoneSpace
    translationMode=BMM_Additive
    translationSpace=BCS_BoneSpace
    alphaBoolBlend {blendInTime=0.3, blendOption=Linear}
    boneToModify {boneName=dyn_jacket_r_bk_01}
    rotation {pitch=-15}
   }
   n105 = "Transform (Modify) Bone - Bone: dyn_jacket_l_bk_01"!(ComponentPose=@n104, bAlphaBoolEnabled:bool=True, Translation="0,0,0", Rotation="-15,0,0", Scale="1,1,1") -> (Pose) {
    alphaInputType=Bool
    rotationMode=BMM_Additive
    rotationSpace=BCS_BoneSpace
    translationMode=BMM_Additive
    translationSpace=BCS_BoneSpace
    alphaBoolBlend {blendInTime=0.3, blendOption=Linear}
    boneToModify {boneName=dyn_jacket_l_bk_01}
    rotation {pitch=-15}
   }
   n106 = Look At - Bone: dyn_jacket_flaps_MSTR!(ComponentPose=@n105, bAlphaBoolEnabled:bool=True) -> (Pose) {
    alphaInputType=Bool
    alphaBoolBlend {blendOption=Linear}
    boneToModify {boneName=dyn_jacket_flaps_MSTR}
    lookAtLocation {x=0}
    lookAtTarget {boneReference {boneName=dyn_jacket_flaps_TGT}}
    lookAt_Axis {axis=<1 0>}
   }
   Save cached pose 'FinalSetup'(Pose=Blend Poses by bool(
     BlendPose_0=Blend Poses by bool(BlendPose_0=@n79, BlendPose_1=@n70, bActiveValue=@n80) {blendType=Linear}
     BlendPose_1=Component To Local(ComponentPose=@n106)
     bActiveValue=Get bRigidBodyEnabled()
    ) {blendType=Linear}) {cacheName=FinalSetup}
  ]}
 EventGraph {nodes [
   Event BlueprintUpdateAnimation()
   > n1 = Sequence!(execute) -> (then_0, then_1, then_2:Exec, then_3:Exec, then_4:Exec, then_5:Exec, then_6:Exec, then_7:Exec, then_8:Exec)
   > n2 = Cast To FortPlayerPawn(Object=GetOwningActor())
   > n3 = GetVariantChannelActiveIndexForItemActor!(
    execute
    Item:Athena Cosmetic Item Definition Object Reference=/BRCosmetics/Athena/Items/Cosmetics/Characters/Character_Apprentice
    Actor:Actor Object Reference=@n2
    Channel:Gameplay Tag Structure=(TagName="Cosmetics.Variant.Channel.Progressive")
    VariantList:Array of Gameplay Tag Structures=Make Array("[0]"=(TagName="Cosmetics.Variant.Property.Stage1"))
    bErrorIfChannelMissing:bool=false
    bErrorIfTagNotInArray:bool=false
   ) -> (then, Index:int=0, bIndexIsValid:bool, bFoundChannel:bool)
   > n4 = Branch(Condition=@n3.bIndexIsValid)
   > Set isLongHair(isLongHair:bool=true) -> (Output_Get:bool)
   Set isLongHair(execute=@n4.else, isLongHair=false)
   n5 = Get SkelMesh()
   n6 = Set thigh_rot_up_l(
    execute=@n1.then_1
    thigh_rot_up_l:double=GetAlphaFromReferenceTransformDelta(SkeletalMeshComponent=@n5, BoneName=thigh_l, BaseName=pelvis, DeltaComponentType=QuaternionTwist, InputRange=(X=0,Y=100), TwistAxis="0,0,1")
   ) -> (Output_Get:double)
   > n7 = Set thigh_rot_up_r(thigh_rot_up_r:double=GetAlphaFromReferenceTransformDelta(SkeletalMeshComponent=@n5, BoneName=thigh_r, BaseName=pelvis, DeltaComponentType=QuaternionTwist, InputRange=(X=0,Y=100), TwistAxis="0,0,1")) -> (Output_Get:double)
   > Set uppperarm_l_twist_alpha(uppperarm_l_twist_alpha:double=GetAlphaFromReferenceTransformDelta(
     SkeletalMeshComponent=@n5
     BoneName=upperarm_l
     BaseName=upperarm_twsit_02_l
     DeltaComponentType=QuaternionTwist
     InputRange=(X=-100,Y=-50)
     TwistAxis="1,0,0"
    )) -> (Output_Get:double)
   > Set knee_rot_r(knee_rot_r:double=float * float(
     A=MapRangeClamped(Value=@n7, InRangeA=0, InRangeB=1, OutRangeA=1, OutRangeB=0.3)
     B=GetAlphaFromReferenceTransformDelta(SkeletalMeshComponent=@n5, BoneName=calf_r, BaseName=thigh_r, DeltaComponentType=QuaternionTwist, InputRange=(X=0,Y=-120), TwistAxis="0,0,1")
    ) {is=title K2Node_PromotableOperator}) -> (Output_Get:double)
   > Set knee_rot_l(knee_rot_l:double=float * float(
     A=MapRangeClamped(Value=@n6, InRangeA=0, InRangeB=1, OutRangeA=1, OutRangeB=0.3)
     B=GetAlphaFromReferenceTransformDelta(SkeletalMeshComponent=@n5, BoneName=calf_l, BaseName=thigh_l, DeltaComponentType=QuaternionTwist, InputRange=(X=0,Y=-120), TwistAxis="0,0,1")
    )) -> (Output_Get:double)
   n8 = Get SkelMesh()
   > Set lowerarm_rot_up_l(lowerarm_rot_up_l:double=GetAlphaFromReferenceTransformDelta(SkeletalMeshComponent=@n8, BoneName=lowerarm_l, BaseName=upperarm_l, InputRange=(X=0,Y=50), TwistAxis="0,0,1")) -> (Output_Get:double)
   > Set lowerarm_rot_up_r(lowerarm_rot_up_r:double=GetAlphaFromReferenceTransformDelta(SkeletalMeshComponent=@n8, BoneName=lowerarm_r, BaseName=upperarm_r, InputRange=(X=0,Y=50), TwistAxis="0,0,1")) -> (Output_Get:double)
   n9 = Gravity Override State Function DEPRECATED(
    execute=@n1.then_2
    CharacterPartAnimInstance=Self-Reference()
    GameTimeInSeconds=GetGameTimeInSeconds()
    GravityOverrideParams=Get Gravity Override Params()
   )
   > Set GravityOverride(GravityOverride=@n9)
   n10 = PlayerMovement(execute=@n1.then_3, AnimInstance=Self-Reference(), XDotProductCoefficient=1, YDotProductCoefficient=1, ZDotProductCoefficient=1)
   n11 = Get bIsCrouching()
   > n12 = Branch(Condition=AND Boolean(A=Get bIsOnGround() -> (bIsOnGround:bool), B=NOT Boolean(A=@n11)))
   n13 = Get SkelMesh()
   > Set pullRightJacketOutAlpha(pullRightJacketOutAlpha:double=float * float(
     A=GetAlphaFromReferenceTransformDelta(SkeletalMeshComponent=@n13, BoneName=thigh_r, BaseName=pelvis, DeltaComponentType=QuaternionTwist, InputRange=(X=0,Y=15), TwistAxis="0,1,1")
     B=MapRangeClamped(Value=@n10, InRangeA=0, InRangeB=1)
    )) -> (Output_Get:double)
   Set pullRightJacketOutAlpha(execute=@n12.else, pullRightJacketOutAlpha=0)
   n14 = Branch(execute=@n1.then_4, Condition=Get weapon_visible() -> (weapon_visible:bool))
   n15 = GetDeltaTransformFromRefPose(self=GetOwnerSkeletalMeshComponent(), BoneName=upperarm_r, BaseName=spine_05)
   n16 = "Absolute (Float)"(A=BreakRotator(InRot=BreakTransform(InTransform=@n15).Rotation).Yaw)
   > Set rightLapelDefUpwardAlpha(rightLapelDefUpwardAlpha:double="Min (Float)"(
     A:double=MapRangeClamped(Value=@n16, InRangeA=35, InRangeB=100)
     B:double=MapRangeClamped(Value=@n16, InRangeA=120, InRangeB=180, OutRangeA=1, OutRangeB=0)
    ) -> (ReturnValue:double) {is=title K2Node_CommutativeAssociativeBinaryOperator}) -> (Output_Get:double)
   > Set rightLapelDefDnwardAlpha(rightLapelDefDnwardAlpha:double="Min (Float)"!(
     A=MapRangeClamped(Value=@n16, InRangeA=100, InRangeB=60)
     B=MapRangeClamped(Value=@n16, InRangeA=35, InRangeB=0, OutRangeA=1, OutRangeB=0)
     C:double=MapRangeClamped(Value=@n16, InRangeA=120, InRangeB=180, OutRangeA=1, OutRangeB=0)
    ) -> (ReturnValue)) -> (Output_Get:double)
   n17 = Get SkelMesh()
   > Set leftLapelDefUpwardAlpha(leftLapelDefUpwardAlpha:double=GetAlphaFromReferenceTransformDelta(
     SkeletalMeshComponent=@n17
     BoneName=upperarm_l
     BaseName=clavicle_l
     DeltaComponentType=QuaternionTwist
     InputRange=(X=-70,Y=-100)
     TwistAxis="1,0,0"
    )) -> (Output_Get:double)
   > Set leftLapelDefDnwardAlpha(leftLapelDefDnwardAlpha:double="Min (Float)"!(
     A=GetAlphaFromReferenceTransformDelta(
      SkeletalMeshComponent=@n17
      BoneName=upperarm_l
      BaseName=clavicle_l
      DeltaComponentType=QuaternionTwist
      InputRange=(X=-120,Y=-50)
      TwistAxis="1,0,0"
     )
     B=GetAlphaFromReferenceTransformDelta(
      SkeletalMeshComponent=@n17
      BoneName=upperarm_l
      BaseName=clavicle_l
      DeltaComponentType=QuaternionTwist
      InputRange=(X=-30,Y=0)
      OutputRange=(X=1,Y=0)
      TwistAxis="1,0,0"
     )
    ) -> (ReturnValue)) -> (Output_Get:double)
   > Set collar_def_alpha(collar_def_alpha:double=GetAlphaFromReferenceTransformDelta(
     SkeletalMeshComponent=@n17
     BoneName=upperarm_r
     BaseName=clavicle_r
     DeltaComponentType=QuaternionTwist
     InputRange=(X=-20,Y=-100)
     TwistAxis="1,0,0"
    )) -> (Output_Get:double)
   Set rightLapelDefUpwardAlpha(execute=@n14.else, rightLapelDefUpwardAlpha=0)
   > Set rightLapelDefDnwardAlpha(rightLapelDefDnwardAlpha=0)
   > Set leftLapelDefUpwardAlpha(leftLapelDefUpwardAlpha=0)
   > Set leftLapelDefDnwardAlpha(leftLapelDefDnwardAlpha=0)
   > Set collar_def_alpha(collar_def_alpha=0)
   Set isFloatingOrDiving(execute=@n1.then_5, isFloatingOrDiving:bool=OR Boolean(A=Get bIsSkydiveDiveMode(), B=Get bIsSkydiveFloating())) -> (Output_Get:bool)
   > n18 = Branch(Condition=Get bIsParachuteOpened())
   > Set isApprenticeGliderOpened(isApprenticeGliderOpened:bool=Get isApprenticeGlider()) -> (Output_Get:bool)
   Set isApprenticeGliderOpened(execute=@n18.else, isApprenticeGliderOpened=false)
   Set FortPawn(execute=@n1.then_6, FortPawn:Fort Pawn Object Reference=TryGetFortPawn()) -> (Output_Get:Fort Pawn Object Reference)
   n19 = Get FortPawn() -> (FortPawn:Fort Pawn Object Reference)
   n20 = Get CurrentWeapon!(self:Fort Pawn Object Reference=@n19) -> (CurrentWeapon:Fort Weapon Object Reference)
   n21 = IsWeaponHolstered!(self:Fort Pawn Object Reference=@n19) -> (ReturnValue:bool)
   > Set weapon_visible(weapon_visible:bool=AND Boolean(A=IsValid(Object=@n20), B=NOT Boolean(A=@n21))) -> (Output_Get:bool)
   n22 = Get AnimInst() -> (AnimInst:Fort Player Anim Instance Object Reference)
   Switch on EGliderType(
    execute=@n1.then_7
    Selection=Break Fort Anim Input Player Glider Anim Asset!(FortAnimInput_PlayerGliderAnimAsset:"Fort Anim Input Player Glider Anim Asset Structure (by ref)"=Get CurrentGliderAnimSet!(self:Fort Player Anim Instance Object Reference=@n22) -> (CurrentGliderAnimSet:Fort Anim Input Player Glider Anim Asset Structure)) -> (
     Default_Base_BS:Blend Space Object Reference
     Default_BodyAdditive_MaleMedium_BS:Blend Space Object Reference
     Default_BodyAdditive_MaleLarge_BS:Blend Space Object Reference
     Default_BodyAdditive_FemaleSmall_BS:Blend Space Object Reference
     Default_BodyAdditive_FemaleMedium_BS:Blend Space Object Reference
     Default_BodyAdditive_FemaleLarge_BS:Blend Space Object Reference
     Default_TurnAdditive_MaleMedium_BS:Blend Space Object Reference
     Default_TurnAdditive_MaleLarge_BS:Blend Space Object Reference
     Default_TurnAdditive_FemaleSmall_BS:Blend Space Object Reference
     Default_TurnAdditive_FemaleMedium_BS:Blend Space Object Reference
     Default_TurnAdditive_FemaleLarge_BS:Blend Space Object Reference
     Into_Base_BS:Blend Space Object Reference
     Into_Base_MaleMedium_BS:Blend Space Object Reference
     Into_BodyAdditive_MaleMedium_BS:Blend Space Object Reference
     Into_BodyAdditive_MaleLarge_BS:Blend Space Object Reference
     Into_BodyAdditive_FemaleSmall_BS:Blend Space Object Reference
     Into_BodyAdditive_FemaleMedium_BS:Blend Space Object Reference
     Into_BodyAdditive_FemaleLarge_BS:Blend Space Object Reference
     Into_TurnAdditive_MaleMedium_BS:Blend Space Object Reference
     Into_TurnAdditive_MaleLarge_BS:Blend Space Object Reference
     Into_TurnAdditive_FemaleSmall_BS:Blend Space Object Reference
     Into_TurnAdditive_FemaleMedium_BS:Blend Space Object Reference
     Into_TurnAdditive_FemaleLarge_BS:Blend Space Object Reference
     Lean_MaleMedium_BS:Blend Space Object Reference
     Lean_MaleLarge_BS:Blend Space Object Reference
     Lean_FemaleSmall_BS:Blend Space Object Reference
     Lean_FemaleMedium_BS:Blend Space Object Reference
     Lean_FemaleLarge_BS:Blend Space Object Reference
     LeanAdditive_Center_MaleMedium_Pose:Anim Sequence Object Reference
     LeanAdditive_Center_MaleLarge_Pose:Anim Sequence Object Reference
     LeanAdditive_Center_FemaleSmall_Pose:Anim Sequence Object Reference
     LeanAdditive_Center_FemaleMedium_Pose:Anim Sequence Object Reference
     LeanAdditive_Center_FemaleLarge_Pose:Anim Sequence Object Reference
     LeanAdditive_Into_BS:Blend Space Object Reference
     LeanAdditive_ForwardInto_Anim:Anim Sequence Object Reference
     LeanAdditive_ForwardInto_FromDeploy_Anim:Anim Sequence Object Reference
     LeanAdditive_BackInto_Anim:Anim Sequence Object Reference
     LeanAdditive_LeftInto_Anim:Anim Sequence Object Reference
     LeanAdditive_RightInto_Anim:Anim Sequence Object Reference
     LeanAdditive_ForwardOut_Anim:Anim Sequence Object Reference
     LeanAdditive_BackOut_Anim:Anim Sequence Object Reference
     LeanAdditive_LeftOut_Anim:Anim Sequence Object Reference
     LeanAdditive_RightOut_Anim:Anim Sequence Object Reference
     ToGlide_BS:Blend Space Object Reference
     ToGlide_Lean_BS:Blend Space Object Reference
     ToDive_BS:Blend Space Object Reference
     ToDive_Lean_BS:Blend Space Object Reference
     Dive_WeaponR_Additive_BS:Blend Space Object Reference
     Glide_WeaponR_Additive_Anim:Anim Sequence Object Reference
     GenericAdditive_Male_BS:Blend Space Object Reference
     GenericAdditive_Female_BS:Blend Space Object Reference
     RootModPitchMin:float=~
     RootModPitchMax:float=~
     RootModYOffsetMin:float=~
     RootModYOffsetMax:float=~
     PelvisModVertical:Pelvis Mod Vertical Input Spring Structure
     PelvisModLateral:Pelvis Mod Lateral Input Spring Structure
     PelvisModBankAngle:Pelvis Mod Bank Angle Input Spring Structure
     ThighModLegAngle:Thigh Mod Leg Angle Input Structure
     ThighModLegPitch:Thigh Mod Leg Pitch Spring Structure
     ThighModLegBank:Thigh Mod Leg Bank Spring Structure
     ThighModLeftLegDrag:Thigh Mod Left Leg Drag Structure
     PlayerGliderType:EGliderType Enum
     bEnableSpringMods:bool=~
     bAllowPlayerDeployRootMod:bool=~
     bUseSurfStyle:bool=~
     bUseLeanPlayRateAdjust:bool=~
     bUseForwardAs_Y_BS:bool=~
     bUseDeployCurveForIntoPlayRateAdjust:bool=~
     LeanForwardValue:float=~
     IdleValue:float=~
     LeanBackValue:float=~
     IdleInterpToSpeed:float=~
     LeanForwardInterpToSpeed:float=~
     LeanBackInterpToSpeed:float=~
    ) {is=title K2Node_BreakStruct}.PlayerGliderType
   )
   > Set has_hand_glider(has_hand_glider:bool=true) -> (Output_Get:bool)
   n23 = GetOwnerFortPlayerAnimInstance!(self:Custom Character Part Anim Instance Object Reference) -> (ReturnValue:Fort Player Anim Instance Object Reference)
   Is Valid(exec=@n1.then_8, InputObject=@n23)
   > n24 = Set isCustomEmotePlaying_0(isCustomEmotePlaying_0:bool=Montage_IsPlaying(self=@n23, Montage=/BRCosmetics/Animation/Game/MainPlayer/Emotes/PapayaComms/PapayaComms_Sit_CMM_M)) -> (Output_Get:bool)
   > Set isCustomEmotePlaying(isCustomEmotePlaying:double=SelectFloat(A=1, B=0, bPickA=@n24)) -> (Output_Get:double)
   Event OnCharacterCustomizationLoaded()
   > n25 = Cast To FortPlayerPawn(Object=GetOwningActor())
   > Is Valid(InputObject=Get apprenticeGliderSkelMesh() -> (apprenticeGliderSkelMesh:Skeletal Mesh Object Reference))
   > n26 = Set currentGlider(currentGlider:Athena Glider Item Definition Object Reference=GetItemDefinitionTypeFromLoadoutWithActor!(
     Actor:Actor Object Reference=@n25
     LoadoutSlot:Gameplay Tag Structure=(TagName="LoadoutSlot.Glider")
     CosmeticType:Athena Cosmetic Item Definition Class Reference=/Script/FortniteGame.AthenaGliderItemDefinition
    ) -> (OutCosmeticItemDef:Athena Glider Item Definition Object Reference)) -> (Output_Get:Athena Glider Item Definition Object Reference)
   > Set isApprenticeGlider(isApprenticeGlider:bool="Equal (Object)"(A=@n26, B=Get gliderItemDef() -> (gliderItemDef:Athena Glider Item Definition Object Reference)) {is=title K2Node_CallFunction}) -> (Output_Get:bool)
   Event BlueprintInitializeAnimation()
   n27 = GetOwnerSkeletalMeshComponent()
   > Is Valid(InputObject=@n27)
   > Set SkelMesh(SkelMesh=@n27)
   > Set AnimInst(AnimInst:Fort Player Anim Instance Object Reference=GetOwnerFortPlayerAnimInstance()) -> (Output_Get:Fort Player Anim Instance Object Reference)
   Get bIsAccelerating2D()
  ]}
]
variables [
 AnimInst {type=FortPlayerAnimInstance, default=null}
 apprenticeGliderSkelMesh {type=SkeletalMesh, default=/Game/Gadgets/Assets/VinderTech_GliderChute/Glider_Apprentice/Meshes/Apprentice_Glider}
 collar_def_alpha {type=number}
 currentGlider {type=AthenaGliderItemDefinition, default=null}
 FortPawn {type=FortPawn, default=null}
 gliderItemDef {type=AthenaGliderItemDefinition, default=/BRCosmetics/Athena/Items/Cosmetics/Gliders/Glider_Apprentice}
 Gravity Override Params {
  type=GravityOverrideParamsStruct
  default {
   skydive_DivingDown {windAmplitude=<50 800 100>, windFrequency=<1.5 30 10>, windOffset=<0 300 0>}
   skydive_FloatingDown {gravityOverride=<0 -700 -100>, windAmplitude=<200 -500 200>, windOffset=<0 -600 100>}
   skydive_GoingUp {gravityOverride=<0 -300 200>, windAmplitude=<200 -500 300>}
   skydive_Parachuting {gravityOverride=<0 -300 -980>, windAmplitude=<200 -150 300>, windFrequency=<1.5 4 6>, windOffset=<0 -100 0>}
  }
 }
 GravityOverride {type=Vector}
 has_hand_glider {type=boolean}
 isApprenticeGlider {type=boolean}
 isApprenticeGliderOpened {type=boolean}
 isCustomEmotePlaying {type=number}
 isCustomEmotePlaying_0 {type=boolean}
 isFloatingOrDiving {type=boolean}
 isLongHair {type=boolean}
 knee_rot_l {type=number}
 knee_rot_r {type=number}
 leftLapelDefDnwardAlpha {type=number}
 leftLapelDefUpwardAlpha {type=number}
 lowerarm_rot_up_l {type=number}
 lowerarm_rot_up_r {type=number}
 pullRightJacketOutAlpha {type=number}
 rightLapelDefDnwardAlpha {type=number}
 rightLapelDefUpwardAlpha {type=number}
 SkelMesh {type=SkeletalMeshComponent}
 thigh_rot_up_l {type=number}
 thigh_rot_up_r {type=number}
 uppperarm_l_twist_alpha {type=number}
 weapon_visible {type=boolean}
]
