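# Physics asset for the dynamic hair on the medium male Apprentice player cosmetic's face accessory part.
# Left, back and right capsule chains hang from a head offset body through master bodies, with angular limits along each chain and an extra tip on the right.
# Cross-constraints tie the outer chains' last links to the back chain's.
# Pelvis, spine, neck, clavicle and upper-arm capsules form a linked body proxy beneath the hair.
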PhysicsAsset /BRCosmetics/Characters/Player/Male/Medium/Bodies/M_MED_Apprentice/Meshes/Parts/M_MED_Apprentice_FaceAcc_Dyn_Hair
bodies {
 clavicle_l [{
   shapeType=Capsule
   length=9.625
   radius=10
   shapeName=clavicle_l_capsule
   center=<6.7755 0.3543 -0.9567>
   rotation=<rot 86.2624 -0.0052 -5.0052>
  }]
 clavicle_r [{
   shapeType=Capsule
   length=9.625
   radius=10
   shapeName=clavicle_l_capsule
   center=<-6.7755 -0.3543 0.9567>
   rotation=<rot 86.2624 -0.0052 174.9948>
  }]
 dyn_hairTips_base_r_01 [{shapeType=Capsule, length=6.5, radius=5.25, shapeName=dyn_hair_r_01_capsule, center=<13 0 0>, rotation=<rot 90 0 0>}]
 dyn_hair_bk_01 [{shapeType=Capsule, length=7.5, radius=10.75, shapeName=dyn_hair_bk_01_capsule, center=<15 0 4>, rotation=<rot 84.9998 0 0>}]
 dyn_hair_bk_02 [{shapeType=Capsule, length=10.75, radius=12.875, shapeName=dyn_hair_bk_02_capsule, center=<20 0 4>, rotation=<rot 84.9995 0 0>}]
 dyn_hair_bk_03 [{shapeType=Capsule, length=16, radius=13.25, shapeName=dyn_hair_bk_03_capsule, center=<24.4292 0 4>, rotation=<rot 90 0 0>}]
 dyn_hair_bk_MSTR [{shapeType=Capsule, length=0.505, radius=0.505, shapeName=dyn_hair_bk_MSTR_capsule, rotation=<rot 0 -90 0>}]
 dyn_hair_l_01 [{shapeType=Capsule, length=5, radius=12.375, shapeName=dyn_hair_l_01_capsule, center=<15 0 4>, rotation=<rot 84.9999 -0.0002 -0.0002>}]
 dyn_hair_l_02 [{shapeType=Capsule, length=8.625, radius=14, shapeName=dyn_hair_l_02_capsule, center=<20 0 4>, rotation=<rot 84.9996 -0.0005 -0.0005>}]
 dyn_hair_l_03 [{shapeType=Capsule, length=9.75, radius=13.125, shapeName=dyn_hair_l_03_capsule, center=<20 0 4>, rotation=<rot 90 0 0>}]
 dyn_hair_l_MSTR [{shapeType=Capsule, length=0.505, radius=0.505, shapeName=dyn_hair_l_MSTR_capsule, rotation=<rot 0 -90 0>}]
 dyn_hair_offset [{shapeType=Capsule, length=1, radius=1, shapeName=head_capsule}]
 dyn_hair_r_01 [{shapeType=Capsule, length=5, radius=12.25, shapeName=dyn_hair_r_01_capsule, center=<15 -3 4>, rotation=<rot 75.203 -13.4468 6.7035>}]
 dyn_hair_r_02 [{shapeType=Capsule, length=8.625, radius=13.625, shapeName=dyn_hair_r_02_capsule, center=<20 0 4>, rotation=<rot 84.9997 0.0004 0.0004>}]
 dyn_hair_r_03 [{shapeType=Capsule, length=9.75, radius=12.875, shapeName=dyn_hair_r_03_capsule, center=<20 0 4>, rotation=<rot 90 0 0>}]
 dyn_hair_r_MSTR [{shapeType=Capsule, length=0.505, radius=0.505, shapeName=dyn_hair_r_MSTR_capsule, rotation=<rot 0 -90 0>}]
 neck_01 [{shapeType=Capsule, length=2.375, radius=13.8117, shapeName=head_capsule, center=<2.5967 6.3005 0>, rotation=<rot 90 0 0>}]
 pelvis [{shapeType=Capsule, length=12, radius=32.375, shapeName=spine_05_capsule, center=<-18.4599 6.7099 0>, rotation=<rot 90 178.2101 173.2102>}]
 spine_02 [{shapeType=Capsule, length=9.375, radius=26.125, shapeName=spine_05_capsule, center=<0.6777 4.3768 0>}]
 spine_05 [{shapeType=Capsule, length=5.375, radius=18.5, shapeName=spine_05_capsule, center=<-3.815 3.778 0>}]
 upperarm_l [{
   shapeType=Capsule
   length=40.875
   radius=10
   shapeName=clavicle_l_capsule
   center=<23.8485 3.0897 -0.9251>
   rotation=<rot 88.5298 -143.8228 -148.7947>
  }]
 upperarm_r [{
   shapeType=Capsule
   length=40.875
   radius=10
   shapeName=clavicle_l_capsule
   center=<-23.8485 -3.0897 0.9251>
   rotation=<rot 88.5298 -143.8228 31.2053>
  }]
}
constraints [
 {
  bone1Name=dyn_hair_l_MSTR
  bone2Name=dyn_hair_offset
  swing1LimitDegrees=45
  swing1Motion=Free
  swing2LimitDegrees=45
  swing2Motion=Free
  twistLimitDegrees=45
  twistMotion=Free
 }
 {bone1Name=dyn_hair_l_01, bone2Name=dyn_hair_l_MSTR, swing1LimitDegrees=25, swing2LimitDegrees=45, twistLimitDegrees=5}
 {bone1Name=dyn_hair_l_02, bone2Name=dyn_hair_l_01, swing1LimitDegrees=60, swing2LimitDegrees=45, twistLimitDegrees=5}
 {bone1Name=dyn_hair_l_03, bone2Name=dyn_hair_l_02, swing1LimitDegrees=60, swing2LimitDegrees=45, twistLimitDegrees=5}
 {
  bone1Name=dyn_hair_bk_MSTR
  bone2Name=dyn_hair_offset
  swing1LimitDegrees=45
  swing1Motion=Free
  swing2LimitDegrees=45
  swing2Motion=Free
  twistLimitDegrees=45
  twistMotion=Free
 }
 {bone1Name=dyn_hair_bk_01, bone2Name=dyn_hair_bk_MSTR, swing1LimitDegrees=25, swing2LimitDegrees=45, twistLimitDegrees=5}
 {bone1Name=dyn_hair_bk_02, bone2Name=dyn_hair_bk_01, swing1LimitDegrees=60, swing2LimitDegrees=45, twistLimitDegrees=5}
 {bone1Name=dyn_hair_bk_03, bone2Name=dyn_hair_bk_02, swing1LimitDegrees=60, swing2LimitDegrees=45, twistLimitDegrees=5}
 {
  bone1Name=dyn_hair_r_MSTR
  bone2Name=dyn_hair_offset
  swing1LimitDegrees=45
  swing1Motion=Free
  swing2LimitDegrees=45
  swing2Motion=Free
  twistLimitDegrees=45
  twistMotion=Free
 }
 {bone1Name=dyn_hair_r_01, bone2Name=dyn_hair_r_MSTR, swing1LimitDegrees=25, swing2LimitDegrees=45, twistLimitDegrees=5}
 {bone1Name=dyn_hair_r_02, bone2Name=dyn_hair_r_01, swing1LimitDegrees=60, swing2LimitDegrees=45, twistLimitDegrees=5}
 {bone1Name=dyn_hair_r_03, bone2Name=dyn_hair_r_02, swing1LimitDegrees=60, swing2LimitDegrees=45, twistLimitDegrees=5}
 {
  bone1Name=spine_05
  bone2Name=spine_02
  swing1LimitDegrees=45
  swing1Motion=Free
  swing2LimitDegrees=45
  swing2Motion=Free
  twistLimitDegrees=45
  twistMotion=Free
 }
 {
  bone1Name=neck_01
  bone2Name=spine_05
  swing1LimitDegrees=45
  swing1Motion=Free
  swing2LimitDegrees=45
  swing2Motion=Free
  twistLimitDegrees=45
  twistMotion=Free
 }
 {
  bone1Name=spine_02
  bone2Name=pelvis
  swing1LimitDegrees=45
  swing1Motion=Free
  swing2LimitDegrees=45
  swing2Motion=Free
  twistLimitDegrees=45
  twistMotion=Free
 }
 {
  bone1Name=clavicle_l
  bone2Name=spine_05
  swing1LimitDegrees=45
  swing1Motion=Free
  swing2LimitDegrees=45
  swing2Motion=Free
  twistLimitDegrees=45
  twistMotion=Free
 }
 {
  bone1Name=dyn_hair_r_03
  bone2Name=dyn_hair_bk_03
  swing1LimitDegrees=5
  swing2LimitDegrees=45
  swing2Motion=Free
  twistLimitDegrees=45
  twistMotion=Free
 }
 {
  bone1Name=dyn_hair_l_03
  bone2Name=dyn_hair_bk_03
  swing1LimitDegrees=5
  swing2LimitDegrees=45
  swing2Motion=Free
  twistLimitDegrees=45
  twistMotion=Free
 }
 {
  bone1Name=clavicle_r
  bone2Name=spine_05
  swing1LimitDegrees=5
  swing2LimitDegrees=45
  swing2Motion=Free
  twistLimitDegrees=45
  twistMotion=Free
 }
 {
  bone1Name=upperarm_r
  bone2Name=clavicle_r
  swing1LimitDegrees=45
  swing1Motion=Free
  swing2LimitDegrees=45
  swing2Motion=Free
  twistLimitDegrees=45
  twistMotion=Free
 }
 {
  bone1Name=dyn_hair_offset
  bone2Name=neck_01
  swing1LimitDegrees=45
  swing1Motion=Free
  swing2LimitDegrees=45
  swing2Motion=Free
  twistLimitDegrees=45
  twistMotion=Free
 }
 {
  bone1Name=upperarm_l
  bone2Name=clavicle_l
  swing1LimitDegrees=45
  swing1Motion=Free
  swing2LimitDegrees=45
  swing2Motion=Free
  twistLimitDegrees=45
  twistMotion=Free
 }
 {bone1Name=dyn_hairTips_base_r_01, bone2Name=dyn_hair_r_01, swing1LimitDegrees=25, swing2LimitDegrees=30, twistLimitDegrees=5}
]
