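# Animation Blueprint that drives secondary hair motion on the Apprentice character's face accessory mesh.
# When bRigidBodyEnabled, RigidBody simulations chosen by skydive-floating and parachute state feed chained, spring-limited AnimDynamics hair-tip bones.
# Otherwise, look-at, head-rotation-weighted bone offsets and spine constraints pose the hair procedurally.
# The reactiveBoolean flag overrides both with another RigidBody pass plus an additive hair animation.
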AnimBlueprint /BRCosmetics/Characters/Player/Male/Medium/Bodies/M_MED_Apprentice/Meshes/Parts/M_MED_Apprentice_FaceAcc_AnimBP
target_skeleton=&/M_MED_Apprentice_FaceAcc_Skeleton
graphs [
 AnimGraph {nodes [
   Anim Dynamics - Bone: dyn_hairTips_base_r_01!(ComponentPose, Alpha=1, GravityScale:float=1, ExternalForce="0,0,0") -> (Pose) {
    angularDampingOverride=1
    angularSpringConstant=500
    bAngularSpring=true
    bGravityOverrideInSimSpace=true
    bOverrideAngularDamping=true
    bUseGravityOverride=true
    simulationSpace=RootRelative
    alphaBoolBlend {blendOption=Linear}
    boundBone {boneName=dyn_hairTips_base_r_01}
    componentAppliedLinearAccClamp=<1000 1000 1000>
    componentLinearAccScale=<0.1 0.1 0.1>
    componentLinearVelScale=<0.6 0.6 0.6>
    gravityOverride {z=-980}
    physicsBodyDefinitions [{
      sphereCollisionRadius=0
      boundBone {boneName=dyn_hairTips_base_r_01}
      boxExtents=<10 20 20>
      constraintSetup {angularLimitsMax=<0 5 30>, angularLimitsMin=<0 -20 -15>, angularTarget=<0 0 0>}
      localJointOffset=<2 0.2 -0.5>
     }]
    relativeSpaceBone {boneName=head}
   }
   Blend Poses by bool(
    BlendPose_0=Layered blend per bone(
     BasePose=Use cached pose 'MeshPose'()
     BlendPoses_0="Sequence Player (additive) 'M_Med_Apprentice_FaceAcc_Hair_Anim'"() {sequence=/BRCosmetics/Characters/Player/Male/Medium/Bodies/M_MED_Apprentice/Animation/M_Med_Apprentice_FaceAcc_Hair_Anim}
    ) {layerSetup [{branchFilters [{boneName=head}]}]}
    BlendPose_1=Use cached pose 'MeshPose'()
    BlendTime_0=0
    BlendTime_1=0.25
    bActiveValue=Get onKill() -> (onKill:bool)
   ) {blendType=Linear}
   n1 = Anim Dynamics - Bone: dyn_rect_hairTips_A_01!(ComponentPose=Local To Component(), bAlphaBoolEnabled:bool=True, GravityScale:float=1, ExternalForce="0,0,0") -> (Pose) {
    alphaInputType=Bool
    angularDampingOverride=0.999
    angularSpringConstant=500
    bAngularSpring=true
    bGravityOverrideInSimSpace=true
    bOverrideAngularDamping=true
    bUseGravityOverride=true
    simulationSpace=BoneRelative
    alphaBoolBlend {blendOption=Linear}
    boundBone {boneName=dyn_rect_hairTips_A_01}
    componentAppliedLinearAccClamp=<1000 1000 1000>
    componentLinearAccScale=<0.2 0.2 0.2>
    componentLinearVelScale=<0.6 0.6 0.6>
    gravityOverride {x=-600}
    physicsBodyDefinitions [{
      sphereCollisionRadius=0
      boundBone {boneName=dyn_rect_hairTips_A_01}
      boxExtents=<10 15 15>
      constraintSetup {angularLimitsMax=<0 0 10>, angularLimitsMin=<0 0 -15>, angularTarget=<0 0 0>}
      localJointOffset=<3 -0.8 0>
     }]
    relativeSpaceBone {boneName=head}
   }
   n2 = Anim Dynamics - Bone: dyn_rect_hairTips_B_01!(ComponentPose=@n1, bAlphaBoolEnabled:bool=True, GravityScale:float=1, ExternalForce="0,0,0") -> (Pose) {
    alphaInputType=Bool
    angularDampingOverride=0.999
    angularSpringConstant=500
    bAngularSpring=true
    bGravityOverrideInSimSpace=true
    bOverrideAngularDamping=true
    bUseGravityOverride=true
    simulationSpace=BoneRelative
    alphaBoolBlend {blendOption=Linear}
    boundBone {boneName=dyn_rect_hairTips_B_01}
    componentAppliedLinearAccClamp=<1000 1000 1000>
    componentLinearAccScale=<0.2 0.2 0.2>
    componentLinearVelScale=<0.6 0.6 0.6>
    gravityOverride {x=-600}
    physicsBodyDefinitions [{
      sphereCollisionRadius=0
      boundBone {boneName=dyn_rect_hairTips_B_01}
      boxExtents=<10 16 16>
      constraintSetup {angularLimitsMax=<0 0 12>, angularLimitsMin=<0 0 -15>, angularTarget=<0 0 0>}
      localJointOffset=<4 -2 0>
     }]
    relativeSpaceBone {boneName=head}
   }
   n3 = Anim Dynamics - Bone: dyn_rect_hairTips_C_01!(ComponentPose=@n2, bAlphaBoolEnabled:bool=True, GravityScale:float=1, ExternalForce="0,0,0") -> (Pose) {
    alphaInputType=Bool
    angularDampingOverride=1
    angularSpringConstant=400
    bAngularSpring=true
    bGravityOverrideInSimSpace=true
    bOverrideAngularDamping=true
    bUseGravityOverride=true
    simulationSpace=RootRelative
    alphaBoolBlend {blendOption=Linear}
    boundBone {boneName=dyn_rect_hairTips_C_01}
    componentAppliedLinearAccClamp=<1000 1000 1000>
    componentLinearAccScale=<0.4 0.4 0.4>
    componentLinearVelScale=<0.6 0.6 0.6>
    gravityOverride {z=-980}
    physicsBodyDefinitions [{
      sphereCollisionRadius=0
      boundBone {boneName=dyn_rect_hairTips_C_01}
      boxExtents=<15 25 20>
      constraintSetup {angularLimitsMax=<0 10 18>, angularLimitsMin=<0 -10 -12>, angularTarget=<0 0 0>}
      localJointOffset=<3 -2 -1>
     }]
    relativeSpaceBone {boneName=head}
   }
   n4 = Anim Dynamics - Bone: dyn_rect_hairTips_D_01!(ComponentPose=@n3, bAlphaBoolEnabled:bool=True, GravityScale:float=1, ExternalForce="0,0,0") -> (Pose) {
    alphaInputType=Bool
    angularDampingOverride=1
    angularSpringConstant=500
    bAngularSpring=true
    bGravityOverrideInSimSpace=true
    bOverrideAngularDamping=true
    bUseGravityOverride=true
    simulationSpace=RootRelative
    alphaBoolBlend {blendOption=Linear}
    boundBone {boneName=dyn_rect_hairTips_D_01}
    componentAppliedLinearAccClamp=<1000 1000 1000>
    componentLinearAccScale=<0.1 0.1 0.1>
    componentLinearVelScale=<0.6 0.6 0.6>
    gravityOverride {z=-980}
    physicsBodyDefinitions [{
      sphereCollisionRadius=0
      boundBone {boneName=dyn_rect_hairTips_D_01}
      boxExtents=<10 20 20>
      constraintSetup {angularLimitsMax=<0 10 10>, angularLimitsMin=<0 -20 -10>, angularTarget=<0 0 0>}
      localJointOffset=<1 -0.5 -2>
     }]
    relativeSpaceBone {boneName=head}
   }
   n5 = Anim Dynamics - Bone: dyn_rect_hairTips_E_01!(ComponentPose=@n4, bAlphaBoolEnabled:bool=True, GravityScale:float=1, ExternalForce="0,0,0") -> (Pose) {
    alphaInputType=Bool
    angularDampingOverride=1
    angularSpringConstant=500
    bAngularSpring=true
    bGravityOverrideInSimSpace=true
    bOverrideAngularDamping=true
    bUseGravityOverride=true
    simulationSpace=RootRelative
    alphaBoolBlend {blendOption=Linear}
    boundBone {boneName=dyn_rect_hairTips_E_01}
    componentAppliedLinearAccClamp=<1000 1000 1000>
    componentLinearAccScale=<0.1 0.1 0.1>
    componentLinearVelScale=<0.6 0.6 0.6>
    gravityOverride {z=-980}
    physicsBodyDefinitions [{
      sphereCollisionRadius=0
      boundBone {boneName=dyn_rect_hairTips_E_01}
      boxExtents=<10 20 20>
      constraintSetup {angularLimitsMax=<0 20 10>, angularLimitsMin=<0 -20 -30>, angularTarget=<0 0 0>}
      localJointOffset=<2 0 -2>
     }]
    relativeSpaceBone {boneName=head}
   }
   n6 = Anim Dynamics - Bone: dyn_rect_hairTips_F_01!(ComponentPose=@n5, bAlphaBoolEnabled:bool=True, GravityScale:float=1, ExternalForce="0,0,0") -> (Pose) {
    alphaInputType=Bool
    angularDampingOverride=0.999
    angularSpringConstant=1000
    bAngularSpring=true
    bGravityOverrideInSimSpace=true
    bOverrideAngularDamping=true
    bUseGravityOverride=true
    simulationSpace=RootRelative
    alphaBoolBlend {blendOption=Linear}
    boundBone {boneName=dyn_rect_hairTips_F_01}
    componentAppliedLinearAccClamp=<1000 1000 1000>
    componentLinearAccScale=<0.1 0.1 0.1>
    componentLinearVelScale=<0.6 0.6 0.6>
    gravityOverride {z=-980}
    physicsBodyDefinitions [{
      sphereCollisionRadius=0
      boundBone {boneName=dyn_rect_hairTips_F_01}
      boxExtents=<10 15 15>
      constraintSetup {angularLimitsMax=<0 20 15>, angularLimitsMin=<0 -10 -15>, angularTarget=<0 0 0>}
      localJointOffset=<0 0 -3>
     }]
    relativeSpaceBone {boneName=head}
   }
   n7 = Anim Dynamics - Bone: dyn_rect_hairTips_G_01!(ComponentPose=@n6, bAlphaBoolEnabled:bool=True, GravityScale:float=1, ExternalForce="0,0,0") -> (Pose) {
    alphaInputType=Bool
    angularDampingOverride=1
    angularSpringConstant=400
    bAngularSpring=true
    bGravityOverrideInSimSpace=true
    bOverrideAngularDamping=true
    bUseGravityOverride=true
    simulationSpace=RootRelative
    alphaBoolBlend {blendOption=Linear}
    boundBone {boneName=dyn_rect_hairTips_G_01}
    componentAppliedLinearAccClamp=<1000 1000 1000>
    componentLinearAccScale=<0.4 0.4 0.4>
    componentLinearVelScale=<0.6 0.6 0.6>
    gravityOverride {z=-980}
    physicsBodyDefinitions [{
      sphereCollisionRadius=0
      boundBone {boneName=dyn_rect_hairTips_G_01}
      boxExtents=<15 25 20>
      constraintSetup {angularLimitsMax=<0 15 30>, angularLimitsMin=<0 -15 -30>, angularTarget=<0 0 0>}
      localJointOffset=<-3 -0.2 -0.5>
     }]
    relativeSpaceBone {boneName=head}
   }
   n8 = Anim Dynamics - Bone: dyn_rect_hairTips_H_01!(ComponentPose=@n7, bAlphaBoolEnabled:bool=True, GravityScale:float=1, ExternalForce="0,0,0") -> (Pose) {
    alphaInputType=Bool
    angularDampingOverride=1
    angularSpringConstant=400
    bAngularSpring=true
    bGravityOverrideInSimSpace=true
    bOverrideAngularDamping=true
    bUseGravityOverride=true
    simulationSpace=RootRelative
    alphaBoolBlend {blendOption=Linear}
    boundBone {boneName=dyn_rect_hairTips_H_01}
    componentAppliedLinearAccClamp=<1000 1000 1000>
    componentLinearAccScale=<0.4 0.4 0.4>
    componentLinearVelScale=<0.6 0.6 0.6>
    gravityOverride {z=-980}
    physicsBodyDefinitions [{
      sphereCollisionRadius=0
      boundBone {boneName=dyn_rect_hairTips_H_01}
      boxExtents=<15 25 20>
      constraintSetup {angularLimitsMax=<0 15 30>, angularLimitsMin=<0 -15 -30>, angularTarget=<0 0 0>}
      localJointOffset=<-2 1.5 -1>
     }]
    relativeSpaceBone {boneName=head}
   }
   n9 = Anim Dynamics - Bone: dyn_rect_hairTips_l_A_01!(ComponentPose=@n8, bAlphaBoolEnabled:bool=True, GravityScale:float=1, ExternalForce="0,0,0") -> (Pose) {
    alphaInputType=Bool
    angularDampingOverride=0.9999
    angularSpringConstant=800
    bAngularSpring=true
    bGravityOverrideInSimSpace=true
    bOverrideAngularDamping=true
    bUseGravityOverride=true
    simulationSpace=RootRelative
    alphaBoolBlend {blendOption=Linear}
    boundBone {boneName=dyn_rect_hairTips_l_A_01}
    componentAppliedLinearAccClamp=<1000 1000 1000>
    componentLinearAccScale=<0.4 0.4 0.4>
    componentLinearVelScale=<0.6 0.6 0.6>
    gravityOverride {z=-980}
    physicsBodyDefinitions [{
      sphereCollisionRadius=0
      boundBone {boneName=dyn_rect_hairTips_l_A_01}
      boxExtents=<15 25 20>
      constraintSetup {angularLimitsMax=<0 5 30>, angularLimitsMin=<0 -35 -30>, angularTarget=<0 0 0>}
      localJointOffset=<5 0 -3>
     }]
   }
   n10 = Anim Dynamics - Bone: dyn_rect_hairTips_l_B_01!(ComponentPose=@n9, bAlphaBoolEnabled:bool=True, GravityScale:float=1, ExternalForce="0,0,0") -> (Pose) {
    alphaInputType=Bool
    angularDampingOverride=0.9999
    angularSpringConstant=800
    bAngularSpring=true
    bGravityOverrideInSimSpace=true
    bOverrideAngularDamping=true
    bUseGravityOverride=true
    simulationSpace=RootRelative
    alphaBoolBlend {blendOption=Linear}
    boundBone {boneName=dyn_rect_hairTips_l_B_01}
    componentAppliedLinearAccClamp=<1000 1000 1000>
    componentLinearAccScale=<0.4 0.4 0.4>
    componentLinearVelScale=<0.6 0.6 0.6>
    gravityOverride {z=-980}
    physicsBodyDefinitions [{
      sphereCollisionRadius=0
      boundBone {boneName=dyn_rect_hairTips_l_B_01}
      boxExtents=<15 25 20>
      constraintSetup {angularLimitsMax=<0 5 30>, angularLimitsMin=<0 -35 -30>, angularTarget=<0 0 0>}
      localJointOffset=<5 -1 0>
     }]
   }
   n11 = Anim Dynamics - Bone: dyn_rect_hairTips_l_C_01!(ComponentPose=@n10, bAlphaBoolEnabled:bool=True, GravityScale:float=1, ExternalForce="0,0,0") -> (Pose) {
    alphaInputType=Bool
    angularDampingOverride=0.9999
    angularSpringConstant=800
    bAngularSpring=true
    bGravityOverrideInSimSpace=true
    bOverrideAngularDamping=true
    bUseGravityOverride=true
    simulationSpace=RootRelative
    alphaBoolBlend {blendOption=Linear}
    boundBone {boneName=dyn_rect_hairTips_l_C_01}
    componentAppliedLinearAccClamp=<1000 1000 1000>
    componentLinearAccScale=<0.4 0.4 0.4>
    componentLinearVelScale=<0.6 0.6 0.6>
    gravityOverride {z=-980}
    physicsBodyDefinitions [{
      sphereCollisionRadius=0
      boundBone {boneName=dyn_rect_hairTips_l_C_01}
      boxExtents=<15 25 20>
      constraintSetup {angularLimitsMax=<0 5 30>, angularLimitsMin=<0 -35 -30>, angularTarget=<0 0 0>}
      localJointOffset=<5 -0.5 0>
     }]
   }
   n12 = Anim Dynamics - Bone: dyn_rect_hairTips_l_D_01!(ComponentPose=@n11, bAlphaBoolEnabled:bool=True, GravityScale:float=1, ExternalForce="0,0,0") -> (Pose) {
    alphaInputType=Bool
    angularDampingOverride=0.9999
    angularSpringConstant=800
    bAngularSpring=true
    bGravityOverrideInSimSpace=true
    bOverrideAngularDamping=true
    bUseGravityOverride=true
    simulationSpace=RootRelative
    alphaBoolBlend {blendOption=Linear}
    boundBone {boneName=dyn_rect_hairTips_l_D_01}
    componentAppliedLinearAccClamp=<1000 1000 1000>
    componentLinearAccScale=<0.4 0.4 0.4>
    componentLinearVelScale=<0.6 0.6 0.6>
    gravityOverride {z=-980}
    physicsBodyDefinitions [{
      sphereCollisionRadius=0
      boundBone {boneName=dyn_rect_hairTips_l_D_01}
      boxExtents=<15 25 20>
      constraintSetup {angularLimitsMax=<0 20 5>, angularLimitsMin=<0 -20 -30>, angularTarget=<0 0 0>}
      localJointOffset=<4 -1 -2>
     }]
   }
   n13 = Anim Dynamics - Bone: dyn_rect_hairTips_r_A_01!(ComponentPose=@n12, bAlphaBoolEnabled:bool=True, GravityScale:float=1, ExternalForce="0,0,0") -> (Pose) {
    alphaInputType=Bool
    angularDampingOverride=1
    angularSpringConstant=500
    bAngularSpring=true
    bGravityOverrideInSimSpace=true
    bOverrideAngularDamping=true
    bUseGravityOverride=true
    simulationSpace=RootRelative
    alphaBoolBlend {blendOption=Linear}
    boundBone {boneName=dyn_rect_hairTips_r_A_01}
    componentAppliedLinearAccClamp=<1000 1000 1000>
    componentLinearAccScale=<0.1 0.1 0.1>
    componentLinearVelScale=<0.6 0.6 0.6>
    gravityOverride {z=-980}
    physicsBodyDefinitions [{
      sphereCollisionRadius=0
      boundBone {boneName=dyn_rect_hairTips_r_A_01}
      boxExtents=<10 20 20>
      constraintSetup {angularLimitsMax=<0 5 10>, angularLimitsMin=<0 -20 -30>, angularTarget=<0 0 0>}
      localJointOffset=<1 2 -0.5>
     }]
    relativeSpaceBone {boneName=head}
   }
   n14 = Anim Dynamics - Bone: dyn_rect_hairTips_r_B_01!(ComponentPose=@n13, bAlphaBoolEnabled:bool=True, GravityScale:float=1, ExternalForce="0,0,0") -> (Pose) {
    alphaInputType=Bool
    angularDampingOverride=0.9999
    angularSpringConstant=800
    bAngularSpring=true
    bGravityOverrideInSimSpace=true
    bOverrideAngularDamping=true
    bUseGravityOverride=true
    simulationSpace=RootRelative
    alphaBoolBlend {blendOption=Linear}
    boundBone {boneName=dyn_rect_hairTips_r_B_01}
    componentAppliedLinearAccClamp=<1000 1000 1000>
    componentLinearAccScale=<0.4 0.4 0.4>
    componentLinearVelScale=<0.6 0.6 0.6>
    gravityOverride {z=-980}
    physicsBodyDefinitions [{
      sphereCollisionRadius=0
      boundBone {boneName=dyn_rect_hairTips_r_B_01}
      boxExtents=<15 25 20>
      constraintSetup {angularLimitsMax=<0 5 5>, angularLimitsMin=<0 -35 -30>, angularTarget=<0 0 0>}
      localJointOffset=<3 1 0>
     }]
   }
   n15 = Anim Dynamics - Bone: dyn_rect_hairTips_r_C_01!(ComponentPose=@n14, bAlphaBoolEnabled:bool=True, GravityScale:float=1, ExternalForce="0,0,0") -> (Pose) {
    alphaInputType=Bool
    angularDampingOverride=0.9999
    angularSpringConstant=800
    bAngularSpring=true
    bGravityOverrideInSimSpace=true
    bOverrideAngularDamping=true
    bUseGravityOverride=true
    simulationSpace=RootRelative
    alphaBoolBlend {blendOption=Linear}
    boundBone {boneName=dyn_rect_hairTips_r_C_01}
    componentAppliedLinearAccClamp=<1000 1000 1000>
    componentLinearAccScale=<0.4 0.4 0.4>
    componentLinearVelScale=<0.6 0.6 0.6>
    gravityOverride {z=-980}
    physicsBodyDefinitions [{
      sphereCollisionRadius=0
      boundBone {boneName=dyn_rect_hairTips_r_C_01}
      boxExtents=<15 25 20>
      constraintSetup {angularLimitsMax=<0 5 15>, angularLimitsMin=<0 -20 -20>, angularTarget=<0 0 0>}
      localJointOffset=<4 0 -1>
     }]
   }
   n16 = Anim Dynamics - Bone: dyn_rect_hairTips_bk_A_01!(ComponentPose=@n15, bAlphaBoolEnabled:bool=True, GravityScale:float=1, ExternalForce="0,0,0") -> (Pose) {
    alphaInputType=Bool
    angularDampingOverride=0.9999
    angularSpringConstant=800
    bAngularSpring=true
    bGravityOverrideInSimSpace=true
    bOverrideAngularDamping=true
    bUseGravityOverride=true
    simulationSpace=RootRelative
    alphaBoolBlend {blendOption=Linear}
    boundBone {boneName=dyn_rect_hairTips_bk_A_01}
    componentAppliedLinearAccClamp=<1000 1000 1000>
    componentLinearAccScale=<0.4 0.4 0.4>
    componentLinearVelScale=<0.6 0.6 0.6>
    gravityOverride {z=-980}
    physicsBodyDefinitions [{
      sphereCollisionRadius=0
      boundBone {boneName=dyn_rect_hairTips_bk_A_01}
      boxExtents=<15 25 20>
      constraintSetup {angularLimitsMax=<0 5 20>, angularLimitsMin=<0 -35 -20>, angularTarget=<0 0 0>}
      localJointOffset=<3 0 -2>
     }]
   }
   n17 = Anim Dynamics - Bone: dyn_rect_hairTips_bk_B_01!(ComponentPose=@n16, bAlphaBoolEnabled:bool=True, GravityScale:float=1, ExternalForce="0,0,0") -> (Pose) {
    alphaInputType=Bool
    angularDampingOverride=0.9999
    angularSpringConstant=800
    bAngularSpring=true
    bGravityOverrideInSimSpace=true
    bOverrideAngularDamping=true
    bUseGravityOverride=true
    simulationSpace=RootRelative
    alphaBoolBlend {blendOption=Linear}
    boundBone {boneName=dyn_rect_hairTips_bk_B_01}
    componentAppliedLinearAccClamp=<1000 1000 1000>
    componentLinearAccScale=<0.4 0.4 0.4>
    componentLinearVelScale=<0.6 0.6 0.6>
    gravityOverride {z=-980}
    physicsBodyDefinitions [{
      sphereCollisionRadius=0
      boundBone {boneName=dyn_rect_hairTips_bk_B_01}
      boxExtents=<15 25 20>
      constraintSetup {angularLimitsMax=<0 5 20>, angularLimitsMin=<0 -35 -20>, angularTarget=<0 0 0>}
      localJointOffset=<3 -0.5 0>
     }]
   }
   n18 = Anim Dynamics - Bone: dyn_rect_hairTips_bk_C_01!(ComponentPose=@n17, bAlphaBoolEnabled:bool=True, GravityScale:float=1, ExternalForce="0,0,0") -> (Pose) {
    alphaInputType=Bool
    angularDampingOverride=0.9999
    angularSpringConstant=800
    bAngularSpring=true
    bGravityOverrideInSimSpace=true
    bOverrideAngularDamping=true
    bUseGravityOverride=true
    simulationSpace=RootRelative
    alphaBoolBlend {blendOption=Linear}
    boundBone {boneName=dyn_rect_hairTips_bk_C_01}
    componentAppliedLinearAccClamp=<1000 1000 1000>
    componentLinearAccScale=<0.4 0.4 0.4>
    componentLinearVelScale=<0.6 0.6 0.6>
    gravityOverride {z=-980}
    physicsBodyDefinitions [{
      sphereCollisionRadius=0
      boundBone {boneName=dyn_rect_hairTips_bk_C_01}
      boxExtents=<15 25 20>
      constraintSetup {angularLimitsMax=<0 5 15>, angularLimitsMin=<0 -20 -20>, angularTarget=<0 0 0>}
      localJointOffset=<4 0 -2>
     }]
   }
   n19 = Anim Dynamics - Bone: dyn_rect_hairTips_bk_D_01!(ComponentPose=@n18, bAlphaBoolEnabled:bool=True, GravityScale:float=1, ExternalForce="0,0,0") -> (Pose) {
    alphaInputType=Bool
    angularDampingOverride=0.9999
    angularSpringConstant=800
    bAngularSpring=true
    bGravityOverrideInSimSpace=true
    bOverrideAngularDamping=true
    bUseGravityOverride=true
    simulationSpace=RootRelative
    alphaBoolBlend {blendOption=Linear}
    boundBone {boneName=dyn_rect_hairTips_bk_D_01}
    componentAppliedLinearAccClamp=<1000 1000 1000>
    componentLinearAccScale=<0.4 0.4 0.4>
    componentLinearVelScale=<0.6 0.6 0.6>
    gravityOverride {z=-980}
    physicsBodyDefinitions [{
      sphereCollisionRadius=0
      boundBone {boneName=dyn_rect_hairTips_bk_D_01}
      boxExtents=<15 25 20>
      constraintSetup {angularLimitsMax=<0 5 30>, angularLimitsMin=<0 -20 -15>, angularTarget=<0 0 0>}
      localJointOffset=<3.5 0 0>
     }]
   }
   n20 = Anim Dynamics - Bone: dyn_rect_hairTips_bk_E_01!(ComponentPose=@n19, bAlphaBoolEnabled:bool=True, GravityScale:float=1, ExternalForce="0,0,0") -> (Pose) {
    alphaInputType=Bool
    angularDampingOverride=0.9999
    angularSpringConstant=800
    bAngularSpring=true
    bGravityOverrideInSimSpace=true
    bOverrideAngularDamping=true
    bUseGravityOverride=true
    simulationSpace=RootRelative
    alphaBoolBlend {blendOption=Linear}
    boundBone {boneName=dyn_rect_hairTips_bk_E_01}
    componentAppliedLinearAccClamp=<1000 1000 1000>
    componentLinearAccScale=<0.4 0.4 0.4>
    componentLinearVelScale=<0.6 0.6 0.6>
    gravityOverride {z=-980}
    physicsBodyDefinitions [{
      sphereCollisionRadius=0
      boundBone {boneName=dyn_rect_hairTips_bk_E_01}
      boxExtents=<15 25 20>
      constraintSetup {angularLimitsMax=<0 5 30>, angularLimitsMin=<0 -20 -30>, angularTarget=<0 0 0>}
      localJointOffset=<4 0 0>
     }]
   }
   n21 = Anim Dynamics - Bone: dyn_rect_hairTips_bk_F_01!(ComponentPose=@n20, bAlphaBoolEnabled:bool=True, GravityScale:float=1, ExternalForce="0,0,0") -> (Pose) {
    alphaInputType=Bool
    angularDampingOverride=0.9999
    angularSpringConstant=800
    bAngularSpring=true
    bGravityOverrideInSimSpace=true
    bOverrideAngularDamping=true
    bUseGravityOverride=true
    simulationSpace=RootRelative
    alphaBoolBlend {blendOption=Linear}
    boundBone {boneName=dyn_rect_hairTips_bk_F_01}
    componentAppliedLinearAccClamp=<1000 1000 1000>
    componentLinearAccScale=<0.4 0.4 0.4>
    componentLinearVelScale=<0.6 0.6 0.6>
    gravityOverride {z=-980}
    physicsBodyDefinitions [{
      sphereCollisionRadius=0
      boundBone {boneName=dyn_rect_hairTips_bk_F_01}
      boxExtents=<15 25 20>
      constraintSetup {angularLimitsMax=<0 5 30>, angularLimitsMin=<0 -20 -30>, angularTarget=<0 0 0>}
      localJointOffset=<4 0 0>
     }]
   }
   Component To Local(ComponentPose=@n21)
   n22 = RigidBody!(ComponentPose=Local To Component(LocalPose=Use cached pose 'basePoses'()), Alpha, OverrideWorldGravity:Vector="0,0,-500", ExternalForce) -> (Pose) {
    bClampLinearTranslationLimitToRefPose=true
    bOverrideWorldGravity=true
    overridePhysicsAsset=&/M_MED_Apprentice_FaceAcc_Dyn_Hair
    simulationSpace=BaseBoneSpace
    alphaBoolBlend {blendOption=Linear}
    alphaScaleBiasClamp {interpSpeedDecreasing=4, interpSpeedIncreasing=6}
    baseBoneRef {boneName=spine_05}
    componentAppliedLinearAccClamp=<3000 3000 3000>
    componentLinearAccScale=<0.25 0.35 0.35>
    componentLinearVelScale=<0.3 0.3 0.3>
    overrideWorldGravity {z=-875}
    simSpaceSettings {maxAngularAcceleration=200, maxAngularVelocity=5, maxLinearAcceleration=500, maxLinearVelocity=200}
   }
   n23 = Component To Local(ComponentPose=@n22)
   n24 = "Sequence Player (additive) 'M_Med_Apprentice_FaceAcc_Hair_Anim'"() {sequence=&M_Med_Apprentice_FaceAcc_Hair_Anim}
   n25 = Look At - Bone: dyn_hair_l_MSTR(ComponentPose=Local To Component(LocalPose=Use cached pose 'basePoses'()), Alpha=1) {
    alphaBoolBlend {blendOption=Linear}
    boneToModify {boneName=dyn_hair_l_MSTR}
    lookAtLocation=<-20 -39 -18>
    lookAtTarget {boneReference {boneName=spine_05}}
    lookAt_Axis {axis=<1 0>}
   }
   n26 = Look At - Bone: dyn_hair_bk_MSTR(ComponentPose=@n25, Alpha=1) {
    alphaBoolBlend {blendOption=Linear}
    boneToModify {boneName=dyn_hair_bk_MSTR}
    lookAtLocation=<-20 -39>
    lookAtTarget {boneReference {boneName=spine_05}}
    lookAt_Axis {axis=<1 0>}
   }
   n27 = Look At - Bone: dyn_hair_r_MSTR(ComponentPose=@n26, Alpha=1) {
    alphaBoolBlend {blendOption=Linear}
    boneToModify {boneName=dyn_hair_r_MSTR}
    lookAtLocation=<-20 -39 18>
    lookAtTarget {boneReference {boneName=spine_05}}
    lookAt_Axis {axis=<1 0>}
   }
   n28 = "Transform (Modify) Bone - Bone: dyn_hair_r_01"(ComponentPose=@n27, Alpha=Get head_twist_l() -> (head_twist_l:double), Translation="0,0,8", Rotation="0,0,0", Scale="1,1,1") {
    rotationMode=BMM_Additive
    rotationSpace=BCS_BoneSpace
    translationMode=BMM_Additive
    translationSpace=BCS_BoneSpace
    alphaBoolBlend {blendOption=Linear}
    boneToModify {boneName=dyn_hair_r_01}
    translation {z=8}
   }
   n29 = Get head_rot_r() -> (head_rot_r:double)
   n30 = "Transform (Modify) Bone - Bone: dyn_hair_r_01"(ComponentPose=@n28, Alpha=@n29, Translation="0,0,5", Rotation="0,-20,0.12", Scale="1,1,1") {
    rotationMode=BMM_Additive
    rotationSpace=BCS_BoneSpace
    translationMode=BMM_Additive
    translationSpace=BCS_BoneSpace
    alphaBoolBlend {blendOption=Linear}
    boneToModify {boneName=dyn_hair_r_01}
    rotation {roll=0.12, yaw=-20}
    translation {z=5}
   }
   n31 = "Transform (Modify) Bone - Bone: dyn_hair_bk_01"(ComponentPose=@n30, Alpha=@n29, Translation="0,0,2", Rotation="0,-10,0", Scale="1,1,1") {
    rotationMode=BMM_Additive
    rotationSpace=BCS_BoneSpace
    translationMode=BMM_Additive
    translationSpace=BCS_BoneSpace
    alphaBoolBlend {blendOption=Linear}
    boneToModify {boneName=dyn_hair_bk_01}
    rotation {yaw=-10}
    translation {z=2}
   }
   n32 = "Transform (Modify) Bone - Bone: dyn_hair_l_01"(ComponentPose=@n31, Alpha=@n29, Translation="0,0,0", Rotation="0,-10,0", Scale="1,1,1") {
    rotationMode=BMM_Additive
    rotationSpace=BCS_BoneSpace
    translationMode=BMM_Additive
    translationSpace=BCS_BoneSpace
    alphaBoolBlend {blendOption=Linear}
    boneToModify {boneName=dyn_hair_l_01}
    rotation {yaw=-10}
   }
   n33 = "Constraint - dyn_hair_bk_03 (P)"(ComponentPose=@n32, Alpha=1, ConstraintWeights_0=1) {
    alphaBoolBlend {blendOption=Linear}
    boneToModify {boneName=dyn_hair_bk_03}
    constraintSetup [{targetBone {boneName=spine_03}}]
    constraintWeights [1]
   }
   n34 = "Constraint - dyn_hair_r_03 (P)"(ComponentPose=@n33, Alpha=1, ConstraintWeights_0=1) {
    alphaBoolBlend {blendOption=Linear}
    boneToModify {boneName=dyn_hair_r_03}
    constraintSetup [{targetBone {boneName=spine_03}}]
    constraintWeights [1]
   }
   n35 = "Constraint - dyn_hair_l_03 (P)"(ComponentPose=@n34, Alpha=1, ConstraintWeights_0=1) {
    alphaBoolBlend {blendOption=Linear}
    boneToModify {boneName=dyn_hair_l_03}
    constraintSetup [{targetBone {boneName=spine_03}}]
    constraintWeights [1]
   }
   n36 = Component To Local(ComponentPose=@n35)
   n37 = Get GravityOverride()
   n38 = RigidBody(ComponentPose=Local To Component(LocalPose=Use cached pose 'prePoseBaseHair'()), OverrideWorldGravity=@n37) {
    bClampLinearTranslationLimitToRefPose=true
    bOverrideWorldGravity=true
    overridePhysicsAsset=&M_MED_Apprentice_FaceAcc_Dyn_Hair
    simulationSpace=BaseBoneSpace
    alphaBoolBlend {blendOption=Linear}
    alphaScaleBiasClamp {interpSpeedDecreasing=4, interpSpeedIncreasing=6}
    baseBoneRef {boneName=spine_05}
    componentAppliedLinearAccClamp=<3000 3000 3000>
    componentLinearAccScale=<0.5 0.5 0.5>
    componentLinearVelScale=<0.3 0.3 0.3>
    overrideWorldGravity {z=-875}
    simSpaceSettings {maxAngularAcceleration=200, maxAngularVelocity=5, maxLinearAcceleration=500, maxLinearVelocity=200}
   }
   n39 = RigidBody!(
    ComponentPose=Local To Component(LocalPose=Use cached pose 'prePoseBaseHair'())
    Alpha
    OverrideWorldGravity=@n37
    ExternalForce
    ComponentLinearAccScale:Vector="0.2,0.2,0.1"
   ) -> (Pose) {
    bClampLinearTranslationLimitToRefPose=true
    bOverrideWorldGravity=true
    overridePhysicsAsset=&M_MED_Apprentice_FaceAcc_Dyn_Hair
    simulationSpace=BaseBoneSpace
    alphaBoolBlend {blendOption=Linear}
    alphaScaleBiasClamp {interpSpeedDecreasing=4, interpSpeedIncreasing=6}
    baseBoneRef {boneName=spine_05}
    componentAppliedLinearAccClamp=<800 800 500>
    componentLinearAccScale=<0.6 0.6 0.2>
    componentLinearVelScale=<0.3 0.3 0.3>
    overrideWorldGravity {z=-875}
    simSpaceSettings {maxAngularAcceleration=200, maxAngularVelocity=5, maxLinearAcceleration=500, maxLinearVelocity=200}
   }
   n40 = Blend Poses by bool(
    BlendPose_0=Component To Local(ComponentPose=@n39)
    BlendPose_1=Component To Local(ComponentPose=@n38)
    bActiveValue=Get bIsParachuteOpened()
   ) {blendType=Linear}
   n41 = RigidBody(
    ComponentPose=Local To Component(LocalPose=Use cached pose 'prePoseBaseHair'())
    OverrideWorldGravity=@n37
    ComponentLinearAccScale="0.2,0.2,0.2"
   ) {
    bClampLinearTranslationLimitToRefPose=true
    bOverrideWorldGravity=true
    overridePhysicsAsset=&/M_MED_Apprentice_FaceAcc_Dyn_Hair_Floating
    simulationSpace=BaseBoneSpace
    alphaBoolBlend {blendOption=Linear}
    alphaScaleBiasClamp {interpSpeedDecreasing=4, interpSpeedIncreasing=6}
    baseBoneRef {boneName=spine_05}
    componentAppliedLinearAccClamp=<800 800 500>
    componentLinearAccScale=<0.6 0.6 0.2>
    componentLinearVelScale=<0.3 0.3 0.3>
    overrideWorldGravity {z=-875}
    simSpaceSettings {maxAngularAcceleration=200, maxAngularVelocity=5, maxLinearAcceleration=500, maxLinearVelocity=200}
   }
   n42 = Component To Local(ComponentPose=@n41)
   n43 = Get bIsSkydiveFloating()
   n44 = Anim Dynamics - Bone: dyn_hairTips_base_01!(
    ComponentPose=Local To Component(LocalPose=Blend Poses by bool(BlendPose_0=@n42, BlendPose_1=@n40, bActiveValue=@n43) {blendType=Linear})
    Alpha=1
    GravityScale:float=1
    ExternalForce="0,0,0"
   ) -> (Pose) {
    angularDampingOverride=1
    angularSpringConstant=400
    bAngularSpring=true
    bGravityOverrideInSimSpace=true
    bOverrideAngularDamping=true
    bUseGravityOverride=true
    simulationSpace=RootRelative
    alphaBoolBlend {blendOption=Linear}
    boundBone {boneName=dyn_hairTips_base_01}
    componentAppliedLinearAccClamp=<1000 1000 1000>
    componentLinearAccScale=<0.4 0.4 0.4>
    componentLinearVelScale=<0.6 0.6 0.6>
    gravityOverride {z=-980}
    physicsBodyDefinitions [{
      sphereCollisionRadius=0
      boundBone {boneName=dyn_hairTips_base_01}
      boxExtents=<15 25 20>
      constraintSetup {angularLimitsMax=<0 15 30>, angularLimitsMin=<0 -15 -30>, angularTarget=<0 0 0>}
      localJointOffset=<-3 -0.2 -0.5>
     }]
    relativeSpaceBone {boneName=head}
   }
   n45 = Anim Dynamics - Bone: dyn_hairTips_base_02!(ComponentPose=@n44, Alpha=1, GravityScale:float=1, ExternalForce="0,0,0") -> (Pose) {
    angularDampingOverride=1
    angularSpringConstant=400
    bAngularSpring=true
    bGravityOverrideInSimSpace=true
    bOverrideAngularDamping=true
    bUseGravityOverride=true
    simulationSpace=RootRelative
    alphaBoolBlend {blendOption=Linear}
    boundBone {boneName=dyn_hairTips_base_02}
    componentAppliedLinearAccClamp=<1000 1000 1000>
    componentLinearAccScale=<0.4 0.4 0.4>
    componentLinearVelScale=<0.6 0.6 0.6>
    gravityOverride {z=-980}
    physicsBodyDefinitions [{
      sphereCollisionRadius=0
      boundBone {boneName=dyn_hairTips_base_02}
      boxExtents=<15 25 20>
      constraintSetup {angularLimitsMax=<0 15 30>, angularLimitsMin=<0 -15 -30>, angularTarget=<0 0 0>}
      localJointOffset=<-2 1.5 -1>
     }]
    relativeSpaceBone {boneName=head}
   }
   n46 = Anim Dynamics - Bone: dyn_hairTips_base_03!(ComponentPose=@n45, Alpha=1, GravityScale:float=1, ExternalForce="0,0,0") -> (Pose) {
    angularDampingOverride=1
    angularSpringConstant=400
    bAngularSpring=true
    bGravityOverrideInSimSpace=true
    bOverrideAngularDamping=true
    bUseGravityOverride=true
    simulationSpace=RootRelative
    alphaBoolBlend {blendOption=Linear}
    boundBone {boneName=dyn_hairTips_base_03}
    componentAppliedLinearAccClamp=<1000 1000 1000>
    componentLinearAccScale=<0.4 0.4 0.4>
    componentLinearVelScale=<0.6 0.6 0.6>
    gravityOverride {z=-980}
    physicsBodyDefinitions [{
      sphereCollisionRadius=0
      boundBone {boneName=dyn_hairTips_base_03}
      boxExtents=<15 25 20>
      constraintSetup {angularLimitsMax=<0 5 30>, angularLimitsMin=<0 -20 -30>, angularTarget=<0 0 0>}
      localJointOffset=<3 0 -1.5>
     }]
    relativeSpaceBone {boneName=head}
   }
   n47 = Anim Dynamics - Bone: dyn_hairTips_base_l_01!(ComponentPose=@n46, Alpha=1, GravityScale:float=1, ExternalForce="0,0,0") -> (Pose) {
    angularDampingOverride=0.9999
    angularSpringConstant=800
    bAngularSpring=true
    bGravityOverrideInSimSpace=true
    bOverrideAngularDamping=true
    bUseGravityOverride=true
    simulationSpace=RootRelative
    alphaBoolBlend {blendOption=Linear}
    boundBone {boneName=dyn_hairTips_base_l_01}
    componentAppliedLinearAccClamp=<1000 1000 1000>
    componentLinearAccScale=<0.4 0.4 0.4>
    componentLinearVelScale=<0.6 0.6 0.6>
    gravityOverride {z=-980}
    physicsBodyDefinitions [{
      sphereCollisionRadius=0
      boundBone {boneName=dyn_hairTips_base_l_01}
      boxExtents=<15 25 20>
      constraintSetup {angularLimitsMax=<0 5 30>, angularLimitsMin=<0 -35 -30>, angularTarget=<0 0 0>}
      localJointOffset=<5 0 -3>
     }]
   }
   n48 = Anim Dynamics - Bone: dyn_hairTips_base_l_02!(ComponentPose=@n47, Alpha=1, GravityScale:float=1, ExternalForce="0,0,0") -> (Pose) {
    angularDampingOverride=0.9999
    angularSpringConstant=800
    bAngularSpring=true
    bGravityOverrideInSimSpace=true
    bOverrideAngularDamping=true
    bUseGravityOverride=true
    simulationSpace=RootRelative
    alphaBoolBlend {blendOption=Linear}
    boundBone {boneName=dyn_hairTips_base_l_02}
    componentAppliedLinearAccClamp=<1000 1000 1000>
    componentLinearAccScale=<0.4 0.4 0.4>
    componentLinearVelScale=<0.6 0.6 0.6>
    gravityOverride {z=-980}
    physicsBodyDefinitions [{
      sphereCollisionRadius=0
      boundBone {boneName=dyn_hairTips_base_l_02}
      boxExtents=<15 25 20>
      constraintSetup {angularLimitsMax=<0 5 30>, angularLimitsMin=<0 -35 -30>, angularTarget=<0 0 0>}
      localJointOffset=<5 -1 -2>
     }]
   }
   n49 = Anim Dynamics - Bone: dyn_hairTips_base_l_03!(ComponentPose=@n48, Alpha=1, GravityScale:float=1, ExternalForce="0,0,0") -> (Pose) {
    angularDampingOverride=0.9999
    angularSpringConstant=800
    bAngularSpring=true
    bGravityOverrideInSimSpace=true
    bOverrideAngularDamping=true
    bUseGravityOverride=true
    simulationSpace=RootRelative
    alphaBoolBlend {blendOption=Linear}
    boundBone {boneName=dyn_hairTips_base_l_03}
    componentAppliedLinearAccClamp=<1000 1000 1000>
    componentLinearAccScale=<0.4 0.4 0.4>
    componentLinearVelScale=<0.6 0.6 0.6>
    gravityOverride {z=-980}
    physicsBodyDefinitions [{
      sphereCollisionRadius=0
      boundBone {boneName=dyn_hairTips_base_l_03}
      boxExtents=<15 25 20>
      constraintSetup {angularLimitsMax=<0 5 5>, angularLimitsMin=<0 -20 -30>, angularTarget=<0 0 0>}
      localJointOffset=<4 -0.8 -0.5>
     }]
   }
   n50 = Anim Dynamics - Bone: dyn_hairTips_base_r_02!(ComponentPose=@n49, Alpha=1, GravityScale:float=1, ExternalForce="0,0,0") -> (Pose) {
    angularDampingOverride=0.9999
    angularSpringConstant=800
    bAngularSpring=true
    bGravityOverrideInSimSpace=true
    bOverrideAngularDamping=true
    bUseGravityOverride=true
    simulationSpace=RootRelative
    alphaBoolBlend {blendOption=Linear}
    boundBone {boneName=dyn_hairTips_base_r_02}
    componentAppliedLinearAccClamp=<1000 1000 1000>
    componentLinearAccScale=<0.4 0.4 0.4>
    componentLinearVelScale=<0.6 0.6 0.6>
    gravityOverride {z=-980}
    physicsBodyDefinitions [{
      sphereCollisionRadius=0
      boundBone {boneName=dyn_hairTips_base_r_02}
      boxExtents=<15 25 20>
      constraintSetup {angularLimitsMax=<0 5 30>, angularLimitsMin=<0 -25 -10>, angularTarget=<0 0 0>}
      localJointOffset=<4 0.4 -0.8>
     }]
   }
   n51 = Anim Dynamics - Bone: dyn_hairTips_base_bk_00!(ComponentPose=@n50, Alpha=1, GravityScale:float=1, ExternalForce="0,0,0") -> (Pose) {
    angularDampingOverride=1
    angularSpringConstant=400
    bAngularSpring=true
    bGravityOverrideInSimSpace=true
    bOverrideAngularDamping=true
    bUseGravityOverride=true
    simulationSpace=RootRelative
    alphaBoolBlend {blendOption=Linear}
    boundBone {boneName=dyn_hairTips_base_bk_00}
    componentAppliedLinearAccClamp=<1000 1000 1000>
    componentLinearAccScale=<0.4 0.4 0.4>
    componentLinearVelScale=<0.6 0.6 0.6>
    gravityOverride {z=-980}
    physicsBodyDefinitions [{
      sphereCollisionRadius=0
      boundBone {boneName=dyn_hairTips_base_bk_00}
      boxExtents=<15 25 20>
      constraintSetup {angularLimitsMax=<0 5 30>, angularLimitsMin=<0 -15 -30>, angularTarget=<0 0 0>}
      localJointOffset=<5 0 -5>
     }]
    relativeSpaceBone {boneName=head}
   }
   n52 = Anim Dynamics - Bone: dyn_hairTips_base_bk_01!(ComponentPose=@n51, Alpha=1, GravityScale:float=1, ExternalForce="0,0,0") -> (Pose) {
    angularDampingOverride=0.9999
    angularSpringConstant=800
    bAngularSpring=true
    bGravityOverrideInSimSpace=true
    bOverrideAngularDamping=true
    bUseGravityOverride=true
    simulationSpace=RootRelative
    alphaBoolBlend {blendOption=Linear}
    boundBone {boneName=dyn_hairTips_base_bk_01}
    componentAppliedLinearAccClamp=<1000 1000 1000>
    componentLinearAccScale=<0.4 0.4 0.4>
    componentLinearVelScale=<0.6 0.6 0.6>
    gravityOverride {z=-980}
    physicsBodyDefinitions [{
      sphereCollisionRadius=0
      boundBone {boneName=dyn_hairTips_base_bk_01}
      boxExtents=<15 25 20>
      constraintSetup {angularLimitsMax=<0 5 20>, angularLimitsMin=<0 -20 -30>, angularTarget=<0 0 0>}
      localJointOffset=<2 -0.2 -0.8>
     }]
   }
   n53 = Anim Dynamics - Bone: dyn_hairTips_base_bk_02!(ComponentPose=@n52, Alpha=1, GravityScale:float=1, ExternalForce="0,0,0") -> (Pose) {
    angularDampingOverride=0.9999
    angularSpringConstant=800
    bAngularSpring=true
    bGravityOverrideInSimSpace=true
    bOverrideAngularDamping=true
    bUseGravityOverride=true
    simulationSpace=RootRelative
    alphaBoolBlend {blendOption=Linear}
    boundBone {boneName=dyn_hairTips_base_bk_02}
    componentAppliedLinearAccClamp=<1000 1000 1000>
    componentLinearAccScale=<0.4 0.4 0.4>
    componentLinearVelScale=<0.6 0.6 0.6>
    gravityOverride {z=-980}
    physicsBodyDefinitions [{
      sphereCollisionRadius=0
      boundBone {boneName=dyn_hairTips_base_bk_02}
      boxExtents=<15 25 20>
      constraintSetup {angularLimitsMax=<0 5 30>, angularLimitsMin=<0 -20 -30>, angularTarget=<0 0 0>}
      localJointOffset=<3.5 0 0>
     }]
   }
   n54 = Anim Dynamics - Bone: dyn_hairTips_base_bk_03!(ComponentPose=@n53, Alpha=1, GravityScale:float=1, ExternalForce="0,0,0") -> (Pose) {
    angularDampingOverride=0.9999
    angularSpringConstant=800
    bAngularSpring=true
    bGravityOverrideInSimSpace=true
    bOverrideAngularDamping=true
    bUseGravityOverride=true
    simulationSpace=RootRelative
    alphaBoolBlend {blendOption=Linear}
    boundBone {boneName=dyn_hairTips_base_bk_03}
    componentAppliedLinearAccClamp=<1000 1000 1000>
    componentLinearAccScale=<0.4 0.4 0.4>
    componentLinearVelScale=<0.6 0.6 0.6>
    gravityOverride {z=-980}
    physicsBodyDefinitions [{
      sphereCollisionRadius=0
      boundBone {boneName=dyn_hairTips_base_bk_03}
      boxExtents=<15 25 20>
      constraintSetup {angularLimitsMax=<0 5 30>, angularLimitsMin=<0 -20 -30>, angularTarget=<0 0 0>}
      localJointOffset=<4 0 0>
     }]
   }
   n55 = Component To Local(ComponentPose=@n54)
   n56 = Get bRigidBodyEnabled()
   Output Pose(Result=Blend Poses by bool(
     BlendPose_0=Apply Additive(Base=@n23, Additive=@n24) {alphaBoolBlend {blendOption=Linear}}
     BlendPose_1=Blend Poses by bool(BlendPose_0=@n55, BlendPose_1=@n36, bActiveValue=@n56) {blendType=Linear}
     BlendTime_0=0.25
     bActiveValue=Get reactiveBoolean() -> (reactiveBoolean:bool)
    ) {blendType=Linear}) {layerGroup=None}
   n57 = Look At - Bone: dyn_hair_l_MSTR(ComponentPose=Local To Component(LocalPose=Use cached pose 'basePoses'()), Alpha=1) {
    alphaBoolBlend {blendOption=Linear}
    boneToModify {boneName=dyn_hair_l_MSTR}
    lookAtLocation=<-20 -39 -18>
    lookAtTarget {boneReference {boneName=spine_05}}
    lookAt_Axis {axis=<1 0>}
   }
   n58 = Look At - Bone: dyn_hair_bk_MSTR(ComponentPose=@n57, Alpha=1) {
    alphaBoolBlend {blendOption=Linear}
    boneToModify {boneName=dyn_hair_bk_MSTR}
    lookAtLocation=<-20 -39>
    lookAtTarget {boneReference {boneName=spine_05}}
    lookAt_Axis {axis=<1 0>}
   }
   n59 = Look At - Bone: dyn_hair_r_MSTR(ComponentPose=@n58, Alpha=1) {
    alphaBoolBlend {blendOption=Linear}
    boneToModify {boneName=dyn_hair_r_MSTR}
    lookAtLocation=<-20 -39 18>
    lookAtTarget {boneReference {boneName=spine_05}}
    lookAt_Axis {axis=<1 0>}
   }
   n60 = "Transform (Modify) Bone - Bone: dyn_hair_offset"(ComponentPose=@n59, Alpha=1, Translation="0,0,0", Rotation=Get head_twist() -> (head_twist:Rotator), Scale="1,1,1") {
    rotationMode=BMM_Additive
    rotationSpace=BCS_BoneSpace
    translationMode=BMM_Additive
    translationSpace=BCS_BoneSpace
    alphaBoolBlend {blendOption=Linear}
    boneToModify {boneName=dyn_hair_offset}
   }
   n61 = Get head_rot_bk() -> (head_rot_bk:double)
   n62 = "Transform (Modify) Bone - Bone: dyn_hair_bk_MSTR"(ComponentPose=@n60, Alpha=@n61, Translation="0,0,3", Rotation="0,0,0", Scale="1,1,1") {
    rotationMode=BMM_Additive
    rotationSpace=BCS_BoneSpace
    translationMode=BMM_Additive
    translationSpace=BCS_BoneSpace
    alphaBoolBlend {blendOption=Linear}
    boneToModify {boneName=dyn_hair_bk_MSTR}
    translation {z=3}
   }
   n63 = "Transform (Modify) Bone - Bone: dyn_hair_r_MSTR"(ComponentPose=@n62, Alpha=@n61, Translation="0,0,5", Rotation="0,0,0", Scale="1,1,1") {
    rotationMode=BMM_Additive
    rotationSpace=BCS_BoneSpace
    translationMode=BMM_Additive
    translationSpace=BCS_BoneSpace
    alphaBoolBlend {blendOption=Linear}
    boneToModify {boneName=dyn_hair_r_MSTR}
    translation {z=5}
   }
   n64 = "Transform (Modify) Bone - Bone: dyn_hair_l_MSTR"(ComponentPose=@n63, Alpha=@n61, Translation="0,0,5", Rotation="0,0,0", Scale="1,1,1") {
    rotationMode=BMM_Additive
    rotationSpace=BCS_BoneSpace
    translationMode=BMM_Additive
    translationSpace=BCS_BoneSpace
    alphaBoolBlend {blendOption=Linear}
    boneToModify {boneName=dyn_hair_l_MSTR}
    translation {z=5}
   }
   n65 = Get head_rot_r()
   n66 = "Transform (Modify) Bone - Bone: dyn_hair_r_MSTR"(ComponentPose=@n64, Alpha=@n65, Translation="0,-6,10", Rotation="0,-50,0", Scale="1,1,1") {
    rotationMode=BMM_Additive
    rotationSpace=BCS_BoneSpace
    translationMode=BMM_Additive
    translationSpace=BCS_BoneSpace
    alphaBoolBlend {blendOption=Linear}
    boneToModify {boneName=dyn_hair_r_MSTR}
    rotation {yaw=-50}
    translation {y=-6, z=10}
   }
   n67 = "Transform (Modify) Bone - Bone: dyn_hair_bk_MSTR"(ComponentPose=@n66, Alpha=@n65, Translation="0,0,5", Rotation="0,-30,0", Scale="1,1,1") {
    rotationMode=BMM_Additive
    rotationSpace=BCS_BoneSpace
    translationMode=BMM_Additive
    translationSpace=BCS_BoneSpace
    alphaBoolBlend {blendOption=Linear}
    boneToModify {boneName=dyn_hair_bk_MSTR}
    rotation {yaw=-30}
    translation {z=5}
   }
   Save cached pose 'prePoseBaseHair'(Pose=Component To Local(ComponentPose=@n67)) {cacheName=prePoseBaseHair}
   Save cached pose 'MeshPose'(Pose=Copy Pose From Mesh(SourceMeshComponent=Get Body_Mesh())) {cacheName=MeshPose}
   n68 = Layered blend per bone!(
    BasePose=Copy Pose From Mesh() {bUseAttachedParent=true}
    BlendPoses_0=Copy Pose From Mesh(SourceMeshComponent=Get Body_Mesh())
    BlendPoses_1:Pose Link Structure
    BlendWeights_0
    BlendWeights_1:float=1
   ) -> (Pose) {
    blendPoses [{linkId=-1, sourceLinkId=-1}, {linkId=-1, sourceLinkId=-1}]
    blendWeights [1,1]
    layerSetup [{branchFilters [{boneName=upperarm_l}, {boneName=upperarm_r}]}, {branchFilters []}]
   }
   Save cached pose 'basePoses'(Pose=@n68) {cacheName=basePoses}
   Use cached pose 'prePoseBaseHair'()
  ]}
 EventGraph {nodes [
   Event BlueprintUpdateAnimation()
   > n1 = Sequence!(execute) -> (then_0, then_1, then_2:Exec)
   n2 = Get Body_Mesh()
   > Set head_twist_l(head_twist_l:double=GetAlphaFromReferenceTransformDelta(SkeletalMeshComponent=@n2, BoneName=head, BaseName=spine_05, DeltaComponentType=QuaternionTwist, InputRange=(X=0,Y=45), TwistAxis="1,0,0")) -> (Output_Get:double)
   > Set head_twist(head_twist:Rotator=MakeRotator(Roll=GetAlphaFromReferenceTransformDelta(
      SkeletalMeshComponent=@n2
      BoneName=head
      BaseName=spine_05
      DeltaComponentType=QuaternionTwist
      InputRange=(X=-80,Y=80)
      OutputRange=(X=15,Y=-15)
      TwistAxis="1,0,0"
     ))) -> (Output_Get:Rotator)
   > Set head_rot_bk(head_rot_bk:double=GetAlphaFromReferenceTransformDelta(SkeletalMeshComponent=@n2, BoneName=head, BaseName=spine_05, DeltaComponentType=QuaternionTwist, InputRange=(X=0,Y=60), TwistAxis="0,0,1")) -> (Output_Get:double)
   > Set head_rot_r(head_rot_r:double=GetAlphaFromReferenceTransformDelta(SkeletalMeshComponent=@n2, BoneName=head, BaseName=spine_05, DeltaComponentType=QuaternionTwist, InputRange=(X=0,Y=40), TwistAxis="0,1,0")) -> (Output_Get:double)
   n3 = Gravity Override State Function DEPRECATED(
    execute=@n1.then_1
    CharacterPartAnimInstance=Self-Reference()
    GameTimeInSeconds=GetGameTimeInSeconds()
    GravityOverrideParams=Get Gravity Override Params()
   )
   > Set GravityOverride(GravityOverride=@n3)
   n4 = GetPartSkeletalMeshComponent()
   n5 = Get Dynamic Material Instance(execute=@n1.then_2, SkeletalMeshComponent=@n4, ElementIndex=1)
   > n6 = Get Scalar Parameter Value!(execute, self:Material Instance Dynamic Object Reference=@n5, ParameterName:Name=Hide Element 02) -> (then, ReturnValue:float) {is=title K2Node_CallFunction}
   > n7 = Set reactiveHairAlpha(reactiveHairAlpha:double=MapRangeClamped(Value=@n6, InRangeA=0, InRangeB=1, OutRangeA=1, OutRangeB=0)) -> (Output_Get:double)
   > Set reactiveBoolean(reactiveBoolean:bool="To Boolean (Integer)"!(InInt:int=Truncate(A=@n7)) -> (ReturnValue:bool)) -> (Output_Get:bool)
   > n8 = Get Dynamic Material Instance(SkeletalMeshComponent=@n4)
   > n9 = Get Scalar Parameter Value(self=@n8, ParameterName=Hide Element 01)
   > Set baseHairAlpha(baseHairAlpha:double=MapRangeClamped(Value=@n9, InRangeA=0, InRangeB=1, OutRangeA=1, OutRangeB=0)) -> (Output_Get:double)
   n10 = Event BlueprintProcessAnimInputEvent()
   > Branch(Condition=IsNamed(self=@n10.AnimInputEvent, InEventName=aie.EnemyPlayerEliminated))
   > Set onKill(onKill:bool=true) -> (Output_Get:bool)
   > RetriggerableDelay(Duration=6)
   > Set onKill(onKill=false)
   Event BlueprintInitializeAnimation()
   > Set Body_Mesh(Body_Mesh=GetOwnerSkeletalMeshComponent())
  ]}
]
variables [
 baseHairAlpha {type=number}
 Body_Mesh {type=SkeletalMeshComponent}
 Gravity Override Params {
  type=GravityOverrideParamsStruct
  default {
   skydive_FloatingDown {gravityOverride=<0 -700 -100>, windAmplitude=<200 -500 300>, windOffset=<0 -600 100>}
   skydive_GoingUp {gravityOverride=<0 -300 500>, windAmplitude=<200 -500 300>}
   skydive_Parachuting {gravityOverride=<0 -500 -980>, windAmplitude=<200 -150 400>, windFrequency=<1.5 4 6>}
  }
 }
 GravityOverride {type=Vector}
 head_rot_bk {type=number}
 head_rot_r {type=number}
 head_twist {type=Rotator}
 head_twist_l {type=number}
 onKill {type=boolean}
 reactiveBoolean {type=boolean}
 reactiveHairAlpha {type=number}
]
